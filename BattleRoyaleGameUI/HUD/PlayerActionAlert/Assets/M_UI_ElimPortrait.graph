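# Translucent UI material for the elimination portrait in the player action alert HUD, animated through scalar parameters.
# Anim_Translate slides the Portrait texture horizontally, while Anim_Distortion drives chromatic aberration by sampling red and green at opposite offsets and blue at center.
# The three alphas are averaged, clipped by an angled box shape and eased linear gradient, then multiplied by Anim_Opacity and further edge-fade gradients.
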Material /BattleRoyaleGameUI/HUD/PlayerActionAlert/Assets/M_UI_ElimPortrait
graph {
 properties {bAutomaticallySetUsageInEditor=false, blendMode=BLEND_Translucent, materialDomain=MD_UI, usedShadingModels=Default Lit}
 nodes [
  n1 = Constant3Vector() {constant=<rgba 1 0 0 0>}
  n2 = ScalarParameter() {defaultValue=0.1, parameterName=BaseTranslation}
  n3 = Constant()
  n4 = ScalarParameter() {defaultValue=1, parameterName=Anim_Translate}
  n5 = Add(
   A=Add(A=TextureCoordinate(), B=AppendVector(A=@n2, B=@n3))
   B=AppendVector(A=LinearInterpolate(Alpha=@n4) {constA=0.4, constB=0}, B=Constant())
  )
  n6 = AppendVector(A=LinearInterpolate(Alpha=ScalarParameter() {defaultValue=1, parameterName=Anim_Distortion}) {constA=0.1, constB=0}, B=Constant())
  n7 = TextureObjectParameter() {parameterName=Portrait, texture=/Game/UI/Foundation/Textures/Icons/Heroes/Athena/Soldier/T-Soldier-HID-883-Athena-Commando-M-ChOneJonesy}
  n8 = TextureSample(
   UVs=Add(A=NamedRerouteDeclaration(None=@n5) {name=UVs, nodeColor=<rgba 0 0 0 1>}, B=@n6)
   Tex=NamedRerouteDeclaration(None=@n7) {name=Image, nodeColor=<rgba 0 0 0 1>}
  ) {samplerSource=SSM_Clamp_WorldGroupSettings}
  n9 = Constant3Vector() {constant=<rgba 0 1 0 0>}
  n10 = NamedRerouteUsage()
  n11 = TextureSample(UVs=Add(A=@n10, B=Multiply(A=@n6) {constB=-1}), Tex=NamedRerouteUsage()) {samplerSource=SSM_Clamp_WorldGroupSettings}
  n12 = TextureSample(UVs=@n10, Tex=NamedRerouteUsage()) {samplerSource=SSM_Clamp_WorldGroupSettings}
  Add(
   A=Add(A=LinearInterpolate(B=@n1, Alpha=@n8.R), B=LinearInterpolate(B=@n9, Alpha=@n11.G))
   B=LinearInterpolate(B=Constant3Vector() {constant=<rgba 0 0 1 0>}, Alpha=@n12.B)
  )
  n13 = NamedRerouteUsage()
  n14 = NamedRerouteUsage()
  n15 = MaterialFunctionCall(Float2=MaterialFunctionCall() {materialFunction=&GetUserInterfaceUV}.Normalized UV) {materialFunction=&BreakOutFloat2Components}
  n16 = AppendVector(A=@n15.R, B=OneMinus(None=@n15.G))
  n17 = Constant() {r=3}
  n18 = ScalarParameter() {defaultValue=14, parameterName=Angle}
  n19 = ScalarParameter() {defaultValue=0.3333, parameterName=Falloff}
  n20 = Min(
   A=Add(A=Add(A=@n13, B=@n14), B=NamedRerouteUsage())
   B=NamedRerouteDeclaration(None=MaterialFunctionCall!(UV=@n16, Corner=@n17, Angle=@n18, Flip, Falloff=@n19) {materialFunction=/FortUILibrary/MaterialFunctions/Shapes/MF_UI_BoxAngle}.Result) {name=Shape, nodeColor=<rgba 0 0 0 1>}
  )
  n21 = MaterialFunctionCall(GradientStart=Constant() {r=0.5}, RotationAngle=Constant() {r=-0.25}) {materialFunction=&MF_UI_LinearGradient}
  n22 = MaterialFunctionCall("Value (0-1)"=OneMinus(None=ComponentMask(None=@n21.Result) {g=true})) {materialFunction=&MF_UI_EaseCurves}
  n23 = ScalarParameter() {defaultValue=1, parameterName=Anim_Opacity}
  n24 = MaterialFunctionCall(GradientStart=Constant() {r=1}, GradientEnd=Constant() {r=0.3333}, RotationAngle=Constant() {r=-0.025}) {materialFunction=&MF_UI_LinearGradient}
  n25 = Constant() {r=1}
  n26 = Constant() {r=0.9}
  n27 = Constant() {r=0.25}
  Multiply(
   A=Multiply(A=Min(A=@n20, B=@n22.EaseOutCubic), B=LinearInterpolate(Alpha=@n23))
   B=Multiply(
    A=MaterialFunctionCall("Value (0-1)"=@n24.Result) {materialFunction=&MF_UI_EaseCurves}.EaseOutSine
    B=MaterialFunctionCall(GradientStart=@n25, GradientEnd=@n26, RotationAngle=@n27) {materialFunction=&MF_UI_LinearGradient}.Result
   )
  )
  NamedRerouteDeclaration(None=Multiply(A=@n8.A) {constB=0.3333}) {name=R, nodeColor=<rgba 0 0 0 1>}
  NamedRerouteDeclaration(None=Multiply(A=@n11.A) {constB=0.3333}) {name=G, nodeColor=<rgba 0 0 0 1>}
  NamedRerouteDeclaration(None=Multiply(A=@n12.A) {constB=0.3333}) {name=B, nodeColor=<rgba 0 0 0 1>}
  TextureCoordinate()
 ]
}
