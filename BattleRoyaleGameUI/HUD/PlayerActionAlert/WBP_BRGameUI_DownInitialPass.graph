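# Battle Royale HUD alert widget for a knocked player, showing portrait, rank icon and name placeholder on an angled backing with brackets and glow.
# A "Knocked!" flavor text sits in a separate visual attachment above the alert content.
# The ShowAlert animation combines transform, opacity and material-parameter tracks to slam in backing and brackets, pop the player info, and float and fade the flavor text.
# The event graph holds only empty event stubs, with no logic to play the animation or fill data.
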WidgetBlueprint /BattleRoyaleGameUI/HUD/PlayerActionAlert/WBP_BRGameUI_DownInitialPass
animations [ShowAlert {
  binding_count=9
  length_seconds=5
  object=&self:ShowAlert
  sequence {bindings [
    Image_Bracket_Right {
     id=#1
     possessed_class=Image
     tracks [
      {
       type=MovieScene2DTransformTrack
       sections [{
         type=MovieScene2DTransformSection
         channels [
          Translation.X {type=MovieSceneScriptingFloatChannel, keys [[1,256],[4,0]]}
          Translation.Y {type=MovieSceneScriptingFloatChannel, const [0,1]}
          Scale.X {type=MovieSceneScriptingFloatChannel, keys [[1,2],[4,0.9],[5.6667,1]]}
          Scale.Y {type=MovieSceneScriptingFloatChannel, keys [[1,2],[4,1]]}
         ]
        }]
      }
      {
       type=MovieSceneWidgetMaterialTrack
       sections [{
         type=MovieSceneParameterSection
         channels [
          Anim_Color {type=MovieSceneScriptingFloatChannel, keys [[4,1],[4.3333,0],[14.3333,1]]}
          Anim_Opacity {type=MovieSceneScriptingFloatChannel, keys [[1,0],[2,1]]}
          Anim_Stretch {type=MovieSceneScriptingFloatChannel, keys [[3,0],[4,0.75],[8,0]]}
         ]
        }]
      }
     ]
    }
    Glow {
     id=#2
     possessed_class=Image
     tracks [{
       type=MovieSceneWidgetMaterialTrack
       sections [{type=MovieSceneParameterSection, channels [Anim_Opacity {type=MovieSceneScriptingFloatChannel, keys [[3.3333,0],[4.6667,1],[20,0]]}]}]
      }]
    }
    Text_Background {
     id=#3
     possessed_class=CommonTextBlock
     tracks [
      {
       type=MovieScene2DTransformTrack
       sections [{
         type=MovieScene2DTransformSection
         channels [
          Translation.X {type=MovieSceneScriptingFloatChannel, default=0}
          Translation.Y {type=MovieSceneScriptingFloatChannel, keys [[5,0],[20,-32]]}
          Scale.X {type=MovieSceneScriptingFloatChannel, keys [[5,0.5],[20,1]]}
          Scale.Y {type=MovieSceneScriptingFloatChannel, keys [[5,0.5],[20,1]]}
         ]
        }]
      }
      {
       type=MovieSceneWidgetMaterialTrack
       display_name=Font.FontMaterial
       sections [{
         type=MovieSceneParameterSection
         channels [
          Max_Alpha {type=MovieSceneScriptingFloatChannel, keys [[5,0],[10,1],[25,1],[40,0]]}
          Min_Alpha {type=MovieSceneScriptingFloatChannel, keys [[5,0],[10,1],[35,0]]}
         ]
        }]
      }
      {
       type=MovieSceneWidgetMaterialTrack
       display_name=Font.Outline Settings.OutlineMaterial
       sections [{type=MovieSceneParameterSection, channels [AnimOpacity {type=MovieSceneScriptingFloatChannel, keys [[5,0],[25,1],[40,0]]}]}]
      }
     ]
    }
    AlertContent {
     id=#4
     possessed_class=Overlay
     tracks [{
       type=MovieScene2DTransformTrack
       sections [{
         type=MovieScene2DTransformSection
         channels [
          Scale.X {type=MovieSceneScriptingFloatChannel, keys [[3,1],[3.3333,0.75],[8.3333,1]]}
          Scale.Y {type=MovieSceneScriptingFloatChannel, keys [[2.3333,1],[3.3333,0.75],[8.3333,1]]}
         ]
        }]
      }]
    }
    PlayerDownText {
     id=#5
     possessed_class=CommonTextBlock
     tracks [
      {
       type=MovieScene2DTransformTrack
       sections [{
         type=MovieScene2DTransformSection
         channels [
          Angle {type=MovieSceneScriptingFloatChannel, keys [[4,-1.4207],[6.1903,0.4896],[7.6193,-0.094],[8.813,0]]}
          Scale.X {type=MovieSceneScriptingFloatChannel, keys [[4,0.7512],[6,1.1351],[7,0.9371],[8.807,1]]}
          Scale.Y {type=MovieSceneScriptingFloatChannel, keys [[4,0.7512],[6,1.1351],[7,0.9371],[8.807,1]]}
         ]
        }]
      }
      {
       type=MovieSceneFloatTrack
       display_name=Render Opacity
       sections [{type=MovieSceneFloatSection, channels [None {type=MovieSceneScriptingFloatChannel, keys [[4,0],[5,1]]}]}]
      }
     ]
    }
    Backing {
     id=#6
     possessed_class=Image
     tracks [
      {
       type=MovieScene2DTransformTrack
       sections [{
         type=MovieScene2DTransformSection
         channels [
          Scale.X {type=MovieSceneScriptingFloatChannel, keys [[0,5],[4,1]]}
          Scale.Y {type=MovieSceneScriptingFloatChannel, keys [[2,0.25],[4,1]]}
          Shear.X {type=MovieSceneScriptingFloatChannel}
          Shear.Y {type=MovieSceneScriptingFloatChannel}
         ]
        }]
      }
      {
       type=MovieSceneWidgetMaterialTrack
       sections [{
         type=MovieSceneParameterSection
         channels [
          Anim_Opacity {type=MovieSceneScriptingFloatChannel, keys [[0,0],[0.3333,1]]}
          Anim_Color {type=MovieSceneScriptingFloatChannel, keys [[4,1],[8,0]]}
         ]
        }]
      }
     ]
    }
    LazyImage_Habanero_RankIcon {
     id=#7
     possessed_class=FortLazyImage
     tracks [
      {
       type=MovieSceneFloatTrack
       display_name=Render Opacity
       sections [{type=MovieSceneFloatSection, channels [None {type=MovieSceneScriptingFloatChannel, keys [[4,0],[5,1]]}]}]
      }
      {
       type=MovieScene2DTransformTrack
       sections [{
         type=MovieScene2DTransformSection
         channels [
          Translation.X {type=MovieSceneScriptingFloatChannel, const [0,1]}
          Translation.Y {type=MovieSceneScriptingFloatChannel, const [0,1]}
          Scale.X {type=MovieSceneScriptingFloatChannel, keys [[4,0],[6,1.25],[8,0.9],[13,1]]}
          Scale.Y {type=MovieSceneScriptingFloatChannel, keys [[4,0],[6,1.25],[8,0.9],[13,1]]}
          Shear.X {type=MovieSceneScriptingFloatChannel, const [0,1]}
          Shear.Y {type=MovieSceneScriptingFloatChannel, const [0,1]}
         ]
        }]
      }
     ]
    }
    LazyImage_PlayerPortrait {
     id=#8
     possessed_class=FortLazyImage
     tracks [
      {
       type=MovieSceneFloatTrack
       display_name=Render Opacity
       sections [{type=MovieSceneFloatSection, channels [None {type=MovieSceneScriptingFloatChannel, keys [[3,0],[4,1]]}]}]
      }
      {
       type=MovieSceneWidgetMaterialTrack
       sections [{type=MovieSceneParameterSection, channels [Anim_Translate {type=MovieSceneScriptingFloatChannel, keys [[3,0],[10,1]]}]}]
      }
     ]
    }
    Image_Bracket_Left {
     id=#9
     possessed_class=Image
     tracks [
      {
       type=MovieScene2DTransformTrack
       sections [{
         type=MovieScene2DTransformSection
         channels [
          Translation.X {type=MovieSceneScriptingFloatChannel, keys [[1,-256],[4,0]]}
          Translation.Y {type=MovieSceneScriptingFloatChannel, const [0,1]}
          Scale.X {type=MovieSceneScriptingFloatChannel, keys [[1,2],[4,0.9],[5.6667,1]]}
          Scale.Y {type=MovieSceneScriptingFloatChannel, keys [[1,2],[4,1]]}
         ]
        }]
      }
      {
       type=MovieSceneWidgetMaterialTrack
       sections [{
         type=MovieSceneParameterSection
         channels [
          Anim_Color {type=MovieSceneScriptingFloatChannel, keys [[4,1],[4.3333,0],[14.3333,1]]}
          Anim_Opacity {type=MovieSceneScriptingFloatChannel, keys [[1,0],[2,1]]}
          Anim_Stretch {type=MovieSceneScriptingFloatChannel, keys [[3,0],[4,0.75],[8,0]]}
         ]
        }]
      }
     ]
    }
   ]}
 }]
graphs [EventGraph {nodes [
   Event PreConstruct()
   Event Construct()
   Event Tick()
  ]}]
variables []
widgets {
 info {inheritedWidgetCount=0, parentClass=&UserWidget, rootWidgetClass=&Overlay, widgetCount=14}
 widgets [
  Overlay_All {widgetClassPath=&Overlay, widget=Overlay_All}
  FlavorText {
   widgetClassPath=&FortVisualAttachment
   parent=Overlay_All
   widget=FlavorText
   properties {contentAnchor=<0.5 0>}
   slot {class=OverlaySlot, properties {horizontalAlignment=HAlign_Center}}
  }
  Text_Background {
   widgetClassPath=&CommonTextBlock
   parent=FlavorText
   widget=Text_Background
   properties {
    justification=Center
    mobileFontSizeMultiplier=0.85
    text="Knocked!"
    textTransformPolicy=ToUpper
    font {
     bForceMonospaced=false
     bMaterialIsStencil=false
     fontMaterial=/BattleRoyaleGameUI/Shared/M_UI_Font_FlavorText_Master
     fontObject=&Burbank
     letterSpacing=0
     monospacedWidth=1
     size=36
     skewAmount=0.25
     typefaceFontName=BurbankRegular-Black
     outlineSettings {
      bApplyOutlineToDropShadows=false
      bMiteredCorners=false
      bSeparateFillAlpha=false
      outlineMaterial=/BattleRoyaleGameUI/Shared/MI_UI_Font_SoftShadow_LargeText
      outlineSize=10
      outlineColor=<rgba 1 1 1 1>
     }
    }
    margin=<ltrb 0 -12 0 -16>
    shadowOffset=<0 0>
    strikeBrush {
     bIsDynamicallyLoaded=false
     drawAs=Image
     imageType=NoImage
     mirroring=NoMirror
     resourceName=None
     resourceObject=null
     tiling=NoTile
     imageSize=<32 32>
     margin=<ltrb 0 0 0 0>
     outlineSettings {
      bUseBrushTransparency=false
      roundingType=HalfHeightRadius
      width=0
      color {colorUseRule=UseColor_Specified, specifiedColor=<rgba 0 0 0 0>}
      cornerRadii=<0 0 0 0>
     }
     tintColor {colorUseRule=UseColor_Specified, specifiedColor=<rgba 0.9 0.9 0.9 1>}
     uVRegion {bIsValid=false, max=<0 0>, min=<0 0>}
    }
   }
   slot {class=SizeBoxSlot, properties {padding=<ltrb -2 0 0 0>}}
  }
  AlertContent {
   widgetClassPath=&Overlay
   parent=Overlay_All
   widget=AlertContent
   slot {class=OverlaySlot, properties {horizontalAlignment=HAlign_Fill, verticalAlignment=VAlign_Bottom}}
  }
  BackingandDecor {
   widgetClassPath=&Overlay
   parent=AlertContent
   widget=BackingandDecor
   slot {class=OverlaySlot, properties {horizontalAlignment=HAlign_Fill, verticalAlignment=VAlign_Fill}}
  }
  Glow {
   widgetClassPath=&Image
   parent=BackingandDecor
   widget=Glow
   properties {
    visibility=SelfHitTestInvisible
    brush {
     bIsDynamicallyLoaded=false
     drawAs=Image
     imageType=FullColor
     mirroring=NoMirror
     resourceName=None
     resourceObject=/BattleRoyaleGameUI/HUD/PlayerActionAlert/Assets/MI_UI_AngledBox_Glow_Knocked
     tiling=NoTile
     imageSize=<32 32>
     margin=<ltrb 0 0 0 0>
     outlineSettings {
      bUseBrushTransparency=false
      roundingType=HalfHeightRadius
      width=0
      color {colorUseRule=UseColor_Specified, specifiedColor=<rgba 0 0 0 0>}
      cornerRadii=<0 0 0 0>
     }
     tintColor {colorUseRule=UseColor_Specified, specifiedColor=<rgba 1 1 1 1>}
     uVRegion {bIsValid=false, max=<0 0>, min=<0 0>}
    }
   }
   slot {class=OverlaySlot, properties {horizontalAlignment=HAlign_Fill, verticalAlignment=VAlign_Fill, padding=<ltrb -56 -32 -56 -56>}}
  }
  Backing {
   widgetClassPath=&Image
   parent=BackingandDecor
   widget=Backing
   properties {
    visibility=SelfHitTestInvisible
    brush {
     bIsDynamicallyLoaded=false
     drawAs=Image
     imageType=FullColor
     mirroring=NoMirror
     resourceName=None
     resourceObject=/BattleRoyaleGameUI/HUD/PlayerActionAlert/Assets/MI_UI_Knocked_Backing
     tiling=NoTile
     imageSize=<32 32>
     margin=<ltrb 0 0 0 0>
     outlineSettings {
      bUseBrushTransparency=false
      roundingType=HalfHeightRadius
      width=0
      color {colorUseRule=UseColor_Specified, specifiedColor=<rgba 0 0 0 0>}
      cornerRadii=<0 0 0 0>
     }
     tintColor {colorUseRule=UseColor_Specified, specifiedColor=<rgba 1 1 1 1>}
     uVRegion {bIsValid=false, max=<0 0>, min=<0 0>}
    }
   }
   slot {class=OverlaySlot, properties {horizontalAlignment=HAlign_Fill, verticalAlignment=VAlign_Center}}
  }
  Image_Bracket_Left {
   widgetClassPath=&Image
   parent=BackingandDecor
   widget=Image_Bracket_Left
   properties {
    visibility=SelfHitTestInvisible
    brush {
     bIsDynamicallyLoaded=false
     drawAs=Image
     imageType=FullColor
     mirroring=NoMirror
     resourceName=None
     resourceObject=/BattleRoyaleGameUI/HUD/PlayerActionAlert/Assets/MI_UI_Knocked_Bracket
     tiling=NoTile
     imageSize=<56 48>
     margin=<ltrb 0 0 0 0>
     outlineSettings {
      bUseBrushTransparency=false
      roundingType=HalfHeightRadius
      width=0
      color {colorUseRule=UseColor_Specified, specifiedColor=<rgba 0 0 0 0>}
      cornerRadii=<0 0 0 0>
     }
     tintColor {colorUseRule=UseColor_Specified, specifiedColor=<rgba 1 1 1 1>}
     uVRegion {bIsValid=false, max=<0 0>, min=<0 0>}
    }
    renderTransformPivot=<0.75 0.5>
   }
   slot {class=OverlaySlot, properties {verticalAlignment=VAlign_Center, padding=<ltrb -12 0 0 0>}}
  }
  Image_Bracket_Right {
   widgetClassPath=&Image
   parent=BackingandDecor
   widget=Image_Bracket_Right
   properties {
    visibility=SelfHitTestInvisible
    brush {
     bIsDynamicallyLoaded=false
     drawAs=Image
     imageType=FullColor
     mirroring=NoMirror
     resourceName=None
     resourceObject=/BattleRoyaleGameUI/HUD/PlayerActionAlert/Assets/MI_UI_Knocked_Bracket_Flipped
     tiling=NoTile
     imageSize=<56 48>
     margin=<ltrb 0 0 0 0>
     outlineSettings {
      bUseBrushTransparency=false
      roundingType=HalfHeightRadius
      width=0
      color {colorUseRule=UseColor_Specified, specifiedColor=<rgba 0 0 0 0>}
      cornerRadii=<0 0 0 0>
     }
     tintColor {colorUseRule=UseColor_Specified, specifiedColor=<rgba 1 1 1 1>}
     uVRegion {bIsValid=false, max=<0 0>, min=<0 0>}
    }
    renderTransformPivot=<0.125 0.5>
   }
   slot {class=OverlaySlot, properties {horizontalAlignment=HAlign_Right, verticalAlignment=VAlign_Center, padding=<ltrb 0 0 -12 0>}}
  }
  LazyImage_PlayerPortrait {
   widgetClassPath=&FortLazyImage
   bIsVariable=true
   parent=BackingandDecor
   widget=LazyImage_PlayerPortrait
   properties {
    materialTextureParamName=Portrait
    brush {
     bIsDynamicallyLoaded=false
     drawAs=Image
     imageType=FullColor
     mirroring=NoMirror
     resourceName=None
     resourceObject=/BattleRoyaleGameUI/HUD/PlayerActionAlert/Assets/M_UI_ElimPortrait
     tiling=NoTile
     imageSize=<44 44>
     margin=<ltrb 0 0 0 0>
     outlineSettings {
      bUseBrushTransparency=false
      roundingType=HalfHeightRadius
      width=0
      color {colorUseRule=UseColor_Specified, specifiedColor=<rgba 0 0 0 0>}
      cornerRadii=<0 0 0 0>
     }
     tintColor {colorUseRule=UseColor_Specified, specifiedColor=<rgba 1 1 1 1>}
     uVRegion {bIsValid=false, max=<0 0>, min=<0 0>}
    }
    loadingBackgroundBrush {
     bIsDynamicallyLoaded=false
     drawAs=NoDrawType
     imageType=NoImage
     mirroring=NoMirror
     resourceName=None
     resourceObject=null
     tiling=NoTile
     imageSize=<64 64>
     margin=<ltrb 0 0 0 0>
     outlineSettings {
      bUseBrushTransparency=false
      roundingType=HalfHeightRadius
      width=0
      color {colorUseRule=UseColor_Specified, specifiedColor=<rgba 0 0 0 0>}
      cornerRadii=<0 0 0 0>
     }
     tintColor {colorUseRule=UseColor_Specified, specifiedColor=<rgba 1 1 1 1>}
     uVRegion {bIsValid=false, max=<0 0>, min=<0 0>}
    }
    renderTransformPivot=<0 0.5>
   }
   slot {class=OverlaySlot, properties {verticalAlignment=VAlign_Bottom, padding=<ltrb -2 0 0 10>}}
  }
  HorizontalBox_DownFeedback {
   widgetClassPath=&HorizontalBox
   parent=AlertContent
   widget=HorizontalBox_DownFeedback
   slot {class=OverlaySlot, properties {verticalAlignment=VAlign_Center, padding=<ltrb 28 0 28 0>}}
  }
  HorizontalBox_DownPlayerInfo {
   widgetClassPath=&HorizontalBox
   parent=HorizontalBox_DownFeedback
   widget=HorizontalBox_DownPlayerInfo
   slot {class=HorizontalBoxSlot, properties {padding=<ltrb 4 0 0 0>}}
  }
  LazyImage_Habanero_RankIcon {
   widgetClassPath=&FortLazyImage
   bIsVariable=true
   parent=HorizontalBox_DownPlayerInfo
   widget=LazyImage_Habanero_RankIcon
   properties {renderTransform {angle=0, scale=<1.2 1.2>, shear=<0 0>, translation=<0 0>}}
   slot {class=HorizontalBoxSlot, properties {horizontalAlignment=HAlign_Center, verticalAlignment=VAlign_Center, padding=<ltrb 2 0 2 0>}}
  }
  PlayerDownText {
   widgetClassPath=&CommonTextBlock
   bIsVariable=true
   parent=HorizontalBox_DownPlayerInfo
   widget=PlayerDownText
   properties {
    justification=Center
    text="{UserName}"
    font {
     bForceMonospaced=false
     bMaterialIsStencil=false
     fontMaterial=/BattleRoyaleGameUI/Shared/MI_UI_PostProcess_Font_Dark
     fontObject=&Burbank
     letterSpacing=0
     monospacedWidth=1
     size=13.5
     skewAmount=0
     typefaceFontName=BurbankRegular-Black
     outlineSettings {
      bApplyOutlineToDropShadows=false
      bMiteredCorners=true
      bSeparateFillAlpha=false
      outlineMaterial=&MI_UI_PostProcess_Font_Dark
      outlineSize=0
      outlineColor=<rgba 1 1 1 1>
     }
    }
    shadowOffset=<0 0>
    strikeBrush {
     bIsDynamicallyLoaded=false
     drawAs=Image
     imageType=NoImage
     mirroring=NoMirror
     resourceName=None
     resourceObject=null
     tiling=NoTile
     imageSize=<32 32>
     margin=<ltrb 0 0 0 0>
     outlineSettings {
      bUseBrushTransparency=false
      roundingType=HalfHeightRadius
      width=0
      color {colorUseRule=UseColor_Specified, specifiedColor=<rgba 0 0 0 0>}
      cornerRadii=<0 0 0 0>
     }
     tintColor {colorUseRule=UseColor_Specified, specifiedColor=<rgba 0.8 0.8 0.8 1>}
     uVRegion {bIsValid=false, max=<0 0>, min=<0 0>}
    }
   }
   slot {class=HorizontalBoxSlot, properties {verticalAlignment=VAlign_Center}}
  }
 ]
}
