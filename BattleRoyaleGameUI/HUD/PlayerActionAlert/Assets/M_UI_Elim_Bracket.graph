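# Translucent UI material that draws an angled bracket for the elimination alert in the Battle Royale HUD player action display.
# MF_UI_BoxAngle masks combine through Min, minus a sine-based center cut, forming the procedural shape; FlipX mirrors it and Anim_Stretch animates its top and bottom angles.
# Color animates from adaptive palette tones to red, knocked orange/yellow gradients from MPC_BRGameUI, or a custom color.
# Static switches select animated color and opacity.
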Material /BattleRoyaleGameUI/HUD/PlayerActionAlert/Assets/M_UI_Elim_Bracket
graph {
 properties {bAutomaticallySetUsageInEditor=false, blendMode=BLEND_Translucent, materialDomain=MD_UI, usedShadingModels=Default Lit}
 nodes [
  n1 = NamedRerouteUsage()
  n2 = NamedRerouteUsage()
  LinearInterpolate(A=@n1, B=Multiply(A=@n1, B=OneMinus(None=@n2)), Alpha=ScalarParameter() {parameterName=Fade})
  Multiply(A=NamedRerouteUsage(), B=OneMinus(None=ScalarParameter() {parameterName=AdditiveBlendDimAmount}))
  n3 = MaterialFunctionCall() {materialFunction=&GetUserInterfaceUV}
  n4 = MaterialFunctionCall(Float2=@n3.Normalized UV) {materialFunction=&BreakOutFloat2Components}
  n5 = LinearInterpolate(A=@n3.Normalized UV, B=AppendVector(A=OneMinus(None=@n4.R), B=@n4.G), Alpha=ScalarParameter() {parameterName=FlipX})
  n6 = Constant()
  n7 = ScalarParameter() {defaultValue=11, parameterName=TopAngle}
  n8 = LinearInterpolate(
   A=@n7
   B=Subtract(A=@n7, B=ScalarParameter() {defaultValue=3, parameterName=StretchAmount_Top}) {constB=3}
   Alpha=ScalarParameter() {parameterName=Anim_Stretch}
  )
  n9 = ScalarParameter() {defaultValue=1, parameterName=Falloff}
  n10 = MaterialFunctionCall(Float2=@n5) {materialFunction=&BreakOutFloat2Components}
  n11 = AppendVector(A=@n10.R, B=OneMinus(None=@n10.G))
  n12 = Constant()
  n13 = ScalarParameter() {defaultValue=13, parameterName=BottomAngle}
  n14 = LinearInterpolate(
   A=@n13
   B=Subtract(A=@n13, B=ScalarParameter() {defaultValue=5, parameterName=StretchAmount_Bottom}) {constB=3}
   Alpha=ScalarParameter() {parameterName=Anim_Stretch}
  )
  n15 = Min(
   A=Saturate(None=MaterialFunctionCall!(UV=@n5, Corner=@n6, Angle=@n8, Flip, Falloff=@n9) {materialFunction=/FortUILibrary/MaterialFunctions/Shapes/MF_UI_BoxAngle}.Result)
   B=Saturate(None=MaterialFunctionCall(UV=@n11, Corner=@n12, Angle=@n14, Falloff=@n9) {materialFunction=&MF_UI_BoxAngle}.Result)
  )
  n16 = Saturate(None=MaterialFunctionCall(UV=@n11, Corner=Constant() {r=3}, Angle=ScalarParameter() {defaultValue=16, parameterName=OustideAngle}, Falloff=@n9) {materialFunction=&MF_UI_BoxAngle}.Result)
  n17 = Add(A=@n11, B=ScalarParameter() {defaultValue=-0.15, parameterName=InnerAngleOffset})
  n18 = Constant() {r=3}
  n19 = ScalarParameter() {defaultValue=12, parameterName=InnerAngle}
  n20 = ComponentMask(None=Sine(None=@n5) {period=2}) {g=true}
  n21 = ScalarParameter() {defaultValue=0.5, parameterName=CenterCutHeight}
  n22 = ScalarParameter() {parameterName=CenterCutSmoothing}
  n23 = Min(
   A=Saturate(None=MaterialFunctionCall(UV=@n17, Corner=@n18, Angle=@n19, Falloff=@n9) {materialFunction=&MF_UI_BoxAngle}.Result)
   B=Saturate(None=MaterialFunctionCall(DistanceField=@n20, Threshold=@n21, Smoothing=@n22) {materialFunction=&MF_UI_AntiAliasing}.Outer)
  )
  NamedRerouteDeclaration(None=Min(A=Min(A=@n15, B=@n16), B=OneMinus(None=@n23))) {name=Shape, nodeColor=<rgba 0 0 0 1>}
  n24 = MaterialFunctionCall!(BufferSource) {materialFunction=/BattleRoyaleGameUI/Shared/MF_UI_AdaptiveColors}
  n25 = ScalarParameter() {defaultValue=3, group=Color, parameterName=AdaptiveColorIndex}
  n26 = LinearInterpolate(A=@n24.Dark, B=@n24.Dark, Alpha=@n25)
  n27 = Saturate(None=Subtract(A=@n25))
  n28 = Subtract(A=@n25) {constB=2}
  NamedRerouteDeclaration(None=LinearInterpolate(A=LinearInterpolate(A=@n26, B=@n24.Medium, Alpha=@n27), B=@n24.White, Alpha=Saturate(None=@n28))) {name=AdaptiveColorOutput, nodeColor=<rgba 0 0 0 1>}
  NamedRerouteDeclaration(None=ComponentMask(None=@n3.Normalized UV) {g=true}) {name=VerticalMask, nodeColor=<rgba 0 0 0 1>}
  n29 = LinearInterpolate(
   A=ScalarParameter() {group=Opacity, parameterName=Opacity_Start}
   B=ScalarParameter() {defaultValue=1, group=Opacity, parameterName=Opacity_Stop}
   Alpha=ScalarParameter() {defaultValue=1, group=Opacity, parameterName=Anim_Opacity}
  )
  n30 = ScalarParameter() {defaultValue=1, group=Opacity, parameterName=Opacity}
  NamedRerouteDeclaration(None=Multiply(A=StaticSwitchParameter(True=@n29, False=@n30) {defaultValue=true, parameterName=AnimatedOpacity}, B=NamedRerouteUsage())) {name=Opacity, nodeColor=<rgba 0 0 0 1>}
  n31 = NamedRerouteUsage()
  n32 = Constant4Vector() {constant=<rgba 0.667 0 0.0889 1>}
  n33 = NamedRerouteUsage()
  n34 = LinearInterpolate(
   A=@n32.RGB
   B=ComponentMask(None=CollectionParameter() {collection=/BattleRoyaleGameUI/Shared/MPC_BRGameUI, parameterName=Red}) {b=true, g=true, r=true}
   Alpha=@n33
  )
  n35 = LinearInterpolate(A=@n32.RGB, B=@n34, Alpha=NamedRerouteUsage())
  n36 = ScalarParameter() {parameterName=Fade}
  n37 = ComponentMask(None=CollectionParameter() {collection=&MPC_BRGameUI, parameterName=Orange}) {b=true, g=true, r=true}
  n38 = LinearInterpolate(
   A=@n37
   B=ComponentMask(None=CollectionParameter() {collection=&MPC_BRGameUI, parameterName=Yellow}) {b=true, g=true, r=true}
   Alpha=LinearInterpolate(Alpha=@n33) {constA=0.2, constB=0.5}
  )
  n39 = LinearInterpolate(A=@n38, B=@n37, Alpha=NamedRerouteUsage())
  n40 = ScalarParameter() {parameterName=Fade}
  n41 = StaticSwitchParameter(
   True=VectorParameter() {group=Color, parameterName=Color, defaultValue=<rgba 1 0 0 1>}.RGB
   False=LinearInterpolate(
    A=LinearInterpolate(A=@n34, B=@n35, Alpha=@n36)
    B=LinearInterpolate(A=@n38, B=@n39, Alpha=@n40)
    Alpha=ScalarParameter() {parameterName=IsKnockedColor}
   )
  ) {parameterName=CustomColor}
  n42 = ScalarParameter() {defaultValue=1, group=Color, parameterName=Anim_Color}
  NamedRerouteDeclaration(None=StaticSwitchParameter(
    True=LinearInterpolate(A=@n31, B=@n41, Alpha=@n42)
    False=StaticSwitchParameter(True=@n31, False=@n41) {defaultValue=true, parameterName=AdaptiveColor}
   ) {defaultValue=true, parameterName=AnimatedColors}) {name=Color, nodeColor=<rgba 0 0 0 1>}
  NamedRerouteDeclaration(None=ComponentMask(None=@n3.Normalized UV) {r=true}) {name=HorizontalMask, nodeColor=<rgba 0 0 0 1>}
 ]
}
