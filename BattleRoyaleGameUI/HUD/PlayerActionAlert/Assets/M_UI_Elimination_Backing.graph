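# Translucent UI material that draws the angled backing for the elimination alert in the HUD.
# An optionally mirrored, 9-sliced UV feeds a BoxAngle shape, which masks the opacity.
# Color blends an indexed adaptive palette toward a custom or collection-driven red-to-orange/yellow knocked color, with static switches selecting animated or fixed modes.
# Opacity either animates between start and stop values or uses an optional rotated, eased gradient, with low-end device adjustments.
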Material /BattleRoyaleGameUI/HUD/PlayerActionAlert/Assets/M_UI_Elimination_Backing
graph {
 properties {bAutomaticallySetUsageInEditor=false, blendMode=BLEND_Translucent, materialDomain=MD_UI, usedShadingModels=Default Lit}
 nodes [
  n1 = MaterialFunctionCall!(BufferSource) {materialFunction=/BattleRoyaleGameUI/Shared/MF_UI_AdaptiveColors}
  n2 = ScalarParameter() {group=Color, parameterName=AdaptiveColorIndex}
  n3 = Subtract(A=@n2)
  n4 = LinearInterpolate(A=LinearInterpolate(A=@n1.Dark, B=@n1.Dark, Alpha=Saturate(None=@n3)), B=@n1.Bright, Alpha=Saturate(None=Subtract(A=@n2) {constB=2}))
  n5 = Saturate(None=Subtract(A=@n2) {constB=3})
  NamedRerouteDeclaration(None=LinearInterpolate(A=LinearInterpolate(A=@n4, B=@n1.White, Alpha=@n5), B=@n1.White, Alpha=ScalarParameter() {parameterName=WhiteMix})) {name=AdaptiveColorOutput, nodeColor=<rgba 0 0 0 1>}
  n6 = LinearInterpolate(
   A=MaterialFunctionCall() {materialFunction=&GetUserInterfaceUV}.Normalized UV
   B=MaterialFunctionCall!(UV, Mirror Y, Swizzle UV) {materialFunction=/FortUILibrary/MaterialFunctions/Basics/MF_UI_Symmetry_2x1_New}.Flipped Symmetry
   Alpha=ScalarParameter() {defaultValue=1, parameterName=UseSymmetry}
  )
  NamedRerouteDeclaration(None=MaterialFunctionCall!(
    UV=MaterialFunctionCall!(UVs=@n6) {materialFunction=/FortUILibrary/MaterialFunctions/Basics/MF_UI_9SliceUVs}.UVs
    Corner=Constant() {r=1}
    Angle=ScalarParameter() {defaultValue=24, parameterName=Angle}
    Flip
    Falloff
   ) {materialFunction=/FortUILibrary/MaterialFunctions/Shapes/MF_UI_BoxAngle}.Result) {name=Shape, nodeColor=<rgba 0 0 0 1>}
  n7 = ScalarParameter() {defaultValue=1, group=Opacity, parameterName=Opacity_Stop}
  n8 = LinearInterpolate(
   A=ScalarParameter() {group=Opacity, parameterName=Opacity_Start}
   B=MaterialFunctionCall!(Default=@n7, LowEnd=Multiply(A=@n7) {constB=0.9}) {materialFunction=/BattleRoyaleGameUI/Shared/MF_UI_LowEnd_ValueSwitcher}.Result
   Alpha=ScalarParameter() {defaultValue=1, group=Opacity, parameterName=Anim_Opacity}
  )
  n9 = ScalarParameter() {defaultValue=1, group=Opacity, parameterName=Opacity}
  n10 = MaterialFunctionCall(Default=@n9, LowEnd=Multiply(A=@n9) {constB=0.9}) {materialFunction=&MF_UI_LowEnd_ValueSwitcher}
  n11 = ComponentMask(None=MaterialFunctionCall(
    UVs=MaterialFunctionCall() {materialFunction=&GetUserInterfaceUV}.Normalized UV
    "Rotation Angle (0-1)"=ScalarParameter() {parameterName=FadeRotation}
   ) {materialFunction=&CustomRotator}.Rotated Values) {r=true}
  n12 = LinearInterpolate(
   A=@n10.Result
   B=Multiply(
    A=@n10.Result
    B=MaterialFunctionCall("Value (0-1)"=@n11) {materialFunction=&MF_UI_EaseCurves}.EaseInOutSine
   )
   Alpha=ScalarParameter() {parameterName=UseGradientFade}
  ) {constA=1}
  NamedRerouteDeclaration(None=Multiply(A=StaticSwitchParameter(True=@n8, False=@n12) {defaultValue=true, parameterName=AnimatedOpacity}, B=NamedRerouteUsage())) {name=Opacity, nodeColor=<rgba 0 0 0 1>}
  n13 = NamedRerouteUsage()
  n14 = CollectionParameter() {collection=/BattleRoyaleGameUI/Shared/MPC_BRGameUI, parameterName=Red}
  n15 = ComponentMask(None=CollectionParameter() {collection=&MPC_BRGameUI, parameterName=Orange}) {b=true, g=true, r=true}
  n16 = ComponentMask(None=CollectionParameter() {collection=&MPC_BRGameUI, parameterName=Yellow}) {b=true, g=true, r=true}
  n17 = StaticSwitchParameter(
   True=VectorParameter() {group=Color, parameterName=Color, defaultValue=<rgba 1 0 0 1>}.RGB
   False=LinearInterpolate(
    A=ComponentMask(None=@n14) {b=true, g=true, r=true}
    B=LinearInterpolate(A=@n15, B=@n16)
    Alpha=ScalarParameter() {parameterName=IsKnockedAnimatedColor}
   )
  ) {parameterName=CustomColor}
  n18 = ScalarParameter() {group=Color, parameterName=Anim_Color}
  NamedRerouteDeclaration(None=StaticSwitchParameter(
    True=LinearInterpolate(A=@n13, B=@n17, Alpha=@n18)
    False=StaticSwitchParameter(True=@n13, False=@n17) {defaultValue=true, parameterName=AdaptiveColor}
   ) {defaultValue=true, parameterName=AnimatedColors}) {name=Color, nodeColor=<rgba 0 0 0 1>}
  NamedRerouteUsage()
  NamedRerouteUsage()
 ]
}
