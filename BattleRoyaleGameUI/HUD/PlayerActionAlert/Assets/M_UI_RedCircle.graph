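# M_UI_RedCircle is an additive UI material for a red circle, stored with the player action alert HUD assets.
# It reads the shared Red color from the MPC_BRGameUI parameter collection and masks it to RGB.
# The MF_UI_Circle material function generates the circle shape procedurally from a constant size.
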Material /BattleRoyaleGameUI/HUD/PlayerActionAlert/Assets/M_UI_RedCircle
graph {
 properties {bAutomaticallySetUsageInEditor=false, blendMode=BLEND_Additive, materialDomain=MD_UI, usedShadingModels=Default Lit}
 nodes [
  ComponentMask(None=CollectionParameter() {collection=/BattleRoyaleGameUI/Shared/MPC_BRGameUI, parameterName=Red}) {b=true, g=true, r=true}
  MaterialFunctionCall(Size=Constant() {r=0.9}) {materialFunction=&MF_UI_Circle}
 ]
}
