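# Translucent UI material for the player action alert HUD, drawing a red digital tile effect coloured from the shared BR UI parameter collection.
# Two particle-grid layers draw rounded boxes with time-randomised sizes, random cell removal and a parameter-driven collapse.
# A StaticParticles switch selects between static tiles and tiles panning at per-row random speeds with gradient shrink, vertical edge masking and optional fade.
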Material /BattleRoyaleGameUI/HUD/PlayerActionAlert/Assets/M_UI_RedDigiTilesFX
graph {
 properties {
  bAutomaticallySetUsageInEditor=false
  blendMode=BLEND_Translucent
  materialDomain=MD_UI
  pixelDepthOffsetMode=PDOM_Legacy
  usedShadingModels=Default Lit
 }
 nodes [
  NamedRerouteDeclaration(None=MaterialFunctionCall() {materialFunction=&MF_UI_UVScaling}.ScaleX) {name=ScaledUVs, nodeColor=<rgba 0 0 0 1>}
  n1 = Constant()
  n2 = MaterialFunctionCall!(
   UV=StaticSwitchParameter(True=NamedRerouteUsage(), False=NamedRerouteUsage()) {defaultValue=true, parameterName=StaticParticles}
   Global Offset
   Global Rotation
   Tiling=ScalarParameter() {defaultValue=11, parameterName=Layer1_TileCount}
   Cell Offset=@n1
   Cell Scale=@n1
   Cell Rotation=@n1
   Random Seed
  ) {materialFunction=/FortUILibrary/MaterialFunctions/Basics/MF_UI_ParticleGrid}
  n3 = Multiply(A=Time(), B=ScalarParameter() {defaultValue=1.5, parameterName=Layer1_TileScaleSpeed})
  n4 = LinearInterpolate(
   A=ScalarParameter() {defaultValue=0.1, parameterName=Layer1_TileScaleMinSize}
   Alpha=Saturate(None=MaterialFunctionCall!(UV=@n2.Random Cell Values, Seed=@n3) {materialFunction=&MF_UI_Random}.Result)
  )
  n5 = MaterialFunctionCall("Value (0-1)"=MaterialFunctionCall() {materialFunction=&LinearGradient}.UGradient) {materialFunction=&MF_UI_EaseCurves}
  n6 = OneMinus(None=@n5.EaseInOutCubic)
  n7 = Multiply(A=@n4, B=@n6)
  NamedRerouteDeclaration(None=Saturate(None=StaticSwitchParameter(True=@n4, False=@n7) {defaultValue=true, parameterName=StaticParticles})) {name=Layer1_RandomTileScaleInfo, nodeColor=<rgba 0 0 0 1>}
  NamedRerouteDeclaration(None=@n6) {name=UGradient_Mask, nodeColor=<rgba 0 0 0 1>}
  n8 = Constant()
  n9 = MaterialFunctionCall(
   UV=NamedRerouteUsage()
   Tiling=ScalarParameter() {defaultValue=9, parameterName=Layer2_TileCount}
   Cell Offset=@n8
   Cell Scale=@n8
   Cell Rotation=@n8
   Random Seed=Constant() {r=1409.8898}
  ) {materialFunction=&MF_UI_ParticleGrid}
  NamedRerouteDeclaration(None=Saturate(None=LinearInterpolate(Alpha=@n9.Random Rows) {constA=0.4})) {name=Layer2_RowSpeedAdjustment, nodeColor=<rgba 0 0 0 1>}
  NamedRerouteDeclaration(None=Saturate(None=@n5.EaseInSine)) {name=FadeMask, nodeColor=<rgba 0 0 0 1>}
  n10 = Multiply(A=ScalarParameter() {defaultValue=-0.25, parameterName=Layer1_TilePanSpeed}, B=NamedRerouteUsage())
  n11 = Constant()
  NamedRerouteDeclaration(None=Panner(Coordinate=NamedRerouteUsage(), Speed=AppendVector(A=@n10, B=@n11)) {speedX=-0.25}) {name=Layer1_PanningUVs, nodeColor=<rgba 0 0 0 1>}
  n12 = Constant()
  n13 = MaterialFunctionCall(
   UV=StaticSwitchParameter(True=NamedRerouteUsage(), False=NamedRerouteUsage()) {defaultValue=true, parameterName=StaticParticles}
   Tiling=ScalarParameter() {defaultValue=11, parameterName=Layer1_TileCount}
   Cell Offset=@n12
   Cell Scale=NamedRerouteUsage()
   Cell Rotation=@n12
  ) {materialFunction=&MF_UI_ParticleGrid}
  n14 = LinearInterpolate(
   A=ScalarParameter() {defaultValue=1, parameterName=Layer1_OverallTileScale}
   B=ScalarParameter() {defaultValue=0.1, parameterName=CollapsedParticleSize}
   Alpha=ScalarParameter() {parameterName=Anim_CollapseTiles}
  ) {constB=0.1}
  n15 = Constant() {r=0.1}
  n16 = StaticBool() {value=true}
  n17 = Saturate(None=MaterialFunctionCall(
    DistanceField=MaterialFunctionCall(UVs=@n13.Cells UV, Size=@n14, CornerRoundness=@n15, KeepAspectRatio=@n16) {materialFunction=&MF_UI_Box}.SDF
    Threshold=Constant()
    Smoothing=Constant() {r=0.05}
   ) {materialFunction=&MF_UI_AntiAliasing}.Inner)
  n18 = NamedRerouteUsage()
  NamedRerouteDeclaration(None=Saturate(None=Min(A=@n17, B=@n18))) {name=Layer1_Tiles, nodeColor=<rgba 0 0 0 1>}
  n19 = StaticSwitchParameter(True=NamedRerouteUsage(), False=NamedRerouteUsage()) {defaultValue=true, parameterName=StaticParticles}
  n20 = ScalarParameter() {defaultValue=11, parameterName=Layer1_TileCount}
  n21 = Constant()
  n22 = Constant() {r=54235}
  n23 = Subtract(
   A=Saturate(None=MaterialFunctionCall(UV=@n19, Tiling=@n20, Cell Offset=@n21, Cell Scale=@n21, Cell Rotation=@n21, Random Seed=@n22) {materialFunction=&MF_UI_ParticleGrid}.Random Cell Values)
   B=ScalarParameter() {defaultValue=0.9, parameterName=Layer1_SubtractTileAmount}
  )
  NamedRerouteDeclaration(None=Saturate(None=Ceil(None=@n23))) {name=Layer1_SubtractionMask, nodeColor=<rgba 0 0 0 1>}
  n24 = Multiply(A=ScalarParameter() {defaultValue=-0.44, parameterName=Layer2_TilePanSpeed}, B=NamedRerouteUsage())
  n25 = Constant()
  NamedRerouteDeclaration(None=Panner(Coordinate=NamedRerouteUsage(), Speed=AppendVector(A=@n24, B=@n25)) {speedX=-0.25}) {name=Layer2_PanningUVs, nodeColor=<rgba 0 0 0 1>}
  n26 = Constant()
  n27 = MaterialFunctionCall(
   UV=StaticSwitchParameter(True=NamedRerouteUsage(), False=NamedRerouteUsage()) {defaultValue=true, parameterName=StaticParticles}
   Tiling=ScalarParameter() {defaultValue=9, parameterName=Layer2_TileCount}
   Cell Offset=@n26
   Cell Scale=@n26
   Cell Rotation=@n26
  ) {materialFunction=&MF_UI_ParticleGrid}
  n28 = Multiply(A=Time(), B=ScalarParameter() {defaultValue=1.25, parameterName=Layer2_TileScaleSpeed})
  n29 = LinearInterpolate(
   A=ScalarParameter() {defaultValue=0.1, parameterName=Layer2_TileScaleMinSize}
   Alpha=Saturate(None=MaterialFunctionCall(UV=@n27.Random Cell Values, Seed=@n28) {materialFunction=&MF_UI_Random}.Result)
  )
  n30 = MaterialFunctionCall() {materialFunction=&LinearGradient}
  n31 = Multiply(
   A=@n29
   B=MaterialFunctionCall("Value (0-1)"=OneMinus(None=@n30.UGradient)) {materialFunction=&MF_UI_EaseCurves}.EaseInSine
  )
  NamedRerouteDeclaration(None=Saturate(None=StaticSwitchParameter(True=@n29, False=@n31) {defaultValue=true, parameterName=StaticParticles})) {name=Layer2_RandomTileScaleInfo, nodeColor=<rgba 0 0 0 1>}
  n32 = Constant()
  n33 = MaterialFunctionCall(
   UV=StaticSwitchParameter(True=NamedRerouteUsage(), False=NamedRerouteUsage()) {defaultValue=true, parameterName=StaticParticles}
   Tiling=ScalarParameter() {defaultValue=9, parameterName=Layer2_TileCount}
   Cell Offset=@n32
   Cell Scale=NamedRerouteUsage()
   Cell Rotation=@n32
  ) {materialFunction=&MF_UI_ParticleGrid}
  n34 = LinearInterpolate(
   A=ScalarParameter() {defaultValue=1, parameterName=Layer2_OverallTileScale}
   B=ScalarParameter() {defaultValue=0.1, parameterName=CollapsedParticleSize}
   Alpha=ScalarParameter() {parameterName=Anim_CollapseTiles}
  ) {constB=0.1}
  n35 = Constant() {r=0.1}
  n36 = Constant() {r=0.125}
  n37 = StaticBool() {value=true}
  n38 = Saturate(None=MaterialFunctionCall(
    DistanceField=MaterialFunctionCall(UVs=@n33.Cells UV, Size=@n34, CornerRoundness=@n35, OutlineThickness=@n36, KeepAspectRatio=@n37) {materialFunction=&MF_UI_Box}.SDF
    Threshold=Constant()
    Smoothing=Constant() {r=0.05}
   ) {materialFunction=&MF_UI_AntiAliasing}.Inner)
  n39 = NamedRerouteUsage()
  NamedRerouteDeclaration(None=Saturate(None=Min(A=@n38, B=@n39))) {name=Layer2_Tiles, nodeColor=<rgba 0 0 0 1>}
  n40 = StaticSwitchParameter(True=NamedRerouteUsage(), False=NamedRerouteUsage()) {defaultValue=true, parameterName=StaticParticles}
  n41 = ScalarParameter() {defaultValue=9, parameterName=Layer2_TileCount}
  n42 = Constant()
  n43 = Constant() {r=1491.3677}
  n44 = Subtract(
   A=Saturate(None=MaterialFunctionCall(UV=@n40, Tiling=@n41, Cell Offset=@n42, Cell Scale=@n42, Cell Rotation=@n42, Random Seed=@n43) {materialFunction=&MF_UI_ParticleGrid}.Random Cell Values)
   B=ScalarParameter() {defaultValue=0.9, parameterName=Layer2_SubtractTileAmount}
  )
  NamedRerouteDeclaration(None=Saturate(None=Ceil(None=@n44))) {name=Layer2_SubtractionMask, nodeColor=<rgba 0 0 0 1>}
  n45 = Constant()
  n46 = MaterialFunctionCall(
   UV=NamedRerouteUsage()
   Tiling=ScalarParameter() {defaultValue=11, parameterName=Layer1_TileCount}
   Cell Offset=@n45
   Cell Scale=@n45
   Cell Rotation=@n45
   Random Seed=Constant() {r=106.5345}
  ) {materialFunction=&MF_UI_ParticleGrid}
  NamedRerouteDeclaration(None=Saturate(None=LinearInterpolate(Alpha=@n46.Random Rows) {constA=0.4})) {name=Layer1_RowSpeedAdjustment, nodeColor=<rgba 0 0 0 1>}
  n47 = NamedRerouteUsage()
  n48 = NamedRerouteUsage()
  n49 = Multiply(A=Saturate(None=Add(A=@n47, B=@n48)), B=NamedRerouteUsage())
  n50 = NamedRerouteUsage()
  n51 = NamedRerouteUsage()
  n52 = NamedRerouteUsage()
  n53 = Saturate(None=Add(A=Min(A=@n50, B=@n51), B=Min(A=@n50, B=@n52)))
  n54 = NamedRerouteUsage()
  n55 = ComponentMask(None=Sine(None=TextureCoordinate()) {period=2}) {g=true}
  n56 = Min(
   A=StaticSwitchParameter(True=Multiply(A=@n53, B=@n54), False=@n53) {parameterName=Fade}
   B=Saturate(None=MaterialFunctionCall("Value (0-1)"=@n55) {materialFunction=&MF_UI_EaseCurves}.EaseInQuint)
  )
  Saturate(None=Multiply(
    A=StaticSwitchParameter(True=@n49, False=@n56) {defaultValue=true, parameterName=StaticParticles}
    B=ScalarParameter() {defaultValue=1, parameterName=IntensityMultiplier}
   ))
  n57 = ComponentMask(None=CollectionParameter() {collection=/BattleRoyaleGameUI/Shared/MPC_BRGameUI, parameterName=Red}) {b=true, g=true, r=true}
  n58 = MaterialFunctionCall(Size=Constant() {r=1}, CornerRoundness=Constant() {r=0.25}, KeepAspectRatio=StaticBool() {value=true}) {materialFunction=&MF_UI_Box}
  n59 = NamedRerouteDeclaration(None=SmoothStep(Value=Add(A=@n58.SDF) {constB=0.5}) {constMax=0, constMin=0.5}) {name=SoftBoxMask, nodeColor=<rgba 0 0 0 1>}
  n60 = Min(A=NamedRerouteUsage(), B=NamedRerouteUsage())
  n61 = NamedRerouteUsage()
  n62 = NamedRerouteUsage()
  StaticSwitchParameter(
   True=Multiply(A=@n57, B=StaticSwitchParameter(True=@n59, False=@n60) {defaultValue=true, parameterName=StaticParticles})
   False=Multiply(A=Min(A=@n61, B=@n62), B=@n57)
  ) {defaultValue=true, parameterName=StaticParticles}
 ]
}
