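# Additive UI material that draws a red radial glow, stored with the player action alert HUD assets.
# The red color comes from the Red parameter in the shared BR Game UI material parameter collection.
# A zero-size MF_UI_Circle SDF is inverted and indexes the GlowFalloffs curve in the shared UI curve atlas, so the glow falloff is curve-driven rather than computed in the graph.
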Material /BattleRoyaleGameUI/HUD/PlayerActionAlert/Assets/M_UI_RedRadialGlow
graph {
 properties {
  bAutomaticallySetUsageInEditor=false
  blendMode=BLEND_Additive
  materialDomain=MD_UI
  pixelDepthOffsetMode=PDOM_Legacy
  usedShadingModels=Default Lit
 }
 nodes [
  ComponentMask(None=CollectionParameter() {collection=/BattleRoyaleGameUI/Shared/MPC_BRGameUI, parameterName=Red}) {b=true, g=true, r=true}
  n1 = Constant()
  CurveAtlasRowParameter(CurveTime=OneMinus(None=MaterialFunctionCall(Size=@n1) {materialFunction=&MF_UI_Circle}.SDF)) {
   atlas=/BattleRoyaleGameUI/Shared/UICurveAtlas_BRGameUI
   curve=/BattleRoyaleGameUI/Shared/UICurve_BRGameUI_GlowFalloffs
   defaultValue=3
   parameterName=GlowCurve
  }
 ]
}
