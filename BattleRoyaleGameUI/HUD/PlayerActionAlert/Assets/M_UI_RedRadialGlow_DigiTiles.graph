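# UI material for a red, tiled digital glow, stored with the HUD's player action alert assets.
# It builds sine grid lines, time-randomized cell tiles, two-layer panning noise and an animated tile swipe, coloring from dark red to the shared collection's Red.
# Eased-sine or rounded-box masks shape the fade, with static switches for overcharge shield variants.
# Alpha-composite blending with an AdditiveMix control varies between solid and additive glow.
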Material /BattleRoyaleGameUI/HUD/PlayerActionAlert/Assets/M_UI_RedRadialGlow_DigiTiles
graph {
 properties {
  bAutomaticallySetUsageInEditor=false
  blendMode=BLEND_AlphaComposite
  materialDomain=MD_UI
  pixelDepthOffsetMode=PDOM_Legacy
  usedShadingModels=Default Lit
 }
 nodes [
  n1 = Multiply(A=ScalarParameter() {defaultValue=37, parameterName=GridSize})
  n2 = Multiply(A=ConstantBiasScale(None=NamedRerouteUsage()) {bias=2}, B=@n1) {constB=10}
  n3 = Divide(B=@n1)
  Add(A=@n2, B=Multiply(A=@n3, B=@n3))
  AppendVector(A=Constant() {r=0.04}, B=Constant() {r=0.04})
  n4 = CollectionParameter() {collection=/BattleRoyaleGameUI/Shared/MPC_BRGameUI, parameterName=Red}
  n5 = LinearInterpolate(A=NamedRerouteUsage(), Alpha=NamedRerouteUsage())
  n6 = Multiply(A=@n5, B=NamedRerouteUsage())
  n7 = MaterialFunctionCall!(
   Base=LinearInterpolate(
    A=Constant3Vector() {constant=<rgba 0.1042 0 0 0>}
    B=ComponentMask(None=@n4) {b=true, g=true, r=true}
    Alpha=LinearInterpolate(A=@n6, B=@n5)
   )
   Add=Multiply(A=NamedRerouteUsage(), B=ScalarParameter() {defaultValue=0.5, parameterName=GridLinesAlpha})
  ) {materialFunction=/Engine/Functions/Engine_MaterialFunctions02/Math/MultiplyAdd}
  n8 = Saturate(None=MaterialFunctionCall!(Blend=NamedRerouteUsage(), Base=Saturate(None=@n7.Result)) {materialFunction=/Engine/Functions/Engine_MaterialFunctions03/Blends/Blend_ColorDodge}.Result)
  n9 = ScalarParameter() {defaultValue=0.9, parameterName=TileSubtraction_Amount}
  n10 = NamedRerouteUsage()
  n11 = StaticSwitchParameter(
   True=NamedRerouteUsage()
   False=Saturate(None=MaterialFunctionCall("Value (0-1)"=@n10) {materialFunction=&MF_UI_EaseCurves}.EaseOutCubic)
  ) {parameterName=RoundedFadeMask}
  n12 = LinearInterpolate(A=@n8, B=Multiply(A=@n8, B=OneMinus(None=@n9)) {constB=0.666}, Alpha=Min(A=NamedRerouteUsage(), B=OneMinus(None=@n11))) {constB=0}
  n13 = Multiply(A=@n12, B=OneMinus(None=ScalarParameter() {parameterName=DimEdgeAmount}))
  n14 = NamedRerouteUsage()
  n15 = Multiply(A=NamedRerouteUsage(), B=NamedRerouteUsage())
  n16 = Sine(None=TextureCoordinate()) {period=2}
  n17 = MaterialFunctionCall("Value (0-1)"=ComponentMask(None=@n16) {g=true}) {materialFunction=&MF_UI_EaseCurves}
  n18 = Multiply(A=NamedRerouteUsage(), B=@n17.EaseInQuint)
  n19 = MaterialFunctionCall("Value (0-1)"=ComponentMask(None=@n16) {r=true}) {materialFunction=&MF_UI_EaseCurves}
  n20 = Saturate(None=MaterialFunctionCall("Value (0-1)"=@n19.EaseInOutSine) {materialFunction=&MF_UI_EaseCurves}.EaseInOutQuint)
  n21 = LinearInterpolate(A=@n20, B=Saturate(None=@n19.EaseInQuint), Alpha=ScalarParameter() {defaultValue=1, parameterName=FadeSelect})
  n22 = LinearInterpolate(A=@n21, B=LinearInterpolate(B=@n21, Alpha=NamedRerouteUsage()), Alpha=ScalarParameter() {parameterName=DimEdgeAmount})
  n23 = Saturate(None=LinearInterpolate(
    A=Multiply(A=@n17.EaseInQuint, B=@n22)
    B=Min(A=@n17.EaseInQuint, B=@n22)
    Alpha=ScalarParameter() {defaultValue=0.2, parameterName=ShapeChangeRatio}
   ) {constAlpha=0.2})
  n24 = NamedRerouteDeclaration(None=StaticSwitchParameter(True=NamedRerouteUsage(), False=StaticSwitchParameter(True=@n18, False=@n23) {defaultValue=true, parameterName=Fade}) {parameterName=RoundedFadeMask}) {name=ShapeMask, nodeColor=<rgba 0 0 0 1>}
  n25 = Saturate(None=LinearInterpolate(B=LinearInterpolate(A=@n13, B=@n12, Alpha=@n14), Alpha=StaticSwitchParameter(True=@n15, False=@n24) {parameterName=IsOverChargeShield}))
  n26 = NamedRerouteUsage()
  Multiply(A=VertexColor().A, B=StaticSwitchParameter(True=Multiply(A=@n25, B=@n26), False=@n25) {parameterName=FadeForOverchargeBacking})
  n27 = Constant()
  n28 = MaterialFunctionCall!(
   UV=NamedRerouteUsage()
   Global Offset
   Global Rotation
   Tiling=ScalarParameter() {defaultValue=37, parameterName=GridSize}
   Cell Offset
   Cell Scale=@n27
   Cell Rotation=@n27
   Random Seed
  ) {materialFunction=/FortUILibrary/MaterialFunctions/Basics/MF_UI_ParticleGrid}
  n29 = Time()
  n30 = ScalarParameter() {defaultValue=0.25, parameterName=TileSpeed}
  n31 = Saturate(None=MaterialFunctionCall!(UV=@n28.Random Cell Values, Seed=Multiply(A=@n29, B=@n30)) {materialFunction=&MF_UI_Random}.Result)
  NamedRerouteDeclaration(None=Multiply(A=SmoothStep(Value=@n31) {constMin=0.6}, B=ScalarParameter() {defaultValue=3, parameterName=AmbientTile_Intensity}) {constB=0.2}) {name=Tiles, nodeColor=<rgba 0 0 0 1>}
  n32 = Sine(None=@n2)
  n33 = MaterialFunctionCall(Float2=@n32) {materialFunction=&BreakOutFloat2Components}
  n34 = MaterialFunctionCall("Value (0-1)"=MaterialFunctionCall("Value (0-1)"=@n33.R) {materialFunction=&MF_UI_EaseCurves}.EaseInQuint) {materialFunction=&MF_UI_EaseCurves}
  n35 = MaterialFunctionCall("Value (0-1)"=MaterialFunctionCall("Value (0-1)"=@n33.G) {materialFunction=&MF_UI_EaseCurves}.EaseInQuint) {materialFunction=&MF_UI_EaseCurves}
  NamedRerouteDeclaration(None=Max(A=Saturate(None=@n34.EaseInQuint), B=Saturate(None=@n35.EaseInQuint))) {name=GridLines, nodeColor=<rgba 0 0 0 1>}
  NamedRerouteDeclaration(None=@n32) {name=ManipulatedScaledUVs, nodeColor=<rgba 0 0 0 1>}
  NamedRerouteDeclaration(None=@n20) {name=HorizontalSine, nodeColor=<rgba 0.5234 1 0 1>}
  NamedRerouteDeclaration(None=Saturate(None=TextureSample() {samplerType=SAMPLERTYPE_LinearColor, texture=&/T_UI_TacOvershield_64x}.R)) {name=ShieldMask, nodeColor=<rgba 0 0 0 1>}
  n36 = Add(A=Add(A=TextureCoordinate()) {constB=-1}, B=ScalarParameter() {defaultValue=1, parameterName=Overcharge_Progress})
  NamedRerouteDeclaration(None=Ceil(None=ComponentMask(None=@n36) {g=true})) {name=Overchrage_ProgressMask, nodeColor=<rgba 0 0 0 1>}
  n37 = TextureCoordinate()
  n38 = Sine(None=@n37) {period=2}
  NamedRerouteDeclaration(None=MaterialFunctionCall("Value (0-1)"=ComponentMask(None=@n38) {r=true}) {materialFunction=&MF_UI_EaseCurves}.EaseInQuint) {name=SineFadeMask, nodeColor=<rgba 0 0 0 1>}
  n39 = Multiply(
   A=NamedRerouteDeclaration(None=MaterialFunctionCall() {materialFunction=&MF_UI_UVScaling}.Scaled To Fill) {name=ScaledUVs, nodeColor=<rgba 0 0 0 1>}
   B=ScalarParameter() {defaultValue=0.5, parameterName=Noise1_Scale}
  )
  n40 = AppendVector(
   A=ScalarParameter() {defaultValue=-0.075, parameterName=Noise1_PanSpeedX}
   B=ScalarParameter() {defaultValue=0.075, parameterName=Noise1_PanSpeedY}
  )
  n41 = Multiply(A=NamedRerouteUsage(), B=ScalarParameter() {defaultValue=0.7, parameterName=Noise2_Scale})
  n42 = AppendVector(
   A=ScalarParameter() {defaultValue=0.05, parameterName=Noise2_PanSpeedX}
   B=ScalarParameter() {defaultValue=-0.05, parameterName=Noise2_PanSpeedY}
  )
  n43 = Multiply(
   A=TextureSample(UVs=Panner(Coordinate=@n39, Speed=@n40)) {samplerType=SAMPLERTYPE_Grayscale, texture=/FortUILibrary/UIKit/Buttons/Textures/LowResBlurredNoise}.R
   B=TextureSample(UVs=Panner(Coordinate=@n41, Speed=@n42)) {samplerType=SAMPLERTYPE_Grayscale, texture=&LowResBlurredNoise}.R
  )
  NamedRerouteDeclaration(None=Saturate(None=Multiply(A=@n43) {constB=2})) {name=Noise, nodeColor=<rgba 0 0 0 1>}
  n44 = MaterialFunctionCall(UV=@n28.Random Cell Values, Seed=Multiply(A=@n29, B=ScalarParameter() {defaultValue=1, parameterName=IntroSwipeTileSpeed})) {materialFunction=&MF_UI_Random}
  NamedRerouteDeclaration(None=LinearInterpolate(Alpha=Saturate(None=@n44.Result)) {constA=0.333}) {name=IntroSwipeTiles, nodeColor=<rgba 0 0 0 1>}
  n45 = ScalarParameter() {parameterName=Anim_Swipe}
  n46 = Sine(None=Add(A=TextureCoordinate(), B=LinearInterpolate(Alpha=@n45) {constA=1, constB=-1})) {period=2}
  n47 = LinearInterpolate(
   A=ComponentMask(None=@n46) {g=true}
   B=ComponentMask(None=@n46) {r=true}
   Alpha=ScalarParameter() {defaultValue=1, parameterName=SwipeDireciton}
  )
  n48 = NamedRerouteDeclaration(None=Saturate(None=MaterialFunctionCall("Value (0-1)"=@n47) {materialFunction=&MF_UI_EaseCurves}.EaseInCubic)) {name=IntroSwipe, nodeColor=<rgba 0 0 0 1>}
  n49 = NamedRerouteUsage()
  NamedRerouteDeclaration(None=Multiply(A=Multiply(A=@n48, B=@n49), B=ScalarParameter() {defaultValue=1, parameterName=SwipeIntensity}) {constB=0}) {name=Anim_IntroSwipe, nodeColor=<rgba 0 0 0 1>}
  n50 = OneMinus(None=ScalarParameter() {parameterName=SoftBorderHardness})
  n51 = Constant() {r=1}
  n52 = ScalarParameter() {defaultValue=1, parameterName=CornerRadius}
  n53 = StaticBool() {value=true}
  n54 = Saturate(None=SmoothStep(
    Max=Multiply(A=@n50) {constB=-1}
    Value=MaterialFunctionCall(Size=@n51, CornerRoundness=@n52, KeepAspectRatio=@n53) {materialFunction=&MF_UI_Box}.SDF
   ) {constMax=-0.25})
  n55 = ComponentMask(None=@n37) {r=true}
  n56 = Multiply(A=Saturate(None=MaterialFunctionCall("Value (0-1)"=@n55) {materialFunction=&MF_UI_EaseCurves}.EaseInCubic))
  n57 = Multiply(
   A=NamedRerouteDeclaration(None=MaterialFunctionCall("Value (0-1)"=@n54) {materialFunction=&MF_UI_EaseCurves}.EaseInQuint) {name=SoftRoundMask, nodeColor=<rgba 0 0 0 1>}
   B=NamedRerouteDeclaration(None=Saturate(None=@n56)) {name=HorizontalFade, nodeColor=<rgba 0 0 0 1>}
  )
  NamedRerouteDeclaration(None=Saturate(None=Multiply(A=@n57) {constB=2})) {name=FadeMask, nodeColor=<rgba 0 0 0 1>}
  NamedRerouteDeclaration(None=SmoothStep(Value=@n31) {constMin=0.9}) {name=SubtractionTiles, nodeColor=<rgba 0 0 0 1>}
  SmoothStep(Value=@n19.EaseInQuint) {constMax=0.25}
  n58 = NamedRerouteUsage()
  n59 = LinearInterpolate(B=ScalarParameter() {defaultValue=0.5, parameterName=AdditiveMix}, Alpha=ScalarParameter() {defaultValue=1, parameterName=Anim_AdditiveMix}) {constA=0.3}
  n60 = StaticSwitchParameter(True=Multiply(A=Multiply(A=@n58, B=@n59), B=NamedRerouteUsage()), False=Saturate(None=LinearInterpolate(B=@n59, Alpha=@n24) {constB=0.25})) {parameterName=IsOverChargeShield}
  n61 = StaticSwitchParameter(True=NamedRerouteUsage(), False=NamedRerouteUsage()) {parameterName=SineFade}
  StaticSwitchParameter(True=Multiply(A=@n60, B=NamedRerouteDeclaration(None=@n61) {name=OverchargeBackingTypeMask, nodeColor=<rgba 0 0 0 1>}), False=@n60) {parameterName=FadeForOverchargeBacking}
 ]
}
