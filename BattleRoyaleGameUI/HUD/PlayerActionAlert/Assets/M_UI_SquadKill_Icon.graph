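# Translucent UI material that draws the squad kill icon for the HUD player action alert: a group or solo character icon (IsSolo) on a circular backing, crossed out by two slashes.
# Anti-aliased distance-field masks form each shape, with enlarged slash outlines subtracted from the icon.
# Animation scalars scale each slash and blend icon and backing colors, using red from the shared BR UI parameter collection.
# The JustSlashes static switch outputs only red slashes with adjustable width.
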Material /BattleRoyaleGameUI/HUD/PlayerActionAlert/Assets/M_UI_SquadKill_Icon
graph {
 properties {
  bAutomaticallySetUsageInEditor=false
  blendMode=BLEND_Translucent
  materialDomain=MD_UI
  pixelDepthOffsetMode=PDOM_Legacy
  usedShadingModels=Default Lit
 }
 nodes [
  n1 = TextureSample(UVs=MaterialFunctionCall(Texture Scale=Constant2Vector() {g=0.8, r=0.8}) {materialFunction=&ScaleUVsByCenter}.UVs) {samplerSource=SSM_Clamp_WorldGroupSettings, texture=/Game/Athena/Items/Accolades/Materials/Textures/Accolades_Icon_group}
  Add(A=@n1.R) {constB=0.5}
  n2 = MaterialFunctionCall(
   UVs=MaterialFunctionCall(UVs=TextureCoordinate(), "Rotation Angle (0-1)"=Constant() {r=-0.03}) {materialFunction=&CustomRotator}.Rotated Values
   Texture Scale=Constant2Vector() {g=1.1, r=1.1}
  ) {materialFunction=&ScaleUVsByCenter}
  n3 = LinearInterpolate(Alpha=ScalarParameter() {defaultValue=1, parameterName=Anim_ScaleSlash}) {constA=16}
  n4 = TextureSample(UVs=Add(A=Multiply(A=@n2.UVs, B=@n3), B=Constant2Vector() {g=-0.02, r=-0.025})) {samplerType=SAMPLERTYPE_LinearGrayscale, texture=&/T_IconStrikethroughDF}
  n5 = Subtract(
   A=NamedRerouteUsage()
   B=MaterialFunctionCall(DistanceField=@n4.R, Threshold=ScalarParameter() {defaultValue=0.375, parameterName=SlashOutlineSize}) {materialFunction=&MF_UI_AntiAliasing}.Outer
  )
  n6 = MaterialFunctionCall(
   UVs=MaterialFunctionCall(UVs=TextureCoordinate(), "Rotation Angle (0-1)"=Constant() {r=0.02}) {materialFunction=&CustomRotator}.Rotated Values
   Texture Scale=Constant2Vector() {g=1.1, r=-1.1}
  ) {materialFunction=&ScaleUVsByCenter}
  n7 = LinearInterpolate(Alpha=ScalarParameter() {defaultValue=1, parameterName=Anim_ScaleSlash_Right}) {constA=16}
  n8 = TextureSample(UVs=Add(A=Multiply(A=@n6.UVs, B=@n7), B=Constant2Vector() {g=-0.0333})) {samplerType=SAMPLERTYPE_LinearGrayscale, texture=&T_IconStrikethroughDF}
  n9 = ScalarParameter() {defaultValue=0.375, parameterName=SlashOutlineSize}
  n10 = Saturate(None=Subtract(
    A=Saturate(None=@n5)
    B=MaterialFunctionCall(DistanceField=@n8.R, Threshold=@n9) {materialFunction=&MF_UI_AntiAliasing}.Outer
   ))
  n11 = MaterialFunctionCall(
   DistanceField=@n4.R
   Threshold=StaticSwitchParameter(True=ScalarParameter() {defaultValue=0.5, parameterName=SlashWidth}, False=Constant() {r=0.5}) {parameterName=JustSlashes}
  ) {materialFunction=&MF_UI_AntiAliasing}
  Add(
   A=NamedRerouteDeclaration(None=@n10) {name=IconColorMask, nodeColor=<rgba 0 0 0 1>}
   B=NamedRerouteDeclaration(None=Saturate(None=@n11.Outer)) {name=LeftSlashMask, nodeColor=<rgba 0 0 0 1>}
  )
  n12 = Constant() {r=0.3}
  n13 = Constant2Vector() {g=1.666, r=1.666}
  n14 = TextureSample(UVs=Add(
    A=MaterialFunctionCall(Texture Scale=@n13) {materialFunction=&ScaleUVsByCenter}.UVs
    B=Constant2Vector() {g=0.0125}
   )) {
   samplerSource=SSM_Clamp_WorldGroupSettings
   samplerType=SAMPLERTYPE_LinearColor
   texture=/FortUILibrary/Textures/Icons/T_UI_Icon_UIKit_Character_SDF
  }
  n15 = Constant() {r=0.725}
  NamedRerouteDeclaration(None=LinearInterpolate(
    A=MaterialFunctionCall(DistanceField=@n1.R, Threshold=@n12) {materialFunction=&MF_UI_AntiAliasing}.Outer
    B=MaterialFunctionCall(DistanceField=@n14.G, Threshold=@n15) {materialFunction=&MF_UI_AntiAliasing}.Outer
    Alpha=ScalarParameter() {parameterName=IsSolo}
   )) {name=IconMask, nodeColor=<rgba 0 0 0 1>}
  n16 = Add(A=MaterialFunctionCall(Size=ScalarParameter() {defaultValue=0.75, parameterName=CircleSize}) {materialFunction=&MF_UI_Circle}.SDF) {constB=0.5}
  n17 = Constant() {r=0.5}
  NamedRerouteDeclaration(None=Multiply(
    A=MaterialFunctionCall(DistanceField=@n16, Threshold=@n17) {materialFunction=&MF_UI_AntiAliasing}.Inner
    B=ScalarParameter() {defaultValue=1, parameterName=BackingAlpha}
   ) {constB=0.6}) {name=BackingMask, nodeColor=<rgba 0 0 0 1>}
  n18 = Saturate(None=Add(A=NamedRerouteUsage(), B=NamedRerouteUsage()))
  NamedRerouteDeclaration(None=Add(A=@n10, B=NamedRerouteDeclaration(None=@n18) {name=SlashMasks, nodeColor=<rgba 0 0 0 1>})) {name=Icon+SlashMask, nodeColor=<rgba 0 0 0 1>}
  n19 = StaticSwitchParameter(True=ScalarParameter() {defaultValue=0.5, parameterName=SlashWidth}, False=Constant() {r=0.5}) {parameterName=JustSlashes}
  Saturate(None=NamedRerouteDeclaration(None=MaterialFunctionCall(DistanceField=@n8.R, Threshold=@n19) {materialFunction=&MF_UI_AntiAliasing}.Outer) {name=RightSlashMask, nodeColor=<rgba 0 0 0 1>})
  n20 = NamedRerouteUsage()
  StaticSwitchParameter(True=@n20, False=LinearInterpolate(A=NamedRerouteUsage(), Alpha=NamedRerouteUsage())) {parameterName=JustSlashes}
  n21 = ComponentMask(None=CollectionParameter() {collection=/BattleRoyaleGameUI/Shared/MPC_BRGameUI, parameterName=Red}) {b=true, g=true, r=true}
  n22 = LinearInterpolate(
   A=ComponentMask(None=CollectionParameter() {collection=&MPC_BRGameUI, parameterName=Red}) {b=true, g=true, r=true}
   B=VectorParameter() {parameterName=BackingColor, defaultValue=<rgba 0 0 0 1>}.RGB
   Alpha=ScalarParameter() {defaultValue=1, parameterName=Anim_Color}
  )
  n23 = LinearInterpolate(
   A=VectorParameter() {parameterName=IconColor_AnimStart, defaultValue=<rgba 0.9823 0.9823 0.9823 1>}.RGB
   B=VectorParameter() {parameterName=IconColor, defaultValue=<rgba 0.9823 0.9823 0.9823 1>}.RGB
   Alpha=ScalarParameter() {defaultValue=1, parameterName=Anim_Color}
  )
  n24 = NamedRerouteUsage()
  StaticSwitchParameter(True=@n21, False=LinearInterpolate(A=LinearInterpolate(A=@n22, B=@n23, Alpha=@n24), B=@n21, Alpha=@n20)) {parameterName=JustSlashes}
 ]
}
