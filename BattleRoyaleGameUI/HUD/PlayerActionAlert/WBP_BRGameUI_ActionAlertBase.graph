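# Base HUD widget for battle royale player action alerts, with a named slot for content and an empty HandleAlertData hook.
# ShowAlert reveals the root, plays the show animation and arms two CVar-driven timers: minimum duration fires EventRequestNextAlert, maximum triggers FadeAlert.
# With bPreventSequentialTriggering set, repeat calls on a visible alert only restart the timers.
# FadeAlert runs the overridable HandleFadeAlert, whose default animation hides the root on finishing.
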WidgetBlueprint /BattleRoyaleGameUI/HUD/PlayerActionAlert/WBP_BRGameUI_ActionAlertBase
animations [
 AnimShowAlert_Default {
  binding_count=1
  length_seconds=0.0167
  object=&self:AnimShowAlert_Default
  sequence {
   bindings [Overlay_AlertRoot {
     id=#1
     possessed_class=Overlay
     tracks [
      {
       type=MovieSceneFloatTrack
       display_name=Render Opacity
       sections [{type=MovieSceneFloatSection, channels [None {type=MovieSceneScriptingFloatChannel, keys [[0,0],[1,1]]}]}]
      }
      {
       type=MovieScene2DTransformTrack
       sections [{
         type=MovieScene2DTransformSection
         channels [
          Translation.X {type=MovieSceneScriptingFloatChannel, const [0,2]}
          Translation.Y {type=MovieSceneScriptingFloatChannel, const [0,2]}
          Scale.X {type=MovieSceneScriptingFloatChannel, const [1,2]}
          Scale.Y {type=MovieSceneScriptingFloatChannel, const [1,2]}
         ]
        }]
      }
     ]
    }]
   display_rate {numerator=60}
  }
 }
 AnimFadeAlert_Default {
  binding_count=1
  length_seconds=0.4
  object=&self:AnimFadeAlert_Default
  sequence {bindings [Overlay_AlertRoot {
     id=#2
     possessed_class=Overlay
     tracks [
      {
       type=MovieScene2DTransformTrack
       sections [{
         type=MovieScene2DTransformSection
         channels [
          Translation.X {type=MovieSceneScriptingFloatChannel, const [0,2]}
          Translation.Y {type=MovieSceneScriptingFloatChannel, keys [[0,0],[8,24]]}
          Scale.X {type=MovieSceneScriptingFloatChannel, keys [[0,1],[8,0.75]]}
          Scale.Y {type=MovieSceneScriptingFloatChannel, keys [[0,1],[8,0.9]]}
         ]
        }]
      }
      {
       type=MovieSceneFloatTrack
       display_name=Render Opacity
       sections [{type=MovieSceneFloatSection, channels [None {type=MovieSceneScriptingFloatChannel, keys [[0,1],[8,0]]}]}]
      }
     ]
    }]}
 }
]
graphs [
 EventGraph {nodes [
   ShowAlert() {is=title K2Node_CustomEvent}
   > n1 = Branch(Condition=AND Boolean(A=Get bAlertShown() -> (bAlertShown:bool), B=Get bPreventSequentialTriggering() -> (bPreventSequentialTriggering:bool)))
   > Clear and Invalidate Timer by Handle(Handle=Get RequestNextAlertTimerHandle() -> (RequestNextAlertTimerHandle:Timer Handle Structure))
   > n2 = Clear and Invalidate Timer by Handle(Handle=Get FadeTimerHandle() -> (FadeTimerHandle:Timer Handle Structure))
   n3 = GetConsoleVariableFloatValue!(execute=[@n2.then, @n9.then], VariableName:String=BattleRoyaleGameUI.PlayerActionAlerts.MinDuration) -> (then, ReturnValue:float)
   n4 = RequestNextAlert() {is=title K2Node_CustomEvent}
   n5 = Set Timer by Event(execute=@n3.then, Delegate=@n4, Time=@n3, bLooping=false)
   > Set RequestNextAlertTimerHandle(RequestNextAlertTimerHandle:Timer Handle Structure=@n5) -> (Output_Get:Timer Handle Structure)
   > n6 = GetConsoleVariableFloatValue(VariableName=BattleRoyaleGameUI.PlayerActionAlerts.MaxDuration)
   n7 = FadeAlert() {is=title K2Node_CustomEvent}
   n8 = Set Timer by Event(execute=@n6.then, Delegate=@n7, Time=@n6, bLooping=false)
   > Set FadeTimerHandle(FadeTimerHandle:Timer Handle Structure=@n8) -> (Output_Get:Timer Handle Structure)
   SetVisibility(execute=@n1.else, self=Get Overlay_AlertRoot() -> (Overlay_AlertRoot:Overlay Object Reference), InVisibility=HitTestInvisible)
   > HandleShowAlert()
   > n9 = Set bAlertShown(bAlertShown:bool=true) -> (Output_Get:bool)
   n10 = StopAndStartAnimation!() -> (OutputDelegate, then, AnimToStop:Widget Animation Object Reference, AnimToStart:Widget Animation Object Reference) {is=title K2Node_CustomEvent}
   > n11 = Branch(Condition=IsAnimationPlaying(InAnimation=@n10.AnimToStop))
   > StopAnimation(InAnimation=@n10.AnimToStop)
   n12 = Play Animation with Finished event!(
    execute:Exec=@n11.else
    Widget:User Widget Object Reference=Self-Reference()
    InAnimation:Widget Animation Object Reference=@n10.AnimToStart
    StartAtTime:float=0
    NumLoopsToPlay:int=1
    PlayMode:EUMGSequencePlayMode Enum
    PlaybackSpeed:float=1
   ) -> (then:Exec, Result:Widget Animation Handle Structure, Finished:Exec) {is=title K2Node_PlayAnimation2}
   HandleAnimFinished!(execute=@n12.Finished, self, AnimReference:Widget Animation Object Reference=@n10.AnimToStart) -> (then)
   Event Construct()
   > SetVisibility(self=Get Overlay_AlertRoot(), InVisibility=Hidden)
   Event Destruct()
   > Clear and Invalidate Timer by Handle(Handle=Get FadeTimerHandle())
   > Set bAlertShown(bAlertShown=false)
   Call Event Request Next Alert!(execute:Exec=@n4.then, self:WBP BRGame UI Action Alert Base Object Reference) -> (then:Exec) {is=title K2Node_CallDelegate}
   Reroute Node(InputPin:Wildcard) -> (OutputPin:Wildcard)
   Set bAlertShown(execute=@n13.then, bAlertShown=false)
   Clear and Invalidate Timer by Handle(execute=@n7.then, Handle=Get FadeTimerHandle())
   > Branch(Condition=Get bAlertShown())
   > n13 = HandleFadeAlert()
  ]}
 EventRequestNextAlert {nodes [EventRequestNextAlert() {is=title K2Node_FunctionEntry}]}
 HandleAlertData {nodes [HandleAlertData!() -> (then, AlertData:Athena Player Action Alert Structure) {is=title K2Node_FunctionEntry}]}
 HandleAnimFinished {nodes [
   n1 = HandleAnimFinished!() -> (then, AnimReference:Widget Animation Object Reference) {is=title K2Node_FunctionEntry}
   > Branch(Condition="Equal (Object)"(A=@n1, B=Get AnimFadeAlert_Default() -> (AnimFadeAlert_Default:Widget Animation Object Reference)))
   > SetVisibility(self=Get Overlay_AlertRoot(), InVisibility=Hidden)
  ]}
 HandleFadeAlert {nodes [
   HandleFadeAlert() {is=title K2Node_FunctionEntry}
   > StopAndStartAnimation!(
    execute
    self
    AnimToStop:Widget Animation Object Reference=Get AnimShowAlert_Default() -> (AnimShowAlert_Default:Widget Animation Object Reference)
    AnimToStart:Widget Animation Object Reference=Get AnimFadeAlert_Default()
   ) -> (then) {is=title K2Node_CallFunction}
  ]}
 HandleShowAlert {nodes [
   HandleShowAlert() {is=title K2Node_FunctionEntry}
   > StopAndStartAnimation(AnimToStop=Get AnimFadeAlert_Default(), AnimToStart=Get AnimShowAlert_Default())
  ]}
]
variables [
 bAlertShown {type=boolean}
 bPreventSequentialTriggering {type=boolean}
 EventRequestNextAlert {}
 FadeTimerHandle {type=TimerHandle}
 MaxAlertDuration {type=number, default=2}
 MinAlertDuration {type=number, default=0.075}
 RequestNextAlertTimerHandle {type=TimerHandle}
]
widgets {
 info {inheritedWidgetCount=0, parentClass=&UserWidget, rootWidgetClass=&Overlay, widgetCount=2}
 widgets [
  Overlay_AlertRoot {widgetClassPath=&Overlay, bIsVariable=true, widget=Overlay_AlertRoot, properties {pixelSnapping=Disabled}}
  AlertContents {
   widgetClassPath=&NamedSlot
   bIsVariable=true
   parent=Overlay_AlertRoot
   widget=AlertContents
   slot {class=OverlaySlot, properties {horizontalAlignment=HAlign_Fill, verticalAlignment=VAlign_Fill}}
  }
 ]
}
