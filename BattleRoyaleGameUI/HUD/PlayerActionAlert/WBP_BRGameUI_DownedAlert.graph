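# HUD alert widget, derived from WBP_BRGameUI_ActionAlertBase, that announces a knocked player with detail text and a rank icon.
# HandleAlertData fills the text, resolves the icon through Habanero display data with overrides, collapsing it on failure, then plays AnimShowAlert.
# AnimShowAlert combines transform, opacity and material-parameter tracks for the pop-in.
# A separate SetPlayerPortrait function, not called here, shows the pawn's character cosmetic preview or collapses the portrait.
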WidgetBlueprint /BattleRoyaleGameUI/HUD/PlayerActionAlert/WBP_BRGameUI_DownedAlert
animations [AnimShowAlert {
  binding_count=9
  length_seconds=0.75
  object=&self:AnimShowAlert
  sequence {
   bindings [
    Image_Bracket_Right {
     id=#1
     possessed_class=Image
     tracks [
      {
       type=MovieScene2DTransformTrack
       sections [{
         type=MovieScene2DTransformSection
         channels [
          Translation.X {type=MovieSceneScriptingFloatChannel, keys [[3,192],[12,0]]}
          Translation.Y {type=MovieSceneScriptingFloatChannel, const [0,1]}
          Scale.X {type=MovieSceneScriptingFloatChannel, keys [[3,2],[12,0.9],[17,1]]}
          Scale.Y {type=MovieSceneScriptingFloatChannel, keys [[3,2],[12,1]]}
         ]
        }]
      }
      {
       type=MovieSceneWidgetMaterialTrack
       sections [{
         type=MovieSceneParameterSection
         channels [
          Anim_Color {type=MovieSceneScriptingFloatChannel, keys [[12,1],[13,0],[45,1]]}
          Anim_Opacity {type=MovieSceneScriptingFloatChannel, keys [[2,0],[4,1],[24,1]]}
          Anim_Stretch {type=MovieSceneScriptingFloatChannel, keys [[9,0],[12,0.75],[24,0]]}
         ]
        }]
      }
     ]
    }
    LazyImage_Habanero_RankIcon {
     id=#2
     possessed_class=FortLazyImage
     tracks [
      {
       type=MovieSceneFloatTrack
       display_name=Render Opacity
       sections [{type=MovieSceneFloatSection, channels [None {type=MovieSceneScriptingFloatChannel, keys [[15,0],[18,1]]}]}]
      }
      {
       type=MovieScene2DTransformTrack
       sections [{
         type=MovieScene2DTransformSection
         channels [
          Translation.X {type=MovieSceneScriptingFloatChannel, const [0,1]}
          Translation.Y {type=MovieSceneScriptingFloatChannel, const [0,1]}
          Scale.X {type=MovieSceneScriptingFloatChannel, keys [[15,0],[21,1.25],[27,0.9],[42,1]]}
          Scale.Y {type=MovieSceneScriptingFloatChannel, keys [[15,0],[21,1.25],[27,0.9],[42,1]]}
          Shear.X {type=MovieSceneScriptingFloatChannel, const [0,1]}
          Shear.Y {type=MovieSceneScriptingFloatChannel, const [0,1]}
         ]
        }]
      }
     ]
    }
    Image_Bracket_Left {
     id=#3
     possessed_class=Image
     tracks [
      {
       type=MovieScene2DTransformTrack
       sections [{
         type=MovieScene2DTransformSection
         channels [
          Translation.X {type=MovieSceneScriptingFloatChannel, keys [[3,-192],[12,0]]}
          Translation.Y {type=MovieSceneScriptingFloatChannel, const [0,1]}
          Scale.X {type=MovieSceneScriptingFloatChannel, keys [[3,2],[12,0.9],[17,1]]}
          Scale.Y {type=MovieSceneScriptingFloatChannel, keys [[3,2],[12,1]]}
         ]
        }]
      }
      {
       type=MovieSceneWidgetMaterialTrack
       sections [{
         type=MovieSceneParameterSection
         channels [
          Anim_Color {type=MovieSceneScriptingFloatChannel, keys [[12,1],[13,0],[45,1]]}
          Anim_Opacity {type=MovieSceneScriptingFloatChannel, keys [[2,0],[4,1],[24,1]]}
          Anim_Stretch {type=MovieSceneScriptingFloatChannel, keys [[9,0],[12,0.75],[24,0]]}
         ]
        }]
      }
     ]
    }
    Text_Background {
     id=#4
     possessed_class=CommonTextBlock
     tracks [
      {
       type=MovieScene2DTransformTrack
       sections [{
         type=MovieScene2DTransformSection
         channels [
          Translation.X {type=MovieSceneScriptingFloatChannel, default=0}
          Translation.Y {type=MovieSceneScriptingFloatChannel, keys [[12,40],[45,0]]}
          Scale.X {type=MovieSceneScriptingFloatChannel, keys [[12,0.5],[45,1]]}
          Scale.Y {type=MovieSceneScriptingFloatChannel, keys [[12,0.5],[45,1]]}
         ]
        }]
      }
      {
       type=MovieSceneWidgetMaterialTrack
       display_name=Font.FontMaterial
       sections [{
         type=MovieSceneParameterSection
         channels [
          Max_Alpha {type=MovieSceneScriptingFloatChannel, keys [[12,0],[24,1],[45,1]]}
          Min_Alpha {type=MovieSceneScriptingFloatChannel, keys [[12,0],[24,1],[45,0.75]]}
         ]
        }]
      }
      {
       type=MovieSceneWidgetMaterialTrack
       display_name=Font.Outline Settings.OutlineMaterial
       sections [{type=MovieSceneParameterSection, channels [AnimOpacity {type=MovieSceneScriptingFloatChannel, keys [[12,0],[45,1]]}]}]
      }
     ]
    }
    LazyImage_PlayerPortrait {
     id=#5
     possessed_class=FortLazyImage
     tracks [
      {
       type=MovieSceneFloatTrack
       display_name=Render Opacity
       sections [{type=MovieSceneFloatSection, channels [None {type=MovieSceneScriptingFloatChannel, keys [[9,0],[12,1]]}]}]
      }
      {
       type=MovieSceneWidgetMaterialTrack
       sections [{type=MovieSceneParameterSection, channels [Anim_Translate {type=MovieSceneScriptingFloatChannel, keys [[9,0],[30,1]]}]}]
      }
     ]
    }
    PlayerDownText {
     id=#6
     possessed_class=CommonTextBlock
     tracks [
      {
       type=MovieSceneFloatTrack
       display_name=Render Opacity
       sections [{type=MovieSceneFloatSection, channels [None {type=MovieSceneScriptingFloatChannel, keys [[15,0],[17,1]]}]}]
      }
      {
       type=MovieScene2DTransformTrack
       sections [{
         type=MovieScene2DTransformSection
         channels [
          Scale.X {type=MovieSceneScriptingFloatChannel, keys [[15,0.75],[21,1.125],[34,1]]}
          Scale.Y {type=MovieSceneScriptingFloatChannel, keys [[15,0.75],[21,1.125],[34,1]]}
         ]
        }]
      }
     ]
    }
    Glow {
     id=#7
     possessed_class=Image
     tracks [{
       type=MovieSceneWidgetMaterialTrack
       sections [{type=MovieSceneParameterSection, channels [Anim_Opacity {type=MovieSceneScriptingFloatChannel, keys [[10,0],[14,1],[45,0]]}]}]
      }]
    }
    AlertContent {
     id=#8
     possessed_class=Overlay
     tracks [{
       type=MovieScene2DTransformTrack
       sections [{
         type=MovieScene2DTransformSection
         channels [
          Scale.X {type=MovieSceneScriptingFloatChannel, keys [[9,1],[12,0.5],[30,1]]}
          Scale.Y {type=MovieSceneScriptingFloatChannel, keys [[9,1],[12,0.5],[30,1]]}
         ]
        }]
      }]
    }
    Backing {
     id=#9
     possessed_class=Image
     tracks [
      {
       type=MovieScene2DTransformTrack
       sections [{
         type=MovieScene2DTransformSection
         channels [
          Scale.X {type=MovieSceneScriptingFloatChannel, keys [[3,4],[12,1]]}
          Scale.Y {type=MovieSceneScriptingFloatChannel, keys [[9,0.125],[12,1]]}
          Shear.X {type=MovieSceneScriptingFloatChannel}
          Shear.Y {type=MovieSceneScriptingFloatChannel}
         ]
        }]
      }
      {
       type=MovieSceneWidgetMaterialTrack
       sections [{
         type=MovieSceneParameterSection
         channels [
          Anim_Opacity {type=MovieSceneScriptingFloatChannel, keys [[0,0],[1,1]]}
          Anim_Color {type=MovieSceneScriptingFloatChannel, keys [[15,1],[24,0]]}
          Anim_FadeColorEdges {type=MovieSceneScriptingFloatChannel, keys [[18,0],[45,1]]}
          Anim_PureBlureMix {type=MovieSceneScriptingFloatChannel, keys [[24,0],[45,1]]}
         ]
        }]
      }
     ]
    }
   ]
   display_rate {numerator=60}
  }
 }]
graphs [
 EventGraph {links [], nodes []}
 HandleAlertData {nodes [
   n1 = HandleAlertData!() -> (then, AlertData:Athena Player Action Alert Structure) {is=title K2Node_FunctionEntry}
   n2 = Break Athena Player Action Alert!(AthenaPlayerActionAlert:"Athena Player Action Alert Structure (by ref)"=@n1) -> (
    AlertType:EAthenaPlayerActionAlert Enum
    DetailText:Text
    RankedProgress:Ranked Progress Replicated Data Structure
    OtherPlayerState:Fort Player State Athena Weak object
    NumKills:int
   ) {is=title K2Node_BreakStruct}
   > "SetText (Text)"(self=Get PlayerDownText() -> (PlayerDownText:Common Text Object Reference), InText=@n2.DetailText)
   > SetRankedProgress!(execute, self, RankedProgress:Ranked Progress Replicated Data Structure=@n2.RankedProgress) -> (then)
   > PlayAnimation(InAnimation=Get AnimShowAlert() -> (AnimShowAlert:Widget Animation Object Reference))
  ]}
 SetPlayerPortrait {nodes [
   n1 = SetPlayerPortrait!() -> (then, OtherPlayerState:Fort Player State Object Reference) {is=title K2Node_FunctionEntry}
   > n2 = Is Valid(InputObject=@n1)
   n3 = GetCurrentPawn!(self:Fort Player State Object Reference=@n1) -> (ReturnValue:Fort Player Pawn Object Reference)
   > n4 = Is Valid(InputObject=@n3)
   n5 = Get LazyImage_PlayerPortrait() -> (LazyImage_PlayerPortrait:Fort Lazy Image Object Reference)
   > SetVisibility(self=@n5, InVisibility=HitTestInvisible)
   > SetBrushFromSoftTexture!(
    execute
    self:Image Object Reference=Get LazyImage_PlayerPortrait()
    SoftTexture:Texture 2D Soft Object Reference=GetSmallPreviewImage!(ItemDefinition:Item Definition Base Object Reference=GetItemDefinitionFromBaseLoadoutWithActor!(Actor:Actor Object Reference=@n3, LoadoutSlot:Gameplay Tag Structure=(TagName="LoadoutSlot.Character")) -> (ReturnValue:Athena Cosmetic Item Definition Object Reference)) -> (ReturnValue:Texture 2D Soft Object Reference)
    bMatchSize:bool=false
   ) -> (then)
   SetVisibility(execute=[@n4.Is Not Valid, @n2.Is Not Valid], self=@n5, InVisibility=Collapsed)
  ]}
 SetRankedProgress {nodes [
   n1 = SetRankedProgress!() -> (then, RankedProgress:Ranked Progress Replicated Data Structure) {is=title K2Node_FunctionEntry}
   n2 = Break Ranked Progress Replicated Data!(RankedProgressReplicatedData:"Ranked Progress Replicated Data Structure (by ref)"=@n1) -> (RankType:String, Rank:int, HighestRank:int, CurrentMMR:int, CurrentPosition:int) {is=title K2Node_BreakStruct}
   n3 = GetHabaneroDisplayDataOverride!(
    InDisplayData:Fort Habanero Display Data Object Reference=Get RankData() -> (RankData:Fort Habanero Display Data Object Reference)
    InOverrideData:Fort Compete UIHabanero Tier Override List Object Reference=Get RankInfoOverrides() -> (RankInfoOverrides:Fort Compete UIHabanero Tier Override List Object Reference)
    InKey:"Fort Compete UIHabanero Tier Override Key Structure (by ref)"=Make Fort Compete UIHabanero Tier Override Key!(
     RankedTrack:Name=String To Name(InString=@n2)
     Season:String
     Rank:int=@n2.Rank
     MatchmakingRating:int=@n2.CurrentMMR
     CurrentPosition:int=@n2.CurrentPosition
    ) -> (FortCompeteUIHabaneroTierOverrideKey:Fort Compete UIHabanero Tier Override Key Structure) {is=title K2Node_MakeStruct}
   ) -> (OutData:Fort Habanero Tier Override Structure, ReturnValue:bool)
   > n4 = Branch(Condition=@n3.ReturnValue)
   n5 = Get LazyImage_Habanero_RankIcon() -> (LazyImage_Habanero_RankIcon:Fort Lazy Image Object Reference)
   > SetVisibility(self=@n5, InVisibility=HitTestInvisible)
   > SetBrushFromSoftTexture(
    self=@n5
    SoftTexture=Break Fort Habanero Tier!(FortHabaneroTier:"Fort Habanero Tier Structure (by ref)"=@n3) -> (
     ForegroundIcon:Material Instance Soft Object Reference
     ForegroundIconSmall:Material Instance Soft Object Reference
     BackgroundIcon:Material Instance Soft Object Reference
     BackgroundIconSmall:Material Instance Soft Object Reference
     WhiteIcon:Texture 2D Soft Object Reference
     SimpleIconSmall:Texture 2D Soft Object Reference
     Name:Text
     TierType:Fort Habanero Tier Type Data Object Reference
     AdditionalLeftIcon:Material Instance Soft Object Reference
     AdditionalRightIcon:Material Instance Soft Object Reference
    ) {is=title K2Node_BreakStruct}.SimpleIconSmall
    bMatchSize=false
   )
   SetVisibility(execute=@n4.else, self=@n5, InVisibility=Collapsed)
  ]}
]
variables [
 RankData {type=FortHabaneroDisplayData, default=/Game/UI/Habanero/HabaneroDisplayData}
 RankInfoOverrides {type=FortCompeteUIHabaneroTierOverrideList, default=/Game/UI/Habanero/DA_RankDisplayDataOverride}
]
widgets {
 info {
  inheritedWidgetCount=0
  namedSlotCount=1
  parentClass=&/WBP_BRGameUI_ActionAlertBase.WBP_BRGameUI_ActionAlertBase_C
  rootWidgetClass=null
  widgetCount=13
 }
 widgets [
  UIContent {widgetClassPath=&Overlay, widget=UIContent, properties {bOverride_Cursor=true}}
  Text_Background {
   widgetClassPath=&CommonTextBlock
   parent=UIContent
   widget=Text_Background
   properties {
    justification=Center
    mobileFontSizeMultiplier=0.85
    text="Knocked!"
    textTransformPolicy=ToUpper
    font {
     bForceMonospaced=false
     bMaterialIsStencil=false
     fontMaterial=/BattleRoyaleGameUI/Shared/M_UI_Font_FlavorText_Master
     fontObject=&Burbank
     letterSpacing=0
     monospacedWidth=1
     size=24
     skewAmount=0.25
     typefaceFontName=BurbankRegular-Black
     outlineSettings {
      bApplyOutlineToDropShadows=false
      bMiteredCorners=false
      bSeparateFillAlpha=false
      outlineMaterial=/BattleRoyaleGameUI/Shared/MI_UI_Font_SoftShadow_LargeText
      outlineSize=10
      outlineColor=<rgba 1 1 1 1>
     }
    }
    margin=<ltrb 0 -12 0 -16>
    shadowOffset=<0 0>
    strikeBrush {
     bIsDynamicallyLoaded=false
     drawAs=Image
     imageType=NoImage
     mirroring=NoMirror
     resourceName=None
     resourceObject=null
     tiling=NoTile
     imageSize=<32 32>
     margin=<ltrb 0 0 0 0>
     outlineSettings {
      bUseBrushTransparency=false
      roundingType=HalfHeightRadius
      width=0
      color {colorUseRule=UseColor_Specified, specifiedColor=<rgba 0 0 0 0>}
      cornerRadii=<0 0 0 0>
     }
     tintColor {colorUseRule=UseColor_Specified, specifiedColor=<rgba 0.9 0.9 0.9 1>}
     uVRegion {bIsValid=false, max=<0 0>, min=<0 0>}
    }
   }
   slot {class=OverlaySlot, properties {horizontalAlignment=HAlign_Center, padding=<ltrb 0 -32 0 0>}}
  }
  AlertContent {
   widgetClassPath=&Overlay
   parent=UIContent
   widget=AlertContent
   slot {class=OverlaySlot, properties {horizontalAlignment=HAlign_Center, verticalAlignment=VAlign_Center}}
  }
  BackingandDecor {
   widgetClassPath=&Overlay
   parent=AlertContent
   widget=BackingandDecor
   slot {class=OverlaySlot, properties {horizontalAlignment=HAlign_Fill, verticalAlignment=VAlign_Fill}}
  }
  Glow {
   widgetClassPath=&Image
   parent=BackingandDecor
   widget=Glow
   properties {
    visibility=SelfHitTestInvisible
    brush {
     bIsDynamicallyLoaded=false
     drawAs=Image
     imageType=FullColor
     mirroring=NoMirror
     resourceName=None
     resourceObject=/BattleRoyaleGameUI/HUD/PlayerActionAlert/Assets/MI_UI_AngledBox_Glow_Knocked
     tiling=NoTile
     imageSize=<32 32>
     margin=<ltrb 0 0 0 0>
     outlineSettings {
      bUseBrushTransparency=false
      roundingType=HalfHeightRadius
      width=0
      color {colorUseRule=UseColor_Specified, specifiedColor=<rgba 0 0 0 0>}
      cornerRadii=<0 0 0 0>
     }
     tintColor {colorUseRule=UseColor_Specified, specifiedColor=<rgba 1 1 1 1>}
     uVRegion {bIsValid=false, max=<0 0>, min=<0 0>}
    }
   }
   slot {class=OverlaySlot, properties {horizontalAlignment=HAlign_Fill, verticalAlignment=VAlign_Fill, padding=<ltrb -56 -32 -56 -56>}}
  }
  Backing {
   widgetClassPath=&Image
   parent=BackingandDecor
   widget=Backing
   properties {
    visibility=SelfHitTestInvisible
    brush {
     bIsDynamicallyLoaded=false
     drawAs=Image
     imageType=FullColor
     mirroring=NoMirror
     resourceName=None
     resourceObject=/BattleRoyaleGameUI/HUD/PlayerActionAlert/Assets/MI_UI_Knocked_Backing
     tiling=NoTile
     imageSize=<32 32>
     margin=<ltrb 0 0 0 0>
     outlineSettings {
      bUseBrushTransparency=false
      roundingType=HalfHeightRadius
      width=0
      color {colorUseRule=UseColor_Specified, specifiedColor=<rgba 0 0 0 0>}
      cornerRadii=<0 0 0 0>
     }
     tintColor {colorUseRule=UseColor_Specified, specifiedColor=<rgba 1 1 1 1>}
     uVRegion {bIsValid=false, max=<0 0>, min=<0 0>}
    }
   }
   slot {class=OverlaySlot, properties {horizontalAlignment=HAlign_Fill, verticalAlignment=VAlign_Center}}
  }
  Image_Bracket_Left {
   widgetClassPath=&Image
   parent=BackingandDecor
   widget=Image_Bracket_Left
   properties {
    visibility=SelfHitTestInvisible
    brush {
     bIsDynamicallyLoaded=false
     drawAs=Image
     imageType=FullColor
     mirroring=NoMirror
     resourceName=None
     resourceObject=/BattleRoyaleGameUI/HUD/PlayerActionAlert/Assets/MI_UI_Knocked_Bracket
     tiling=NoTile
     imageSize=<56 48>
     margin=<ltrb 0 0 0 0>
     outlineSettings {
      bUseBrushTransparency=false
      roundingType=HalfHeightRadius
      width=0
      color {colorUseRule=UseColor_Specified, specifiedColor=<rgba 0 0 0 0>}
      cornerRadii=<0 0 0 0>
     }
     tintColor {colorUseRule=UseColor_Specified, specifiedColor=<rgba 1 1 1 1>}
     uVRegion {bIsValid=false, max=<0 0>, min=<0 0>}
    }
    renderTransformPivot=<0.75 0.5>
   }
   slot {class=OverlaySlot, properties {verticalAlignment=VAlign_Center, padding=<ltrb -12 0 0 0>}}
  }
  Image_Bracket_Right {
   widgetClassPath=&Image
   parent=BackingandDecor
   widget=Image_Bracket_Right
   properties {
    visibility=SelfHitTestInvisible
    brush {
     bIsDynamicallyLoaded=false
     drawAs=Image
     imageType=FullColor
     mirroring=NoMirror
     resourceName=None
     resourceObject=/BattleRoyaleGameUI/HUD/PlayerActionAlert/Assets/MI_UI_Knocked_Bracket_Flipped
     tiling=NoTile
     imageSize=<56 48>
     margin=<ltrb 0 0 0 0>
     outlineSettings {
      bUseBrushTransparency=false
      roundingType=HalfHeightRadius
      width=0
      color {colorUseRule=UseColor_Specified, specifiedColor=<rgba 0 0 0 0>}
      cornerRadii=<0 0 0 0>
     }
     tintColor {colorUseRule=UseColor_Specified, specifiedColor=<rgba 1 1 1 1>}
     uVRegion {bIsValid=false, max=<0 0>, min=<0 0>}
    }
    renderTransformPivot=<0.125 0.5>
   }
   slot {class=OverlaySlot, properties {horizontalAlignment=HAlign_Right, verticalAlignment=VAlign_Center, padding=<ltrb 0 0 -12 0>}}
  }
  LazyImage_PlayerPortrait {
   widgetClassPath=&FortLazyImage
   bIsVariable=true
   parent=BackingandDecor
   widget=LazyImage_PlayerPortrait
   properties {
    materialTextureParamName=Portrait
    visibility=Collapsed
    brush {
     bIsDynamicallyLoaded=false
     drawAs=Image
     imageType=FullColor
     mirroring=NoMirror
     resourceName=None
     resourceObject=/BattleRoyaleGameUI/HUD/PlayerActionAlert/Assets/M_UI_ElimPortrait
     tiling=NoTile
     imageSize=<44 44>
     margin=<ltrb 0 0 0 0>
     outlineSettings {
      bUseBrushTransparency=false
      roundingType=HalfHeightRadius
      width=0
      color {colorUseRule=UseColor_Specified, specifiedColor=<rgba 0 0 0 0>}
      cornerRadii=<0 0 0 0>
     }
     tintColor {colorUseRule=UseColor_Specified, specifiedColor=<rgba 1 1 1 1>}
     uVRegion {bIsValid=false, max=<0 0>, min=<0 0>}
    }
    loadingBackgroundBrush {
     bIsDynamicallyLoaded=false
     drawAs=NoDrawType
     imageType=NoImage
     mirroring=NoMirror
     resourceName=None
     resourceObject=null
     tiling=NoTile
     imageSize=<64 64>
     margin=<ltrb 0 0 0 0>
     outlineSettings {
      bUseBrushTransparency=false
      roundingType=HalfHeightRadius
      width=0
      color {colorUseRule=UseColor_Specified, specifiedColor=<rgba 0 0 0 0>}
      cornerRadii=<0 0 0 0>
     }
     tintColor {colorUseRule=UseColor_Specified, specifiedColor=<rgba 1 1 1 1>}
     uVRegion {bIsValid=false, max=<0 0>, min=<0 0>}
    }
    renderTransformPivot=<0 0.5>
   }
   slot {class=OverlaySlot, properties {verticalAlignment=VAlign_Bottom, padding=<ltrb -2 0 0 10>}}
  }
  HorizontalBox_DownFeedback {
   widgetClassPath=&HorizontalBox
   parent=AlertContent
   widget=HorizontalBox_DownFeedback
   slot {class=OverlaySlot, properties {horizontalAlignment=HAlign_Center, verticalAlignment=VAlign_Center, padding=<ltrb 12 0 12 0>}}
  }
  HorizontalBox_DownPlayerInfo {
   widgetClassPath=&HorizontalBox
   parent=HorizontalBox_DownFeedback
   widget=HorizontalBox_DownPlayerInfo
   slot {class=HorizontalBoxSlot}
  }
  LazyImage_Habanero_RankIcon {
   widgetClassPath=&FortLazyImage
   bIsVariable=true
   parent=HorizontalBox_DownPlayerInfo
   widget=LazyImage_Habanero_RankIcon
   properties {renderTransform {angle=0, scale=<1.2 1.2>, shear=<0 0>, translation=<0 0>}}
   slot {class=HorizontalBoxSlot, properties {horizontalAlignment=HAlign_Center, verticalAlignment=VAlign_Center, padding=<ltrb 2 0 2 0>}}
  }
  PlayerDownText {
   widgetClassPath=&CommonTextBlock
   bIsVariable=true
   parent=HorizontalBox_DownPlayerInfo
   widget=PlayerDownText
   properties {
    justification=Center
    minDesiredWidth=64
    text="{UserName}"
    font {
     bForceMonospaced=false
     bMaterialIsStencil=false
     fontMaterial=/BattleRoyaleGameUI/Shared/M_UI_Font_White
     fontObject=&BurbankSmall
     letterSpacing=0
     monospacedWidth=1
     size=13.5
     skewAmount=0
     typefaceFontName=Black
     outlineSettings {
      bApplyOutlineToDropShadows=false
      bMiteredCorners=true
      bSeparateFillAlpha=false
      outlineMaterial=/BattleRoyaleGameUI/Shared/MI_UI_Font_SoftShadow_SmallText_Black
      outlineSize=5
      outlineColor=<rgba 1 1 1 1>
     }
    }
    margin=<ltrb -2 -4 0 -4>
    shadowOffset=<0 0>
    strikeBrush {
     bIsDynamicallyLoaded=false
     drawAs=Image
     imageType=NoImage
     mirroring=NoMirror
     resourceName=None
     resourceObject=null
     tiling=NoTile
     imageSize=<32 32>
     margin=<ltrb 0 0 0 0>
     outlineSettings {
      bUseBrushTransparency=false
      roundingType=HalfHeightRadius
      width=0
      color {colorUseRule=UseColor_Specified, specifiedColor=<rgba 0 0 0 0>}
      cornerRadii=<0 0 0 0>
     }
     tintColor {colorUseRule=UseColor_Specified, specifiedColor=<rgba 0.8 0.8 0.8 1>}
     uVRegion {bIsValid=false, max=<0 0>, min=<0 0>}
    }
   }
   slot {class=HorizontalBoxSlot, properties {verticalAlignment=VAlign_Center}}
  }
 ]
}
