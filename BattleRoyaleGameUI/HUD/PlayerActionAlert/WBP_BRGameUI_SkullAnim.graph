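# HUD widget for the battle royale player action alert that plays a one-second elimination skull drop animation.
# Material and transform tracks slam in a center skull, with an expanding burst and blue/green chromatic distortion copies; the animated Glow image is collapsed.
# PlayAnimWithOptionalDelay plays immediately or hides the root until a timer fires PlayAnim, which reveals it and runs AnimSkullDrop.
# A variant, AnimSkullDrop_0, is not referenced in the graph.
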WidgetBlueprint /BattleRoyaleGameUI/HUD/PlayerActionAlert/WBP_BRGameUI_SkullAnim
animations [
 AnimSkullDrop {
  binding_count=7
  length_seconds=1
  object=&self:AnimSkullDrop
  sequence {
   bindings [
    CenterSkull_ChromaticL {
     id=#1
     possessed_class=Image
     tracks [{
       type=MovieSceneWidgetMaterialTrack
       sections [{
         type=MovieSceneParameterSection
         channels [
          Anim_Alpha {type=MovieSceneScriptingFloatChannel, keys [[12,0],[15,1],[24,1],[60,0]]}
          ScreenDistortionAmount {type=MovieSceneScriptingFloatChannel, keys [[12,0],[18,1],[60,0]]}
         ]
        }]
      }]
    }
    CenterSkull {
     id=#2
     possessed_class=Image
     tracks [
      {
       type=MovieScene2DTransformTrack
       sections [{
         type=MovieScene2DTransformSection
         channels [
          Scale.X {type=MovieSceneScriptingFloatChannel, keys [[6,3],[12,0.75],[18,1]]}
          Scale.Y {type=MovieSceneScriptingFloatChannel, keys [[6,3],[12,0.75],[18,1]]}
         ]
        }]
      }
      {
       type=MovieSceneWidgetMaterialTrack
       sections [{
         type=MovieSceneParameterSection
         channels [
          Anim_Color {type=MovieSceneScriptingFloatChannel, keys [[24,1],[54,0]]}
          Anim_Alpha {type=MovieSceneScriptingFloatChannel, keys [[6,0],[9,1]]}
         ]
        }]
      }
     ]
    }
    CenterSkull_Burst {
     id=#3
     possessed_class=Image
     tracks [
      {
       type=MovieScene2DTransformTrack
       sections [{
         type=MovieScene2DTransformSection
         channels [
          Scale.X {type=MovieSceneScriptingFloatChannel, keys [[12,1],[60,3]]}
          Scale.Y {type=MovieSceneScriptingFloatChannel, keys [[12,1],[60,3]]}
         ]
        }]
      }
      {
       type=MovieSceneWidgetMaterialTrack
       sections [{
         type=MovieSceneParameterSection
         channels [
          Anim_Alpha {type=MovieSceneScriptingFloatChannel, keys [[12,0],[15.946,1],[60,0]]}
          Anim_Smoothing {type=MovieSceneScriptingFloatChannel, keys [[12,0],[60,0.5]]}
          Anim_Color {type=MovieSceneScriptingFloatChannel, const [0,1]}
         ]
        }]
      }
     ]
    }
    CenterSkull_ChromaticR {
     id=#4
     possessed_class=Image
     tracks [{
       type=MovieSceneWidgetMaterialTrack
       sections [{
         type=MovieSceneParameterSection
         channels [
          Anim_Alpha {type=MovieSceneScriptingFloatChannel, keys [[12,0],[15,1],[24,1],[60,0]]}
          ScreenDistortionAmount {type=MovieSceneScriptingFloatChannel, keys [[12,0],[18,1],[60,0]]}
         ]
        }]
      }]
    }
    SizeBox_0 {id=#5, possessed_class=SizeBox, tracks []}
    OverlaySlot_1 {
     id=#6
     parent_id=#5
     possessed_class=OverlaySlot
     tracks [{
       type=MovieSceneMarginTrack
       display_name=Padding
       sections [{
         type=MovieSceneMarginSection
         channels [
          Left {type=MovieSceneScriptingFloatChannel, keys [[0,-12],[6,0]]}
          Top {type=MovieSceneScriptingFloatChannel, default=0}
          Right {type=MovieSceneScriptingFloatChannel, keys [[0,-12],[6,0]]}
          Bottom {type=MovieSceneScriptingFloatChannel, default=0}
         ]
        }]
      }]
    }
    Glow {
     id=#7
     possessed_class=Image
     tracks [{
       type=MovieSceneFloatTrack
       display_name=Render Opacity
       sections [{type=MovieSceneFloatSection, channels [None {type=MovieSceneScriptingFloatChannel, keys [[18,0],[45,1]]}]}]
      }]
    }
   ]
   display_rate {numerator=60}
  }
 }
 AnimSkullDrop_0 {
  binding_count=8
  length_seconds=1.0011
  object=&self:AnimSkullDrop_0
  sequence {
   bindings [
    CenterSkull_ChromaticL {
     id=#1
     possessed_class=Image
     tracks [{
       type=MovieSceneWidgetMaterialTrack
       sections [{
         type=MovieSceneParameterSection
         channels [
          Anim_Alpha {type=MovieSceneScriptingFloatChannel, keys [[12,0],[15,1],[24,1],[54,0]]}
          ScreenDistortionAmount {type=MovieSceneScriptingFloatChannel, keys [[12,0],[18,1],[54,0]]}
         ]
        }]
      }]
    }
    CenterSkull {
     id=#2
     possessed_class=Image
     tracks [
      {
       type=MovieScene2DTransformTrack
       sections [{
         type=MovieScene2DTransformSection
         channels [
          Translation.X {type=MovieSceneScriptingFloatChannel, default=0}
          Translation.Y {type=MovieSceneScriptingFloatChannel, keys [[0,-80],[12,0]]}
          Scale.X {type=MovieSceneScriptingFloatChannel, keys [[12,0.5],[16,0.9],[20,1]]}
          Scale.Y {type=MovieSceneScriptingFloatChannel, keys [[9,3],[16,0.9],[20,1]]}
         ]
        }]
      }
      {
       type=MovieSceneWidgetMaterialTrack
       sections [{
         type=MovieSceneParameterSection
         channels [
          Anim_Color {type=MovieSceneScriptingFloatChannel, keys [[12,0],[14,1],[24,1],[54,0]]}
          Anim_Alpha {type=MovieSceneScriptingFloatChannel, keys [[6,0],[9,1]]}
         ]
        }]
      }
     ]
    }
    CenterSkull_Burst {
     id=#3
     possessed_class=Image
     tracks [
      {
       type=MovieScene2DTransformTrack
       sections [{
         type=MovieScene2DTransformSection
         channels [
          Scale.X {type=MovieSceneScriptingFloatChannel, keys [[12,1],[60,2]]}
          Scale.Y {type=MovieSceneScriptingFloatChannel, keys [[12,1],[60,2]]}
         ]
        }]
      }
      {
       type=MovieSceneWidgetMaterialTrack
       sections [{
         type=MovieSceneParameterSection
         channels [
          Anim_Alpha {type=MovieSceneScriptingFloatChannel, keys [[12,0],[16,1],[60,0]]}
          Anim_Smoothing {type=MovieSceneScriptingFloatChannel, keys [[12,0],[60,0.5]]}
          Anim_Color {type=MovieSceneScriptingFloatChannel, const [0,1]}
         ]
        }]
      }
     ]
    }
    CenterSkull_ChromaticR {
     id=#4
     possessed_class=Image
     tracks [{
       type=MovieSceneWidgetMaterialTrack
       sections [{
         type=MovieSceneParameterSection
         channels [
          Anim_Alpha {type=MovieSceneScriptingFloatChannel, keys [[12,0],[15,1],[24,1],[54,0]]}
          ScreenDistortionAmount {type=MovieSceneScriptingFloatChannel, keys [[12,0],[18,1],[54,0]]}
         ]
        }]
      }]
    }
    OverlaySlot_1 {id=#8, parent_id=#4, possessed_class=OverlaySlot, tracks []}
    OverlaySlot_1 {
     id=#9
     parent_id=#10
     possessed_class=OverlaySlot
     tracks [{
       type=MovieSceneMarginTrack
       display_name=Padding
       sections [{
         type=MovieSceneMarginSection
         channels [
          Left {type=MovieSceneScriptingFloatChannel, keys [[-0.063,-12],[0,-16],[6,0]]}
          Top {type=MovieSceneScriptingFloatChannel, default=0}
          Right {type=MovieSceneScriptingFloatChannel, keys [[-0.063,-12],[0,-16],[6,0]]}
          Bottom {type=MovieSceneScriptingFloatChannel, default=0}
         ]
        }]
      }]
    }
    Glow {
     id=#7
     possessed_class=Image
     tracks [{
       type=MovieSceneFloatTrack
       display_name=Render Opacity
       sections [{type=MovieSceneFloatSection, channels [None {type=MovieSceneScriptingFloatChannel, keys [[18,0],[45,1]]}]}]
      }]
    }
    SizeBox_0 {id=#10, possessed_class=SizeBox, tracks []}
   ]
   display_rate {numerator=60}
  }
 }
]
graphs [EventGraph {nodes [
   n1 = PlayAnimWithOptionalDelay!() -> (OutputDelegate, then, DelayDuration:double=~) {is=title K2Node_CustomEvent}
   > n2 = Branch(Condition="float <= float"(A=@n1.DelayDuration) -> (ReturnValue=~))
   > PlayAnim()
   n3 = SetVisibility(execute=@n2.else, self=Get Overlay_Root() -> (Overlay_Root:Overlay Object Reference), InVisibility=Hidden)
   n4 = PlayAnim() {is=title K2Node_CustomEvent}
   n5 = Set Timer by Event(execute=@n3.then, Delegate=@n4, Time=@n1.DelayDuration, bLooping=false)
   > Set DelayTimerHandle(DelayTimerHandle:Timer Handle Structure=@n5) -> (Output_Get:Timer Handle Structure)
   Event Destruct()
   n6 = Get AnimSkullDrop() -> (AnimSkullDrop:Widget Animation Object Reference)
   > Branch(Condition=IsAnimationPlaying(InAnimation=@n6))
   > StopAnimation(InAnimation=@n6)
   > Clear and Invalidate Timer by Handle(Handle=Get DelayTimerHandle() -> (DelayTimerHandle:Timer Handle Structure))
   Event Construct()
   > SetVisibility(self=Get Overlay_Root(), InVisibility=Hidden)
   PlayAnimation(execute=@n7.then, InAnimation=Get AnimSkullDrop())
   n7 = SetVisibility(execute=@n8.then, self=Get Overlay_Root(), InVisibility=HitTestInvisible)
   n8 = Clear and Invalidate Timer by Handle(execute=@n4.then, Handle=Get DelayTimerHandle())
  ]}]
variables [DelayTimerHandle {type=TimerHandle}]
widgets {
 info {inheritedWidgetCount=0, parentClass=&UserWidget, rootWidgetClass=&Overlay, widgetCount=8}
 widgets [
  Overlay_Root {widgetClassPath=&Overlay, bIsVariable=true, widget=Overlay_Root}
  SizeBox_0 {
   widgetClassPath=&SizeBox
   parent=Overlay_Root
   widget=SizeBox_0
   properties {bOverride_HeightOverride=true, bOverride_WidthOverride=true, heightOverride=24, widthOverride=24}
   slot {class=OverlaySlot, properties {horizontalAlignment=HAlign_Center, verticalAlignment=VAlign_Center}}
  }
  Overlay_29 {widgetClassPath=&Overlay, parent=SizeBox_0, widget=Overlay_29, slot {class=SizeBoxSlot}}
  Glow {
   widgetClassPath=&Image
   parent=Overlay_29
   widget=Glow
   properties {
    visibility=Collapsed
    brush {
     bIsDynamicallyLoaded=false
     drawAs=Image
     imageType=FullColor
     mirroring=NoMirror
     resourceName=None
     resourceObject=/BattleRoyaleGameUI/HUD/PlayerActionAlert/Assets/M_UI_RedRadialGlow
     tiling=NoTile
     imageSize=<32 32>
     margin=<ltrb 0 0 0 0>
     outlineSettings {
      bUseBrushTransparency=false
      roundingType=HalfHeightRadius
      width=0
      color {colorUseRule=UseColor_Specified, specifiedColor=<rgba 0 0 0 0>}
      cornerRadii=<0 0 0 0>
     }
     tintColor {colorUseRule=UseColor_Specified, specifiedColor=<rgba 0.0208 0 0.25 1>}
     uVRegion {bIsValid=false, max=<0 0>, min=<0 0>}
    }
    renderTransform {angle=0, scale=<2 2>, shear=<0 0>, translation=<0 0>}
   }
   slot {class=OverlaySlot, properties {horizontalAlignment=HAlign_Center, verticalAlignment=VAlign_Center}}
  }
  CenterSkull_Burst {
   widgetClassPath=&Image
   parent=Overlay_29
   widget=CenterSkull_Burst
   properties {
    visibility=SelfHitTestInvisible
    brush {
     bIsDynamicallyLoaded=false
     drawAs=Image
     imageType=FullColor
     mirroring=NoMirror
     resourceName=None
     resourceObject=/BattleRoyaleGameUI/HUD/PlayerActionAlert/Assets/M_UI_ElimIcon_StaticWhite
     tiling=NoTile
     imageSize=<32 32>
     margin=<ltrb 0 0 0 0>
     outlineSettings {
      bUseBrushTransparency=false
      roundingType=HalfHeightRadius
      width=0
      color {colorUseRule=UseColor_Specified, specifiedColor=<rgba 0 0 0 0>}
      cornerRadii=<0 0 0 0>
     }
     tintColor {colorUseRule=UseColor_Specified, specifiedColor=<rgba 1 1 1 1>}
     uVRegion {bIsValid=false, max=<0 0>, min=<0 0>}
    }
   }
   slot {class=OverlaySlot, properties {horizontalAlignment=HAlign_Center, verticalAlignment=VAlign_Center}}
  }
  CenterSkull_ChromaticL {
   widgetClassPath=&Image
   parent=Overlay_29
   widget=CenterSkull_ChromaticL
   properties {
    visibility=SelfHitTestInvisible
    brush {
     bIsDynamicallyLoaded=false
     drawAs=Image
     imageType=FullColor
     mirroring=NoMirror
     resourceName=None
     resourceObject=/BattleRoyaleGameUI/HUD/PlayerActionAlert/Assets/MI_UI_ElimIcon_StaticBlue
     tiling=NoTile
     imageSize=<32 32>
     margin=<ltrb 0 0 0 0>
     outlineSettings {
      bUseBrushTransparency=false
      roundingType=HalfHeightRadius
      width=0
      color {colorUseRule=UseColor_Specified, specifiedColor=<rgba 0 0 0 0>}
      cornerRadii=<0 0 0 0>
     }
     tintColor {colorUseRule=UseColor_Specified, specifiedColor=<rgba 1 1 1 1>}
     uVRegion {bIsValid=false, max=<0 0>, min=<0 0>}
    }
    renderTransform {angle=0, scale=<1.25 1.25>, shear=<0 0>, translation=<0 0>}
   }
   slot {class=OverlaySlot, properties {horizontalAlignment=HAlign_Center, verticalAlignment=VAlign_Center}}
  }
  CenterSkull_ChromaticR {
   widgetClassPath=&Image
   parent=Overlay_29
   widget=CenterSkull_ChromaticR
   properties {
    visibility=SelfHitTestInvisible
    brush {
     bIsDynamicallyLoaded=false
     drawAs=Image
     imageType=FullColor
     mirroring=NoMirror
     resourceName=None
     resourceObject=/BattleRoyaleGameUI/HUD/PlayerActionAlert/Assets/MI_UI_ElimIcon_StaticGreen
     tiling=NoTile
     imageSize=<32 32>
     margin=<ltrb 0 0 0 0>
     outlineSettings {
      bUseBrushTransparency=false
      roundingType=HalfHeightRadius
      width=0
      color {colorUseRule=UseColor_Specified, specifiedColor=<rgba 0 0 0 0>}
      cornerRadii=<0 0 0 0>
     }
     tintColor {colorUseRule=UseColor_Specified, specifiedColor=<rgba 1 1 1 1>}
     uVRegion {bIsValid=false, max=<0 0>, min=<0 0>}
    }
    renderTransform {angle=0, scale=<1.25 1.25>, shear=<0 0>, translation=<0 0>}
   }
   slot {class=OverlaySlot, properties {horizontalAlignment=HAlign_Center, verticalAlignment=VAlign_Center}}
  }
  CenterSkull {
   widgetClassPath=&Image
   parent=Overlay_29
   widget=CenterSkull
   properties {
    visibility=SelfHitTestInvisible
    brush {
     bIsDynamicallyLoaded=false
     drawAs=Image
     imageType=FullColor
     mirroring=NoMirror
     resourceName=None
     resourceObject=&M_UI_ElimIcon_StaticWhite
     tiling=NoTile
     imageSize=<32 32>
     margin=<ltrb 0 0 0 0>
     outlineSettings {
      bUseBrushTransparency=false
      roundingType=HalfHeightRadius
      width=0
      color {colorUseRule=UseColor_Specified, specifiedColor=<rgba 0 0 0 0>}
      cornerRadii=<0 0 0 0>
     }
     tintColor {colorUseRule=UseColor_Specified, specifiedColor=<rgba 1 1 1 1>}
     uVRegion {bIsValid=false, max=<0 0>, min=<0 0>}
    }
   }
   slot {class=OverlaySlot, properties {horizontalAlignment=HAlign_Center, verticalAlignment=VAlign_Center}}
  }
 ]
}
