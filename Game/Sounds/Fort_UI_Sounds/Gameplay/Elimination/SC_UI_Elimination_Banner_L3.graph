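# SoundCue for a level 3 elimination banner UI sound, judging by its name, in Fortnite's gameplay elimination set.
# A single wave player feeds the output with the UI_Notify_Elimination_03 wave, with no modulation or randomization.
# It routes to the critical in-game UI sound class with virtualization disabled.
# A concurrency override allows one instance at a time, stopping the oldest when retriggered.
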SoundCue /Game/Sounds/Fort_UI_Sounds/Gameplay/Elimination/SC_UI_Elimination_Banner_L3
root=WavePlayer_0
nodes [{id=WavePlayer_0, properties {soundWaveAssetPtr=&/UI_Notify_Elimination_03}}]
properties {
 bOverrideConcurrency=true
 maxDistance=2097152
 priority=0
 soundClassObject=&Fort_UI_InGame_Critical
 virtualizationMode=Disabled
 volumeMultiplier=1
 concurrencyOverrides {maxCount=1, resolutionRule=StopOldest}
}
