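# Battle Royale HUD elimination alert, derived from ActionAlertBase, showing the eliminated player's name, rank icon and portrait with a kill-streak banner.
# HandleAlertData clamps NumKills, limited to one unless a console variable enables multi-eliminations, to drive skull count, tiered sound and banner text.
# Three or more kills play the multi-elimination FX animation instead of the single one.
# SetNumSkulls recycles skull widgets through a pool, staggering their animations; touch compact UI gets adjusted styling.
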WidgetBlueprint /BattleRoyaleGameUI/HUD/PlayerActionAlert/WBP_BRGameUI_EliminationAlert
animations [
 AnimShowAlert_Base {
  binding_count=4
  length_seconds=0.5
  object=&self:AnimShowAlert_Base
  sequence {
   bindings [
    Backing {
     id=#1
     possessed_class=Image
     tracks [
      {
       type=MovieScene2DTransformTrack
       sections [{
         type=MovieScene2DTransformSection
         channels [
          Scale.X {type=MovieSceneScriptingFloatChannel, keys [[3,5],[12,1]]}
          Scale.Y {type=MovieSceneScriptingFloatChannel, keys [[6,0.125],[12,1]]}
          Shear.X {type=MovieSceneScriptingFloatChannel}
          Shear.Y {type=MovieSceneScriptingFloatChannel}
         ]
        }]
      }
      {
       type=MovieSceneWidgetMaterialTrack
       sections [{
         type=MovieSceneParameterSection
         channels [
          Anim_Opacity {type=MovieSceneScriptingFloatChannel, keys [[0,0],[1,1]]}
          Anim_Color {type=MovieSceneScriptingFloatChannel, keys [[15,1],[24,0]]}
         ]
        }]
      }
     ]
    }
    Image_Bracket_Right {
     id=#2
     possessed_class=Image
     tracks [
      {
       type=MovieScene2DTransformTrack
       sections [{
         type=MovieScene2DTransformSection
         channels [
          Translation.X {type=MovieSceneScriptingFloatChannel, keys [[3,192],[12,0]]}
          Translation.Y {type=MovieSceneScriptingFloatChannel, const [0,1]}
          Scale.X {type=MovieSceneScriptingFloatChannel, keys [[3,2],[12,0.9],[17,1]]}
          Scale.Y {type=MovieSceneScriptingFloatChannel, keys [[3,2],[12,1]]}
         ]
        }]
      }
      {
       type=MovieSceneWidgetMaterialTrack
       sections [{
         type=MovieSceneParameterSection
         channels [
          Anim_Opacity {type=MovieSceneScriptingFloatChannel, keys [[2,0],[4,1]]}
          Anim_Stretch {type=MovieSceneScriptingFloatChannel, keys [[9,0],[12,0.75],[24,0]]}
         ]
        }]
      }
     ]
    }
    Image_Bracket_Left {
     id=#3
     possessed_class=Image
     tracks [
      {
       type=MovieScene2DTransformTrack
       sections [{
         type=MovieScene2DTransformSection
         channels [
          Translation.X {type=MovieSceneScriptingFloatChannel, keys [[3,-192],[12,0]]}
          Translation.Y {type=MovieSceneScriptingFloatChannel, const [0,1]}
          Scale.X {type=MovieSceneScriptingFloatChannel, keys [[3,2],[12,0.9],[17,1]]}
          Scale.Y {type=MovieSceneScriptingFloatChannel, keys [[3,2],[12,1]]}
         ]
        }]
      }
      {
       type=MovieSceneWidgetMaterialTrack
       sections [{
         type=MovieSceneParameterSection
         channels [
          Anim_Opacity {type=MovieSceneScriptingFloatChannel, keys [[2,0],[4,1]]}
          Anim_Stretch {type=MovieSceneScriptingFloatChannel, keys [[9,0],[12,0.75],[24,0]]}
         ]
        }]
      }
     ]
    }
    Overlay_Elimination {
     id=#4
     possessed_class=Overlay
     tracks [{
       type=MovieScene2DTransformTrack
       sections [{
         type=MovieScene2DTransformSection
         channels [
          Scale.X {type=MovieSceneScriptingFloatChannel, keys [[9,1],[12,0.5],[30,1]]}
          Scale.Y {type=MovieSceneScriptingFloatChannel, keys [[9,1],[12,0.5],[30,1]]}
         ]
        }]
      }]
    }
   ]
   display_rate {numerator=60}
  }
 }
 Anim_MultiElimFX {
  binding_count=17
  length_seconds=2
  object=&self:Anim_MultiElimFX
  sequence {bindings [
    Image_Bracket_Right {
     id=#5
     possessed_class=Image
     tracks [{
       type=MovieSceneWidgetMaterialTrack
       sections [{type=MovieSceneParameterSection, channels [Anim_Color {type=MovieSceneScriptingFloatChannel, keys [[3.6667,1],[4,0],[6,0],[20,1]]}]}]
      }]
    }
    Image_CornerArrow_BottomRight {
     id=#6
     possessed_class=Image
     tracks [
      {
       type=MovieScene2DTransformTrack
       sections [{
         type=MovieScene2DTransformSection
         channels [
          Translation.X {type=MovieSceneScriptingFloatChannel, keys [[4,0],[23,48]]}
          Translation.Y {type=MovieSceneScriptingFloatChannel, keys [[4,0],[23,48]]}
         ]
        }]
      }
      {
       type=MovieSceneWidgetMaterialTrack
       sections [{
         type=MovieSceneParameterSection
         channels [
          Anim_Opacity {type=MovieSceneScriptingFloatChannel, keys [[3,0],[4,1],[21,1],[23,0]]}
          Anim_Scale {type=MovieSceneScriptingFloatChannel, keys [[5,0],[23,1]]}
         ]
        }]
      }
     ]
    }
    Image_LargeRedParticles {
     id=#7
     possessed_class=Image
     tracks [
      {
       type=MovieScene2DTransformTrack
       sections [{
         type=MovieScene2DTransformSection
         channels [
          Scale.X {type=MovieSceneScriptingFloatChannel, keys [[4,0.25],[22.6667,2]]}
          Scale.Y {type=MovieSceneScriptingFloatChannel, keys [[4,0.25],[22.6667,2]]}
         ]
        }]
      }
      {
       type=MovieSceneWidgetMaterialTrack
       sections [{
         type=MovieSceneParameterSection
         channels [
          Anim_CollapseTiles {type=MovieSceneScriptingFloatChannel, keys [[4,1],[4.6667,0],[39,1]]}
          Anim_Opacity {type=MovieSceneScriptingFloatChannel, keys [[4,0],[4.6667,1],[37.3333,1],[39,0]]}
          IntensityMultiplier {type=MovieSceneScriptingFloatChannel, thinned_from=5, keys [[4,1],[4.6667,4],[12.6667,4],[27.6667,1]]}
         ]
        }]
      }
     ]
    }
    Image_CornerArrow_TopLeft {
     id=#8
     possessed_class=Image
     tracks [
      {
       type=MovieSceneWidgetMaterialTrack
       sections [{
         type=MovieSceneParameterSection
         channels [
          Anim_Opacity {type=MovieSceneScriptingFloatChannel, keys [[3,0],[4,1],[21,1],[23,0]]}
          Anim_Scale {type=MovieSceneScriptingFloatChannel, keys [[5,0],[23,1]]}
         ]
        }]
      }
      {
       type=MovieScene2DTransformTrack
       sections [{
         type=MovieScene2DTransformSection
         channels [
          Translation.X {type=MovieSceneScriptingFloatChannel, keys [[4,0],[23,-64]]}
          Translation.Y {type=MovieSceneScriptingFloatChannel, keys [[4,0],[23,-64]]}
         ]
        }]
      }
     ]
    }
    Image_CornerArrow_TopRight {
     id=#9
     possessed_class=Image
     tracks [
      {
       type=MovieScene2DTransformTrack
       sections [{
         type=MovieScene2DTransformSection
         channels [
          Translation.X {type=MovieSceneScriptingFloatChannel, keys [[4,0],[23,64]]}
          Translation.Y {type=MovieSceneScriptingFloatChannel, keys [[4,0],[23,-64]]}
         ]
        }]
      }
      {
       type=MovieSceneWidgetMaterialTrack
       sections [{
         type=MovieSceneParameterSection
         channels [
          Anim_Opacity {type=MovieSceneScriptingFloatChannel, keys [[3,0],[4,1],[21,1],[23,0]]}
          Anim_Scale {type=MovieSceneScriptingFloatChannel, keys [[5,0],[23,1]]}
         ]
        }]
      }
     ]
    }
    Image_DarkParticles {
     id=#10
     possessed_class=Image
     tracks [
      {
       type=MovieSceneWidgetMaterialTrack
       sections [{
         type=MovieSceneParameterSection
         channels [
          Anim_CollapseTiles {type=MovieSceneScriptingFloatChannel, keys [[5,1],[6,0],[11,0],[25,1]]}
          Anim_Opacity {type=MovieSceneScriptingFloatChannel, keys [[5,0],[6,1],[23,1],[25,0]]}
          IntensityMultiplier {type=MovieSceneScriptingFloatChannel, keys [[5,1],[7.3333,8],[15,1]]}
         ]
        }]
      }
      {
       type=MovieScene2DTransformTrack
       sections [{
         type=MovieScene2DTransformSection
         channels [
          Scale.X {type=MovieSceneScriptingFloatChannel, keys [[5,0.25],[24.6667,2]]}
          Scale.Y {type=MovieSceneScriptingFloatChannel, keys [[5,0.25],[24.6667,2]]}
         ]
        }]
      }
     ]
    }
    Image_Bolt_Right {
     id=#11
     possessed_class=Image
     tracks [
      {
       type=MovieSceneWidgetMaterialTrack
       sections [{
         type=MovieSceneParameterSection
         channels [
          Anim_Color {type=MovieSceneScriptingFloatChannel, keys [[6,1],[30,0]]}
          Anim_Opacity {type=MovieSceneScriptingFloatChannel, keys [[5,0],[6,1],[30,0.5]]}
         ]
        }]
      }
      {
       type=MovieScene2DTransformTrack
       sections [{
         type=MovieScene2DTransformSection
         channels [
          Translation.X {type=MovieSceneScriptingFloatChannel, const [0,1]}
          Translation.Y {type=MovieSceneScriptingFloatChannel, const [0,1]}
          Scale.X {type=MovieSceneScriptingFloatChannel, default=0.5, keys [[5,0],[7,1.5],[30,0.5]]}
          Scale.Y {type=MovieSceneScriptingFloatChannel, keys [[5,0.5],[7,1.25],[19,1]]}
         ]
        }]
      }
     ]
    }
    Image_Bracket_Left {
     id=#12
     possessed_class=Image
     tracks [{
       type=MovieSceneWidgetMaterialTrack
       sections [{type=MovieSceneParameterSection, channels [Anim_Color {type=MovieSceneScriptingFloatChannel, keys [[3.6667,1],[4,0],[6,0],[20,1]]}]}]
      }]
    }
    Image_Bolt_Left {
     id=#13
     possessed_class=Image
     tracks [
      {
       type=MovieScene2DTransformTrack
       sections [{
         type=MovieScene2DTransformSection
         channels [
          Translation.X {type=MovieSceneScriptingFloatChannel, const [0,1]}
          Translation.Y {type=MovieSceneScriptingFloatChannel, const [0,1]}
          Scale.X {type=MovieSceneScriptingFloatChannel, keys [[5,0],[7,1.5],[30,0.5]]}
          Scale.Y {type=MovieSceneScriptingFloatChannel, keys [[5,0.5],[7,1.25],[19,1]]}
         ]
        }]
      }
      {
       type=MovieSceneWidgetMaterialTrack
       sections [{
         type=MovieSceneParameterSection
         channels [
          Anim_Color {type=MovieSceneScriptingFloatChannel, keys [[6,1],[30,0]]}
          Anim_Opacity {type=MovieSceneScriptingFloatChannel, keys [[5,0],[6,1],[30,0.5]]}
         ]
        }]
      }
     ]
    }
    Overlay_CornerArrows {
     id=#14
     possessed_class=Overlay
     tracks [{
       type=MovieSceneByteTrack
       display_name=Visibility
       sections [{type=MovieSceneByteSection, channels [None {type=MovieSceneScriptingByteChannel, keys [[0,4],[40,2]]}]}]
      }]
    }
    Backing {
     id=#15
     possessed_class=Image
     tracks [{
       type=MovieSceneWidgetMaterialTrack
       sections [{type=MovieSceneParameterSection, channels [Anim_Color {type=MovieSceneScriptingFloatChannel, keys [[3,0],[4,1],[8,0]]}]}]
      }]
    }
    Overlay_Elimination {
     id=#16
     possessed_class=Overlay
     tracks [{
       type=MovieScene2DTransformTrack
       sections [{
         type=MovieScene2DTransformSection
         channels [
          Scale.X {type=MovieSceneScriptingFloatChannel, keys [[3,1],[3.3333,0.75],[8.3333,1]]}
          Scale.Y {type=MovieSceneScriptingFloatChannel, keys [[2.3333,1],[3.3333,0.75],[8.3333,1]]}
         ]
        }]
      }]
    }
    Image_LargeSkull {
     id=#17
     possessed_class=Image
     tracks [
      {
       type=MovieSceneWidgetMaterialTrack
       sections [{
         type=MovieSceneParameterSection
         channels [
          Anim_Color {type=MovieSceneScriptingFloatChannel, default=1, const [0,1]}
          AdditiveRatio {type=MovieSceneScriptingFloatChannel, const [0,1]}
          Anim_Alpha {type=MovieSceneScriptingFloatChannel, keys [[4,0],[8,1],[10,1],[40,0.5]]}
          Anim_Smoothing {type=MovieSceneScriptingFloatChannel, keys [[4,0.5],[10,0]]}
         ]
        }]
      }
      {
       type=MovieScene2DTransformTrack
       sections [{
         type=MovieScene2DTransformSection
         channels [
          Scale.X {type=MovieSceneScriptingFloatChannel, keys [[4,0],[15,1]]}
          Scale.Y {type=MovieSceneScriptingFloatChannel, keys [[4,0],[15,1]]}
         ]
        }]
      }
      {
       type=MovieSceneByteTrack
       display_name=Visibility
       sections [{type=MovieSceneByteSection, channels [None {type=MovieSceneScriptingByteChannel, const [4,1]}]}]
      }
     ]
    }
    Image_CornerArrow_BottomLeft {
     id=#18
     possessed_class=Image
     tracks [
      {
       type=MovieScene2DTransformTrack
       sections [{
         type=MovieScene2DTransformSection
         channels [
          Translation.X {type=MovieSceneScriptingFloatChannel, keys [[4,0],[23,-48]]}
          Translation.Y {type=MovieSceneScriptingFloatChannel, keys [[4,0],[23,48]]}
         ]
        }]
      }
      {
       type=MovieSceneWidgetMaterialTrack
       sections [{
         type=MovieSceneParameterSection
         channels [
          Anim_Opacity {type=MovieSceneScriptingFloatChannel, keys [[3,0],[4,1],[21,1],[23,0]]}
          Anim_Scale {type=MovieSceneScriptingFloatChannel, keys [[5,0],[23,1]]}
         ]
        }]
      }
     ]
    }
    Image_BurstSkull {
     id=#19
     possessed_class=Image
     tracks [
      {
       type=MovieSceneWidgetMaterialTrack
       sections [{
         type=MovieSceneParameterSection
         channels [
          Anim_Alpha {type=MovieSceneScriptingFloatChannel, keys [[0,0],[5,0],[6,1],[40,0]]}
          Anim_Smoothing {type=MovieSceneScriptingFloatChannel, keys [[5,0],[29,0.5]]}
         ]
        }]
      }
      {
       type=MovieScene2DTransformTrack
       sections [{
         type=MovieScene2DTransformSection
         channels [
          Scale.X {type=MovieSceneScriptingFloatChannel, keys [[5,0.5],[40,2]]}
          Scale.Y {type=MovieSceneScriptingFloatChannel, keys [[5,0.5],[40,2]]}
         ]
        }]
      }
     ]
    }
    ParticlesAndSkullGlow {
     id=#20
     possessed_class=Overlay
     tracks [{
       type=MovieSceneByteTrack
       display_name=Visibility
       sections [{type=MovieSceneByteSection, channels [None {type=MovieSceneScriptingByteChannel, keys [[0,4],[40,2]]}]}]
      }]
    }
    Glow {
     id=#21
     possessed_class=Image
     tracks [{
       type=MovieSceneWidgetMaterialTrack
       sections [{type=MovieSceneParameterSection, channels [Anim_Opacity {type=MovieSceneScriptingFloatChannel, keys [[0,0],[4,0],[5.3333,1],[20.6667,0]]}]}]
      }]
    }
   ]}
 }
 Anim_SingleElimFX {
  binding_count=6
  length_seconds=1.5
  object=&self:Anim_SingleElimFX
  sequence {bindings [
    Image_Bracket_Left {
     id=#22
     possessed_class=Image
     tracks [{
       type=MovieSceneWidgetMaterialTrack
       sections [{type=MovieSceneParameterSection, channels [Anim_Color {type=MovieSceneScriptingFloatChannel, keys [[0,1],[3.6667,1],[4,0],[6,0],[15,1]]}]}]
      }]
    }
    Image_Bracket_Right {
     id=#23
     possessed_class=Image
     tracks [{
       type=MovieSceneWidgetMaterialTrack
       sections [{type=MovieSceneParameterSection, channels [Anim_Color {type=MovieSceneScriptingFloatChannel, keys [[3.6667,1],[4,0],[6,0],[15,1]]}]}]
      }]
    }
    Glow {
     id=#24
     possessed_class=Image
     tracks [{
       type=MovieSceneWidgetMaterialTrack
       sections [{type=MovieSceneParameterSection, channels [Anim_Opacity {type=MovieSceneScriptingFloatChannel, keys [[0,0],[4,0],[5.3333,1],[20.6667,0]]}]}]
      }]
    }
    Image_Bolt_Right {
     id=#25
     possessed_class=Image
     tracks [
      {
       type=MovieSceneWidgetMaterialTrack
       sections [{
         type=MovieSceneParameterSection
         channels [
          Anim_Color {type=MovieSceneScriptingFloatChannel, keys [[6,1],[30,0]]}
          Anim_Opacity {type=MovieSceneScriptingFloatChannel, keys [[5,0],[6,1],[30,0]]}
         ]
        }]
      }
      {
       type=MovieScene2DTransformTrack
       sections [{
         type=MovieScene2DTransformSection
         channels [
          Translation.X {type=MovieSceneScriptingFloatChannel, const [0,1]}
          Translation.Y {type=MovieSceneScriptingFloatChannel, const [0,1]}
          Scale.X {type=MovieSceneScriptingFloatChannel, default=0.5, keys [[5,0],[7,1.5],[30,0.5]]}
          Scale.Y {type=MovieSceneScriptingFloatChannel, keys [[5,0.5],[7,1.25],[19,1]]}
         ]
        }]
      }
     ]
    }
    Image_Bolt_Left {
     id=#26
     possessed_class=Image
     tracks [
      {
       type=MovieScene2DTransformTrack
       sections [{
         type=MovieScene2DTransformSection
         channels [
          Translation.X {type=MovieSceneScriptingFloatChannel, const [0,1]}
          Translation.Y {type=MovieSceneScriptingFloatChannel, const [0,1]}
          Scale.X {type=MovieSceneScriptingFloatChannel, keys [[5,0],[7,1.5],[30,0.5]]}
          Scale.Y {type=MovieSceneScriptingFloatChannel, keys [[5,0.5],[7,1.25],[19,1]]}
         ]
        }]
      }
      {
       type=MovieSceneWidgetMaterialTrack
       sections [{
         type=MovieSceneParameterSection
         channels [
          Anim_Color {type=MovieSceneScriptingFloatChannel, keys [[6,1],[30,0]]}
          Anim_Opacity {type=MovieSceneScriptingFloatChannel, keys [[5,0],[6,1],[30,0]]}
         ]
        }]
      }
     ]
    }
    Image_LargeSkull {
     id=#27
     possessed_class=Image
     tracks [{
       type=MovieSceneByteTrack
       display_name=Visibility
       sections [{type=MovieSceneByteSection, channels [None {type=MovieSceneScriptingByteChannel, const [2,1]}]}]
      }]
    }
   ]}
 }
 AnimRefreshInfo {
  binding_count=4
  length_seconds=0.75
  object=&self:AnimRefreshInfo
  sequence {
   bindings [
    Text_Background {
     id=#28
     possessed_class=CommonTextBlock
     tracks [
      {
       type=MovieScene2DTransformTrack
       sections [{
         type=MovieScene2DTransformSection
         channels [
          Translation.X {type=MovieSceneScriptingFloatChannel, default=0}
          Translation.Y {type=MovieSceneScriptingFloatChannel, keys [[12,48],[45,0]]}
          Scale.X {type=MovieSceneScriptingFloatChannel, keys [[12,0.5],[45,1]]}
          Scale.Y {type=MovieSceneScriptingFloatChannel, keys [[12,0.5],[45,1]]}
         ]
        }]
      }
      {
       type=MovieSceneWidgetMaterialTrack
       display_name=Font.FontMaterial
       sections [{
         type=MovieSceneParameterSection
         channels [
          Max_Alpha {type=MovieSceneScriptingFloatChannel, keys [[12,0],[21,1],[45,1]]}
          Min_Alpha {type=MovieSceneScriptingFloatChannel, keys [[12,0],[21,1],[45,0.75]]}
         ]
        }]
      }
      {
       type=MovieSceneWidgetMaterialTrack
       display_name=Font.Outline Settings.OutlineMaterial
       sections [{type=MovieSceneParameterSection, channels [AnimOpacity {type=MovieSceneScriptingFloatChannel, keys [[12,0],[45,1]]}]}]
      }
     ]
    }
    LazyImage_Habanero_RankIcon {
     id=#29
     possessed_class=FortLazyImage
     tracks [
      {
       type=MovieSceneFloatTrack
       display_name=Render Opacity
       sections [{type=MovieSceneFloatSection, channels [None {type=MovieSceneScriptingFloatChannel, keys [[15,0],[18,1]]}]}]
      }
      {
       type=MovieScene2DTransformTrack
       sections [{
         type=MovieScene2DTransformSection
         channels [
          Translation.X {type=MovieSceneScriptingFloatChannel, const [0,1]}
          Translation.Y {type=MovieSceneScriptingFloatChannel, const [0,1]}
          Scale.X {type=MovieSceneScriptingFloatChannel, keys [[15,0],[21,1.25],[27,0.9],[42,1]]}
          Scale.Y {type=MovieSceneScriptingFloatChannel, keys [[15,0],[21,1.25],[27,0.9],[42,1]]}
          Shear.X {type=MovieSceneScriptingFloatChannel, const [0,1]}
          Shear.Y {type=MovieSceneScriptingFloatChannel, const [0,1]}
         ]
        }]
      }
     ]
    }
    Text_PlayerName {
     id=#30
     possessed_class=CommonTextBlock
     tracks [
      {
       type=MovieSceneFloatTrack
       display_name=Render Opacity
       sections [{type=MovieSceneFloatSection, channels [None {type=MovieSceneScriptingFloatChannel, keys [[15,0],[17,1]]}]}]
      }
      {
       type=MovieScene2DTransformTrack
       sections [{
         type=MovieScene2DTransformSection
         channels [
          Scale.X {type=MovieSceneScriptingFloatChannel, keys [[15,0.75],[21,1.125],[34,1]]}
          Scale.Y {type=MovieSceneScriptingFloatChannel, keys [[15,0.75],[21,1.125],[34,1]]}
         ]
        }]
      }
     ]
    }
    LazyImage_PlayerPortrait {
     id=#31
     possessed_class=FortLazyImage
     tracks [{
       type=MovieSceneWidgetMaterialTrack
       sections [{
         type=MovieSceneParameterSection
         channels [
          Anim_Translate {type=MovieSceneScriptingFloatChannel, keys [[9,0],[21,1]]}
          Anim_Distortion {type=MovieSceneScriptingFloatChannel, keys [[9,0],[45,1]]}
          Anim_Opacity {type=MovieSceneScriptingFloatChannel, keys [[9,0],[12,1]]}
         ]
        }]
      }]
    }
   ]
   display_rate {numerator=60}
  }
 }
]
graphs [
 EventGraph {nodes [
   Event PreConstruct()
   > SetMobileStyle() {is=title K2Node_CallFunction}
  ]}
 HandleAlertData {nodes [
   n1 = HandleAlertData!() -> (then, AlertData:Athena Player Action Alert Structure) {is=title K2Node_FunctionEntry}
   n2 = Break Athena Player Action Alert!(AthenaPlayerActionAlert:"Athena Player Action Alert Structure (by ref)"=@n1) -> (
    AlertType:EAthenaPlayerActionAlert Enum
    DetailText:Text
    RankedProgress:Ranked Progress Replicated Data Structure
    OtherPlayerState:Fort Player State Athena Weak object
    NumKills:int
   ) {is=title K2Node_BreakStruct}
   > "SetText (Text)"(self=Get Text_PlayerName() -> (Text_PlayerName:Common Text Object Reference), InText=@n2.DetailText)
   > SetRankedProgress!(execute, self, RankedProgress:Ranked Progress Replicated Data Structure=@n2.RankedProgress) -> (then)
   > SetPlayerPortraitFromPlayerState!(execute, self, PlayerState:Fort Player State Athena Object Reference=@n2.OtherPlayerState) -> (then)
   > n3 = GetConsoleVariableBoolValue!(execute, VariableName:String=BattleRoyaleGameUI.EliminationAlert.MultipleEliminations.Enabled) -> (then, ReturnValue:bool)
   > n4 = Branch(Condition=@n3)
   n5 = "Min (Integer)"(A=SelectInt(A=Get MaxSkullsToDisplay() -> (MaxSkullsToDisplay:int=0), B=1, bPickA=@n3), B="Max (Integer)"(A=@n2.NumKills, B=1)) -> (ReturnValue=0)
   SetNumSkulls!(execute=[@n7.then, @n4.then], self, NumSkulls:int=@n5) -> (then)
   > PlayAlertSound!(execute, self, Level:int=@n5) -> (then)
   n6 = int - int(A=@n5)
   > "SetText (Text)"(
    self=Get Text_Background() -> (Text_Background:Common Text Object Reference)
    InText=SelectText(
     A="LOCTEXT(\"Multi Elim!\")"
     B=Select!(
      Option 0:Text="LOCTEXT(\"Elimination!\")"
      Option 1:Text="LOCTEXT(\"Double Elim!\")"
      Option 2:Text="LOCTEXT(\"Triple Elim!\")"
      Option 3:Text="LOCTEXT(\"Quad Elim!\")"
      Index:int=@n6
     ) -> (ReturnValue:Text)
     bPickA="integer >= integer"(A=@n6, B=4)
    )
   )
   > PlayAnimation(InAnimation=Select!(
     Option 0:Widget Animation Object Reference=Get Anim_SingleElimFX() -> (Anim_SingleElimFX:Widget Animation Object Reference)
     Option 1:Widget Animation Object Reference=Get Anim_MultiElimFX() -> (Anim_MultiElimFX:Widget Animation Object Reference)
     Index:bool="integer >= integer"(A=@n5, B=3)
    ) -> (ReturnValue:Widget Animation Object Reference))
   > PlayAnimation(InAnimation=Get AnimRefreshInfo() -> (AnimRefreshInfo:Widget Animation Object Reference))
   n7 = ResetSkulls(execute=@n4.else)
  ]}
 HandleShowAlert {nodes [
   HandleShowAlert() {is=title K2Node_FunctionEntry}
   > n1 = Branch(Condition=Get bAlertShown() -> (bAlertShown:bool))
   Parent: HandleShowAlert(execute=@n1.else) {is=title K2Node_CallParentFunction}
   > n2 = GetConsoleVariableFloatValue!(execute, VariableName:String=BattleRoyaleGameUI.EliminationAlert.ScaleMultiplier) -> (then, ReturnValue:float)
   > SetRenderScale!(execute, self:Widget Object Reference, Scale:Vector 2D Structure=To Vector2D!(InDouble:double=@n2) -> (ReturnValue:Vector 2D Structure)) -> (then)
   > ResetSkulls()
   > PlayAnimation(InAnimation=Get AnimShowAlert_Base() -> (AnimShowAlert_Base:Widget Animation Object Reference))
  ]}
 PlayAlertSound {nodes [
   n1 = PlayAlertSound!() -> (then, Level:int) {is=title K2Node_FunctionEntry}
   n2 = Select!(
    Option 0:Sound Base Object Reference
    Option 1:Sound Base Object Reference=/Game/Sounds/Fort_UI_Sounds/Gameplay/Elimination/SC_UI_Elimination_Banner
    Option 2:Sound Base Object Reference=/Game/Sounds/Fort_UI_Sounds/Gameplay/Elimination/SC_UI_Elimination_Banner_L2
    Option 3:Sound Base Object Reference=/Game/Sounds/Fort_UI_Sounds/Gameplay/Elimination/SC_UI_Elimination_Banner_L3
    Index:int="Min (Integer)"(A=@n1, B=3) -> (ReturnValue=0)
   ) -> (ReturnValue:Sound Base Object Reference)
   > Is Valid(InputObject=@n2)
   > PlaySound2D(Sound=@n2)
  ]}
 ResetSkulls {nodes [
   ResetSkulls() {is=title K2Node_FunctionEntry}
   > SetNumSkulls(NumSkulls=0)
  ]}
 SetMobileStyle {nodes [
   SetMobileStyle() {is=title K2Node_FunctionEntry}
   n1 = GetOwningPlayer()
   > Branch(Condition=AND Boolean(
     A=IsUsingCompactUI()
     B=IsUsingInputType!(InputType:ECommonInputType Enum=Touch, PlayerController:Player Controller Object Reference=@n1) -> (ReturnValue:bool)
    ))
   > SetRenderScale(self=Get Overlay_FX_Elements() -> (Overlay_FX_Elements:Overlay Object Reference), Scale=(X=0.85,Y=0.85))
   > SetPadding!(
    execute
    self:Overlay Slot Object Reference=SlotAsOverlaySlot(Widget=Get Text_PlayerName())
    InPadding_Left:float=0
    InPadding_Top:float=7
    InPadding_Right:float=0
    InPadding_Bottom:float=0
   ) -> (then)
  ]}
 SetNumSkulls {nodes [
   n1 = SetNumSkulls!() -> (then, NumSkulls:int) {is=title K2Node_FunctionEntry}
   n2 = Length!(TargetSet:Set of WBP BRGame UI Skull Anim Object References=Get SkullWidgets() -> (SkullWidgets:Set of WBP BRGame UI Skull Anim Object References)) -> (ReturnValue:int=0) {is=title K2Node_CallFunction}
   n3 = "integer < integer"(A=@n1, B=@n2)
   > n4 = Branch(Condition=@n3)
   > n5 = While Loop!(execute:Exec, Condition:bool=@n3) -> (LoopBody:Exec, Completed:Exec) {is=title K2Node_MacroInstance}
   n6 = "integer > integer"(A=@n1, B=@n2)
   Branch(execute=[@n4.else, @n5.Completed], Condition=@n6)
   > While Loop(Condition=@n6)
   n7 = Get SkullWidgetPool() -> (SkullWidgetPool:Set of WBP BRGame UI Skull Anim Object References)
   > n8 = Branch(Condition=Is Empty!(TargetSet:Set of WBP BRGame UI Skull Anim Object References=@n7) -> (ReturnValue:bool) {is=title K2Node_CallFunction})
   > n9 = Create Widget!(
    execute:Exec
    Class:User Widget Class Reference=&/WBP_BRGameUI_SkullAnim.WBP_BRGameUI_SkullAnim_C
    OwningPlayer:Player Controller Object Reference
   ) -> (then:Exec, ReturnValue:WBP BRGame UI Skull Anim Object Reference)
   > n10 = Set CacheNewSkullWidget(CacheNewSkullWidget:WBP BRGame UI Skull Anim Object Reference=@n9) -> (Output_Get:WBP BRGame UI Skull Anim Object Reference)
   > n11 = Add!(
    execute
    TargetSet:Set of WBP BRGame UI Skull Anim Object References=Get SkullWidgets()
    NewItem:"WBP BRGame UI Skull Anim Object Reference (by ref)"=@n10
   ) -> (then) {is=title K2Node_CallFunction}
   n12 = Get CacheNewSkullWidget() -> (CacheNewSkullWidget:WBP BRGame UI Skull Anim Object Reference)
   AddChild(execute=[@n14.then, @n11.then], self=Get HBox_SkullContainer() -> (HBox_SkullContainer:Horizontal Box Object Reference), Content=@n12)
   > PlayAnimWithOptionalDelay!(
    execute
    self:WBP BRGame UI Skull Anim Object Reference=@n12
    DelayDuration:double=float * float(A:int=Get NewSkullCounter() -> (NewSkullCounter:int=0), B=Get SkullAnimStaggerTime() -> (SkullAnimStaggerTime:double)) {is=title K2Node_PromotableOperator}
   ) -> (then)
   > Increment Int(Value=Get NewSkullCounter() -> (NewSkullCounter=0))
   n13 = Set CacheNewSkullWidget(
    execute=@n8.else
    CacheNewSkullWidget=Get Item By Index!(
     TargetSet:Set of WBP BRGame UI Skull Anim Object References=@n7
     Index:int=Last Index!(TargetSet:Set of WBP BRGame UI Skull Anim Object References=@n7) -> (ReturnValue:int=0) {is=title K2Node_CallFunction}
    ) -> (Item:WBP BRGame UI Skull Anim Object Reference) {is=title K2Node_CallFunction}
   )
   > Add(TargetSet=Get SkullWidgets(), NewItem=@n13)
   > n14 = Remove!(execute, TargetSet:Set of WBP BRGame UI Skull Anim Object References=@n7, Item:"WBP BRGame UI Skull Anim Object Reference (by ref)"=@n13) -> (then, ReturnValue:bool)
   n15 = Get SkullWidgets()
   n16 = Get Item By Index(TargetSet=@n15, Index=Last Index(TargetSet=@n15) -> (ReturnValue=0))
   RemoveFromParent(execute=@n5.LoopBody, self=@n16)
   > Add(TargetSet=Get SkullWidgetPool(), NewItem=@n16)
   > Remove(TargetSet=@n15, Item=@n16)
  ]}
 SetPlayerPortraitFromPlayerState {nodes [
   n1 = SetPlayerPortraitFromPlayerState!() -> (then, PlayerState:Fort Player State Athena Object Reference) {is=title K2Node_FunctionEntry}
   > n2 = Is Valid(InputObject=@n1)
   n3 = GetCurrentPawn!(self:Fort Player State Object Reference=@n1) -> (ReturnValue:Fort Player Pawn Object Reference)
   > n4 = Is Valid(InputObject=@n3)
   n5 = Get LazyImage_PlayerPortrait() -> (LazyImage_PlayerPortrait:Fort Lazy Image Object Reference)
   > SetVisibility(self=@n5, InVisibility=HitTestInvisible)
   > SetBrushFromSoftTexture!(
    execute
    self:Image Object Reference=Get LazyImage_PlayerPortrait()
    SoftTexture:Texture 2D Soft Object Reference=GetSmallPreviewImage!(ItemDefinition:Item Definition Base Object Reference=GetItemDefinitionFromBaseLoadoutWithActor!(Actor:Actor Object Reference=@n3, LoadoutSlot:Gameplay Tag Structure=(TagName="LoadoutSlot.Character")) -> (ReturnValue:Athena Cosmetic Item Definition Object Reference)) -> (ReturnValue:Texture 2D Soft Object Reference)
    bMatchSize:bool=false
   ) -> (then)
   SetVisibility(execute=[@n4.Is Not Valid, @n2.Is Not Valid], self=@n5, InVisibility=Collapsed)
  ]}
 SetRankedProgress {nodes [
   n1 = SetRankedProgress!() -> (then, RankedProgress:Ranked Progress Replicated Data Structure) {is=title K2Node_FunctionEntry}
   n2 = Break Ranked Progress Replicated Data!(RankedProgressReplicatedData:"Ranked Progress Replicated Data Structure (by ref)"=@n1) -> (RankType:String, Rank:int, HighestRank:int, CurrentMMR:int, CurrentPosition:int) {is=title K2Node_BreakStruct}
   n3 = GetHabaneroDisplayDataOverride!(
    InDisplayData:Fort Habanero Display Data Object Reference=Get RankData() -> (RankData:Fort Habanero Display Data Object Reference)
    InOverrideData:Fort Compete UIHabanero Tier Override List Object Reference=Get RankInfoOverrides() -> (RankInfoOverrides:Fort Compete UIHabanero Tier Override List Object Reference)
    InKey:"Fort Compete UIHabanero Tier Override Key Structure (by ref)"=Make Fort Compete UIHabanero Tier Override Key!(
     RankedTrack:Name=String To Name(InString=@n2)
     Season:String
     Rank:int=@n2.Rank
     MatchmakingRating:int=@n2.CurrentMMR
     CurrentPosition:int=@n2.CurrentPosition
    ) -> (FortCompeteUIHabaneroTierOverrideKey:Fort Compete UIHabanero Tier Override Key Structure) {is=title K2Node_MakeStruct}
   ) -> (OutData:Fort Habanero Tier Override Structure, ReturnValue:bool)
   > n4 = Branch(Condition=@n3.ReturnValue)
   n5 = Get LazyImage_Habanero_RankIcon() -> (LazyImage_Habanero_RankIcon:Fort Lazy Image Object Reference)
   > SetVisibility(self=@n5, InVisibility=HitTestInvisible)
   > SetBrushFromSoftTexture(
    self=@n5
    SoftTexture=Break Fort Habanero Tier!(FortHabaneroTier:"Fort Habanero Tier Structure (by ref)"=@n3) -> (
     ForegroundIcon:Material Instance Soft Object Reference
     ForegroundIconSmall:Material Instance Soft Object Reference
     BackgroundIcon:Material Instance Soft Object Reference
     BackgroundIconSmall:Material Instance Soft Object Reference
     WhiteIcon:Texture 2D Soft Object Reference
     SimpleIconSmall:Texture 2D Soft Object Reference
     Name:Text
     TierType:Fort Habanero Tier Type Data Object Reference
     AdditionalLeftIcon:Material Instance Soft Object Reference
     AdditionalRightIcon:Material Instance Soft Object Reference
    ) {is=title K2Node_BreakStruct}.SimpleIconSmall
    bMatchSize=false
   )
   SetVisibility(execute=@n4.else, self=@n5, InVisibility=Collapsed)
  ]}
]
variables [
 MaxSkullsToDisplay {type=integer, default=10}
 RankData {type=FortHabaneroDisplayData, default=/Game/UI/Habanero/HabaneroDisplayData}
 RankInfoOverrides {type=FortCompeteUIHabaneroTierOverrideList, default=/Game/UI/Habanero/DA_RankDisplayDataOverride}
 SkullAnimStaggerTime {type=number}
 SkullWidgetPool {type="array<WBP_BRGameUI_SkullAnim_C>", default []}
 SkullWidgets {type="array<WBP_BRGameUI_SkullAnim_C>", default []}
]
widgets {
 info {
  inheritedWidgetCount=0
  namedSlotCount=1
  parentClass=&/WBP_BRGameUI_ActionAlertBase.WBP_BRGameUI_ActionAlertBase_C
  rootWidgetClass=null
  widgetCount=28
 }
 widgets [
  Overlay_EliminationRoot {widgetClassPath=&Overlay, bIsVariable=true, widget=Overlay_EliminationRoot, properties {bOverride_Cursor=true}}
  HBox_KillPlayerInfo {
   widgetClassPath=&HorizontalBox
   parent=Overlay_EliminationRoot
   widget=HBox_KillPlayerInfo
   slot {class=OverlaySlot, properties {horizontalAlignment=HAlign_Center, verticalAlignment=VAlign_Center}}
  }
  Image_Bolt_Left {
   widgetClassPath=&Image
   parent=HBox_KillPlayerInfo
   widget=Image_Bolt_Left
   properties {
    visibility=SelfHitTestInvisible
    brush {
     bIsDynamicallyLoaded=false
     drawAs=Image
     imageType=FullColor
     mirroring=NoMirror
     resourceName=None
     resourceObject=/BattleRoyaleGameUI/HUD/PlayerActionAlert/Assets/MI_UI_Elimination_Bolt_Left
     tiling=NoTile
     imageSize=<192 20>
     margin=<ltrb 0 0 0 0>
     outlineSettings {
      bUseBrushTransparency=false
      roundingType=HalfHeightRadius
      width=0
      color {colorUseRule=UseColor_Specified, specifiedColor=<rgba 0 0 0 0>}
      cornerRadii=<0 0 0 0>
     }
     tintColor {colorUseRule=UseColor_Specified, specifiedColor=<rgba 1 1 1 1>}
     uVRegion {bIsValid=false, max=<0 0>, min=<0 0>}
    }
    renderTransform {angle=0, scale=<1 1>, shear=<16 0>, translation=<0 0>}
    renderTransformPivot=<1 0.5>
   }
   slot {class=HorizontalBoxSlot, properties {verticalAlignment=VAlign_Bottom, padding=<ltrb 0 0 8 6>}}
  }
  Overlay_Elimination {
   widgetClassPath=&Overlay
   parent=HBox_KillPlayerInfo
   widget=Overlay_Elimination
   slot {class=HorizontalBoxSlot, properties {verticalAlignment=VAlign_Center}}
  }
  BackingIcons+FX {
   widgetClassPath=&FortVisualAttachment
   parent=Overlay_Elimination
   widget=BackingIcons+FX
   properties {heightOverride=144, widthOverride=144, contentAnchor=<0.5 0.4>}
   slot {class=OverlaySlot, properties {horizontalAlignment=HAlign_Center}}
  }
  Overlay_FX_Elements {widgetClassPath=&Overlay, bIsVariable=true, parent=BackingIcons+FX, widget=Overlay_FX_Elements, slot {class=SizeBoxSlot}}
  ParticlesAndSkullGlow {
   widgetClassPath=&Overlay
   parent=Overlay_FX_Elements
   widget=ParticlesAndSkullGlow
   properties {visibility=Hidden}
   slot {class=OverlaySlot, properties {horizontalAlignment=HAlign_Fill, verticalAlignment=VAlign_Fill}}
  }
  Image_DarkParticles {
   widgetClassPath=&Image
   parent=ParticlesAndSkullGlow
   widget=Image_DarkParticles
   properties {
    visibility=SelfHitTestInvisible
    brush {
     bIsDynamicallyLoaded=false
     drawAs=Image
     imageType=FullColor
     mirroring=NoMirror
     resourceName=None
     resourceObject=/BattleRoyaleGameUI/HUD/PlayerActionAlert/Assets/MI_UI_ElimParticles_Static_Dark
     tiling=NoTile
     imageSize=<144 144>
     margin=<ltrb 0 0 0 0>
     outlineSettings {
      bUseBrushTransparency=false
      roundingType=HalfHeightRadius
      width=0
      color {colorUseRule=UseColor_Specified, specifiedColor=<rgba 0 0 0 0>}
      cornerRadii=<0 0 0 0>
     }
     tintColor {colorUseRule=UseColor_Specified, specifiedColor=<rgba 1 1 1 1>}
     uVRegion {bIsValid=false, max=<0 0>, min=<0 0>}
    }
    renderTransform {angle=90, scale=<1 1>, shear=<0 0>, translation=<0 0>}
   }
   slot {class=OverlaySlot, properties {horizontalAlignment=HAlign_Fill, verticalAlignment=VAlign_Center, padding=<ltrb 0 0 0 8>}}
  }
  Image_LargeRedParticles {
   widgetClassPath=&Image
   parent=ParticlesAndSkullGlow
   widget=Image_LargeRedParticles
   properties {
    visibility=SelfHitTestInvisible
    brush {
     bIsDynamicallyLoaded=false
     drawAs=Image
     imageType=FullColor
     mirroring=NoMirror
     resourceName=None
     resourceObject=/BattleRoyaleGameUI/HUD/PlayerActionAlert/Assets/MI_UI_ElimParticles_Static_Red
     tiling=NoTile
     imageSize=<144 144>
     margin=<ltrb 0 0 0 0>
     outlineSettings {
      bUseBrushTransparency=false
      roundingType=HalfHeightRadius
      width=0
      color {colorUseRule=UseColor_Specified, specifiedColor=<rgba 0 0 0 0>}
      cornerRadii=<0 0 0 0>
     }
     tintColor {colorUseRule=UseColor_Specified, specifiedColor=<rgba 1 1 1 1>}
     uVRegion {bIsValid=false, max=<0 0>, min=<0 0>}
    }
   }
   slot {class=OverlaySlot, properties {horizontalAlignment=HAlign_Fill, verticalAlignment=VAlign_Fill}}
  }
  Image_BurstSkull {
   widgetClassPath=&Image
   parent=ParticlesAndSkullGlow
   widget=Image_BurstSkull
   properties {
    visibility=Collapsed
    brush {
     bIsDynamicallyLoaded=false
     drawAs=Image
     imageType=FullColor
     mirroring=NoMirror
     resourceName=None
     resourceObject=/BattleRoyaleGameUI/HUD/PlayerActionAlert/Assets/MI_UI_ElimIcon_BurstGlow
     tiling=NoTile
     imageSize=<144 144>
     margin=<ltrb 0 0 0 0>
     outlineSettings {
      bUseBrushTransparency=false
      roundingType=HalfHeightRadius
      width=0
      color {colorUseRule=UseColor_Specified, specifiedColor=<rgba 0 0 0 0>}
      cornerRadii=<0 0 0 0>
     }
     tintColor {colorUseRule=UseColor_Specified, specifiedColor=<rgba 1 1 1 1>}
     uVRegion {bIsValid=false, max=<0 0>, min=<0 0>}
    }
   }
   slot {class=OverlaySlot, properties {horizontalAlignment=HAlign_Center, verticalAlignment=VAlign_Center, padding=<ltrb 0 0 0 12>}}
  }
  Image_LargeSkull {
   widgetClassPath=&Image
   parent=Overlay_FX_Elements
   widget=Image_LargeSkull
   properties {
    visibility=Hidden
    brush {
     bIsDynamicallyLoaded=false
     drawAs=Image
     imageType=FullColor
     mirroring=NoMirror
     resourceName=None
     resourceObject=/BattleRoyaleGameUI/HUD/PlayerActionAlert/Assets/MI_UI_LargeElimIcon_Back
     tiling=NoTile
     imageSize=<160 160>
     margin=<ltrb 0 0 0 0>
     outlineSettings {
      bUseBrushTransparency=false
      roundingType=HalfHeightRadius
      width=0
      color {colorUseRule=UseColor_Specified, specifiedColor=<rgba 0 0 0 0>}
      cornerRadii=<0 0 0 0>
     }
     tintColor {colorUseRule=UseColor_Specified, specifiedColor=<rgba 1 1 1 1>}
     uVRegion {bIsValid=false, max=<0 0>, min=<0 0>}
    }
   }
   slot {class=OverlaySlot, properties {horizontalAlignment=HAlign_Center, verticalAlignment=VAlign_Center, padding=<ltrb 0 -16 0 0>}}
  }
  Glow {
   widgetClassPath=&Image
   parent=Overlay_Elimination
   widget=Glow
   properties {
    visibility=SelfHitTestInvisible
    brush {
     bIsDynamicallyLoaded=false
     drawAs=Image
     imageType=FullColor
     mirroring=NoMirror
     resourceName=None
     resourceObject=/BattleRoyaleGameUI/HUD/PlayerActionAlert/Assets/MI_UI_AngledBox_Glow_Elim
     tiling=NoTile
     imageSize=<32 32>
     margin=<ltrb 0 0 0 0>
     outlineSettings {
      bUseBrushTransparency=false
      roundingType=HalfHeightRadius
      width=0
      color {colorUseRule=UseColor_Specified, specifiedColor=<rgba 0 0 0 0>}
      cornerRadii=<0 0 0 0>
     }
     tintColor {colorUseRule=UseColor_Specified, specifiedColor=<rgba 1 1 1 1>}
     uVRegion {bIsValid=false, max=<0 0>, min=<0 0>}
    }
   }
   slot {class=OverlaySlot, properties {horizontalAlignment=HAlign_Fill, verticalAlignment=VAlign_Fill, padding=<ltrb -56 -24 -56 -56>}}
  }
  FlavorText {
   widgetClassPath=&FortVisualAttachment
   parent=Overlay_Elimination
   widget=FlavorText
   properties {contentAnchor=<0.5 1>}
   slot {class=OverlaySlot, properties {horizontalAlignment=HAlign_Center}}
  }
  Text_Background {
   widgetClassPath=&CommonTextBlock
   bIsVariable=true
   parent=FlavorText
   widget=Text_Background
   properties {
    justification=Center
    mobileFontSizeMultiplier=0.85
    text="Eliminated!"
    textTransformPolicy=ToUpper
    font {
     bForceMonospaced=false
     bMaterialIsStencil=false
     fontMaterial=/BattleRoyaleGameUI/Shared/M_UI_Font_FlavorText_Master
     fontObject=&Burbank
     letterSpacing=0
     monospacedWidth=1
     size=24
     skewAmount=0.25
     typefaceFontName=BurbankRegular-Black
     outlineSettings {
      bApplyOutlineToDropShadows=false
      bMiteredCorners=false
      bSeparateFillAlpha=false
      outlineMaterial=/BattleRoyaleGameUI/Shared/MI_UI_Font_SoftShadow_LargeText
      outlineSize=10
      outlineColor=<rgba 1 1 1 1>
     }
    }
    margin=<ltrb 0 -12 0 -16>
    shadowOffset=<0 0>
    strikeBrush {
     bIsDynamicallyLoaded=false
     drawAs=Image
     imageType=NoImage
     mirroring=NoMirror
     resourceName=None
     resourceObject=null
     tiling=NoTile
     imageSize=<32 32>
     margin=<ltrb 0 0 0 0>
     outlineSettings {
      bUseBrushTransparency=false
      roundingType=HalfHeightRadius
      width=0
      color {colorUseRule=UseColor_Specified, specifiedColor=<rgba 0 0 0 0>}
      cornerRadii=<0 0 0 0>
     }
     tintColor {colorUseRule=UseColor_Specified, specifiedColor=<rgba 0.9 0.9 0.9 1>}
     uVRegion {bIsValid=false, max=<0 0>, min=<0 0>}
    }
   }
   slot {class=SizeBoxSlot, properties {padding=<ltrb -2 0 0 0>}}
  }
  Backing {
   widgetClassPath=&Image
   parent=Overlay_Elimination
   widget=Backing
   properties {
    visibility=SelfHitTestInvisible
    brush {
     bIsDynamicallyLoaded=false
     drawAs=Image
     imageType=FullColor
     mirroring=NoMirror
     resourceName=None
     resourceObject=/BattleRoyaleGameUI/HUD/PlayerActionAlert/Assets/M_UI_Elimination_Backing
     tiling=NoTile
     imageSize=<32 32>
     margin=<ltrb 0 0 0 0>
     outlineSettings {
      bUseBrushTransparency=false
      roundingType=HalfHeightRadius
      width=0
      color {colorUseRule=UseColor_Specified, specifiedColor=<rgba 0 0 0 0>}
      cornerRadii=<0 0 0 0>
     }
     tintColor {colorUseRule=UseColor_Specified, specifiedColor=<rgba 1 1 1 1>}
     uVRegion {bIsValid=false, max=<0 0>, min=<0 0>}
    }
   }
   slot {class=OverlaySlot, properties {horizontalAlignment=HAlign_Fill, verticalAlignment=VAlign_Bottom}}
  }
  Image_Bracket_Left {
   widgetClassPath=&Image
   parent=Overlay_Elimination
   widget=Image_Bracket_Left
   properties {
    visibility=SelfHitTestInvisible
    brush {
     bIsDynamicallyLoaded=false
     drawAs=Image
     imageType=FullColor
     mirroring=NoMirror
     resourceName=None
     resourceObject=/BattleRoyaleGameUI/HUD/PlayerActionAlert/Assets/M_UI_Elim_Bracket
     tiling=NoTile
     imageSize=<56 48>
     margin=<ltrb 0 0 0 0>
     outlineSettings {
      bUseBrushTransparency=false
      roundingType=HalfHeightRadius
      width=0
      color {colorUseRule=UseColor_Specified, specifiedColor=<rgba 0 0 0 0>}
      cornerRadii=<0 0 0 0>
     }
     tintColor {colorUseRule=UseColor_Specified, specifiedColor=<rgba 1 1 1 1>}
     uVRegion {bIsValid=false, max=<0 0>, min=<0 0>}
    }
    renderTransformPivot=<1 0.5>
   }
   slot {class=OverlaySlot, properties {verticalAlignment=VAlign_Center, padding=<ltrb -12 0 0 -6>}}
  }
  Image_Bracket_Right {
   widgetClassPath=&Image
   parent=Overlay_Elimination
   widget=Image_Bracket_Right
   properties {
    visibility=SelfHitTestInvisible
    brush {
     bIsDynamicallyLoaded=false
     drawAs=Image
     imageType=FullColor
     mirroring=NoMirror
     resourceName=None
     resourceObject=/BattleRoyaleGameUI/HUD/PlayerActionAlert/Assets/M_UI_Elim_Bracket_Flipped
     tiling=NoTile
     imageSize=<56 48>
     margin=<ltrb 0 0 0 0>
     outlineSettings {
      bUseBrushTransparency=false
      roundingType=HalfHeightRadius
      width=0
      color {colorUseRule=UseColor_Specified, specifiedColor=<rgba 0 0 0 0>}
      cornerRadii=<0 0 0 0>
     }
     tintColor {colorUseRule=UseColor_Specified, specifiedColor=<rgba 1 1 1 1>}
     uVRegion {bIsValid=false, max=<0 0>, min=<0 0>}
    }
    renderTransformPivot=<0 0.5>
   }
   slot {class=OverlaySlot, properties {horizontalAlignment=HAlign_Right, verticalAlignment=VAlign_Center, padding=<ltrb 0 0 -12 -6>}}
  }
  LazyImage_PlayerPortrait {
   widgetClassPath=&FortLazyImage
   bIsVariable=true
   parent=Overlay_Elimination
   widget=LazyImage_PlayerPortrait
   properties {
    materialTextureParamName=Portrait
    brush {
     bIsDynamicallyLoaded=false
     drawAs=Image
     imageType=FullColor
     mirroring=NoMirror
     resourceName=None
     resourceObject=/BattleRoyaleGameUI/HUD/PlayerActionAlert/Assets/M_UI_ElimPortrait
     tiling=NoTile
     imageSize=<44 44>
     margin=<ltrb 0 0 0 0>
     outlineSettings {
      bUseBrushTransparency=false
      roundingType=HalfHeightRadius
      width=0
      color {colorUseRule=UseColor_Specified, specifiedColor=<rgba 0 0 0 0>}
      cornerRadii=<0 0 0 0>
     }
     tintColor {colorUseRule=UseColor_Specified, specifiedColor=<rgba 1 1 1 1>}
     uVRegion {bIsValid=false, max=<0 0>, min=<0 0>}
    }
    loadingBackgroundBrush {
     bIsDynamicallyLoaded=false
     drawAs=NoDrawType
     imageType=NoImage
     mirroring=NoMirror
     resourceName=None
     resourceObject=null
     tiling=NoTile
     imageSize=<64 64>
     margin=<ltrb 0 0 0 0>
     outlineSettings {
      bUseBrushTransparency=false
      roundingType=HalfHeightRadius
      width=0
      color {colorUseRule=UseColor_Specified, specifiedColor=<rgba 0 0 0 0>}
      cornerRadii=<0 0 0 0>
     }
     tintColor {colorUseRule=UseColor_Specified, specifiedColor=<rgba 1 1 1 1>}
     uVRegion {bIsValid=false, max=<0 0>, min=<0 0>}
    }
   }
   slot {class=OverlaySlot, properties {verticalAlignment=VAlign_Bottom, padding=<ltrb -2 0 0 0>}}
  }
  Overlay_CornerArrows {
   widgetClassPath=&Overlay
   parent=Overlay_Elimination
   widget=Overlay_CornerArrows
   properties {visibility=Hidden}
   slot {class=OverlaySlot, properties {horizontalAlignment=HAlign_Fill, verticalAlignment=VAlign_Fill, padding=<ltrb -20 -4 -20 -12>}}
  }
  Image_CornerArrow_TopLeft {
   widgetClassPath=&Image
   parent=Overlay_CornerArrows
   widget=Image_CornerArrow_TopLeft
   properties {
    visibility=SelfHitTestInvisible
    brush {
     bIsDynamicallyLoaded=false
     drawAs=Image
     imageType=FullColor
     mirroring=NoMirror
     resourceName=None
     resourceObject=/BattleRoyaleGameUI/HUD/PlayerActionAlert/Assets/M_UI_CornerArrow_Elim
     tiling=NoTile
     imageSize=<32 32>
     margin=<ltrb 0 0 0 0>
     outlineSettings {
      bUseBrushTransparency=false
      roundingType=HalfHeightRadius
      width=0
      color {colorUseRule=UseColor_Specified, specifiedColor=<rgba 0 0 0 0>}
      cornerRadii=<0 0 0 0>
     }
     tintColor {colorUseRule=UseColor_Specified, specifiedColor=<rgba 1 1 1 1>}
     uVRegion {bIsValid=false, max=<0 0>, min=<0 0>}
    }
    renderTransform {angle=2, scale=<1 1>, shear=<34 22>, translation=<0 0>}
   }
   slot {class=OverlaySlot}
  }
  Image_CornerArrow_TopRight {
   widgetClassPath=&Image
   parent=Overlay_CornerArrows
   widget=Image_CornerArrow_TopRight
   properties {
    visibility=SelfHitTestInvisible
    brush {
     bIsDynamicallyLoaded=false
     drawAs=Image
     imageType=FullColor
     mirroring=NoMirror
     resourceName=None
     resourceObject=&M_UI_CornerArrow_Elim
     tiling=NoTile
     imageSize=<32 32>
     margin=<ltrb 0 0 0 0>
     outlineSettings {
      bUseBrushTransparency=false
      roundingType=HalfHeightRadius
      width=0
      color {colorUseRule=UseColor_Specified, specifiedColor=<rgba 0 0 0 0>}
      cornerRadii=<0 0 0 0>
     }
     tintColor {colorUseRule=UseColor_Specified, specifiedColor=<rgba 1 1 1 1>}
     uVRegion {bIsValid=false, max=<0 0>, min=<0 0>}
    }
    renderTransform {angle=-2, scale=<-1 1>, shear=<-34 -22>, translation=<0 0>}
   }
   slot {class=OverlaySlot, properties {horizontalAlignment=HAlign_Right}}
  }
  Image_CornerArrow_BottomLeft {
   widgetClassPath=&Image
   parent=Overlay_CornerArrows
   widget=Image_CornerArrow_BottomLeft
   properties {
    visibility=SelfHitTestInvisible
    brush {
     bIsDynamicallyLoaded=false
     drawAs=Image
     imageType=FullColor
     mirroring=NoMirror
     resourceName=None
     resourceObject=&M_UI_CornerArrow_Elim
     tiling=NoTile
     imageSize=<24 24>
     margin=<ltrb 0 0 0 0>
     outlineSettings {
      bUseBrushTransparency=false
      roundingType=HalfHeightRadius
      width=0
      color {colorUseRule=UseColor_Specified, specifiedColor=<rgba 0 0 0 0>}
      cornerRadii=<0 0 0 0>
     }
     tintColor {colorUseRule=UseColor_Specified, specifiedColor=<rgba 1 1 1 1>}
     uVRegion {bIsValid=false, max=<0 0>, min=<0 0>}
    }
    renderTransform {angle=-4, scale=<1 -1>, shear=<-8 -24>, translation=<0 0>}
   }
   slot {class=OverlaySlot, properties {verticalAlignment=VAlign_Bottom, padding=<ltrb 14 0 0 0>}}
  }
  Image_CornerArrow_BottomRight {
   widgetClassPath=&Image
   parent=Overlay_CornerArrows
   widget=Image_CornerArrow_BottomRight
   properties {
    visibility=SelfHitTestInvisible
    brush {
     bIsDynamicallyLoaded=false
     drawAs=Image
     imageType=FullColor
     mirroring=NoMirror
     resourceName=None
     resourceObject=&M_UI_CornerArrow_Elim
     tiling=NoTile
     imageSize=<24 24>
     margin=<ltrb 0 0 0 0>
     outlineSettings {
      bUseBrushTransparency=false
      roundingType=HalfHeightRadius
      width=0
      color {colorUseRule=UseColor_Specified, specifiedColor=<rgba 0 0 0 0>}
      cornerRadii=<0 0 0 0>
     }
     tintColor {colorUseRule=UseColor_Specified, specifiedColor=<rgba 1 1 1 1>}
     uVRegion {bIsValid=false, max=<0 0>, min=<0 0>}
    }
    renderTransform {angle=4, scale=<-1 -1>, shear=<8 24>, translation=<0 0>}
   }
   slot {class=OverlaySlot, properties {horizontalAlignment=HAlign_Right, verticalAlignment=VAlign_Bottom, padding=<ltrb 14 0 14 0>}}
  }
  HBox_RankIconAndUsername {
   widgetClassPath=&HorizontalBox
   parent=Overlay_Elimination
   widget=HBox_RankIconAndUsername
   slot {class=OverlaySlot, properties {horizontalAlignment=HAlign_Center, verticalAlignment=VAlign_Center, padding=<ltrb 32 6 32 0>}}
  }
  LazyImage_Habanero_RankIcon {
   widgetClassPath=&FortLazyImage
   bIsVariable=true
   parent=HBox_RankIconAndUsername
   widget=LazyImage_Habanero_RankIcon
   properties {
    brush {
     bIsDynamicallyLoaded=false
     drawAs=Image
     imageType=NoImage
     mirroring=NoMirror
     resourceName=None
     resourceObject=null
     tiling=NoTile
     imageSize=<28 28>
     margin=<ltrb 0 0 0 0>
     outlineSettings {
      bUseBrushTransparency=false
      roundingType=HalfHeightRadius
      width=0
      color {colorUseRule=UseColor_Specified, specifiedColor=<rgba 0 0 0 0>}
      cornerRadii=<0 0 0 0>
     }
     tintColor {colorUseRule=UseColor_Specified, specifiedColor=<rgba 1 1 1 1>}
     uVRegion {bIsValid=false, max=<0 0>, min=<0 0>}
    }
    loadingBackgroundBrush {
     bIsDynamicallyLoaded=false
     drawAs=NoDrawType
     imageType=NoImage
     mirroring=NoMirror
     resourceName=None
     resourceObject=null
     tiling=NoTile
     imageSize=<64 64>
     margin=<ltrb 0 0 0 0>
     outlineSettings {
      bUseBrushTransparency=false
      roundingType=HalfHeightRadius
      width=0
      color {colorUseRule=UseColor_Specified, specifiedColor=<rgba 0 0 0 0>}
      cornerRadii=<0 0 0 0>
     }
     tintColor {colorUseRule=UseColor_Specified, specifiedColor=<rgba 1 1 1 1>}
     uVRegion {bIsValid=false, max=<0 0>, min=<0 0>}
    }
    renderTransform {angle=0, scale=<1.2 1.2>, shear=<0 0>, translation=<0 0>}
   }
   slot {class=HorizontalBoxSlot, properties {horizontalAlignment=HAlign_Center, verticalAlignment=VAlign_Center, padding=<ltrb 2 0 2 0>}}
  }
  Text_PlayerName {
   widgetClassPath=&CommonTextBlock
   bIsVariable=true
   parent=HBox_RankIconAndUsername
   widget=Text_PlayerName
   properties {
    justification=Center
    minDesiredWidth=64
    mobileFontSizeMultiplier=1.15
    text="{Player}"
    font {
     bForceMonospaced=false
     bMaterialIsStencil=false
     fontMaterial=/BattleRoyaleGameUI/Shared/M_UI_Font_White
     fontObject=&BurbankSmall
     letterSpacing=0
     monospacedWidth=1
     size=13.5
     skewAmount=0
     typefaceFontName=Black
     outlineSettings {
      bApplyOutlineToDropShadows=false
      bMiteredCorners=false
      bSeparateFillAlpha=true
      outlineMaterial=/BattleRoyaleGameUI/Shared/MI_UI_Font_SoftShadow_SmallText_Black
      outlineSize=5
      outlineColor=<rgba 1 1 1 1>
     }
    }
    margin=<ltrb -2 -4 0 -4>
    shadowOffset=<0 0>
   }
   slot {class=HorizontalBoxSlot, properties {verticalAlignment=VAlign_Center}}
  }
  Image_Bolt_Right {
   widgetClassPath=&Image
   parent=HBox_KillPlayerInfo
   widget=Image_Bolt_Right
   properties {
    visibility=SelfHitTestInvisible
    brush {
     bIsDynamicallyLoaded=false
     drawAs=Image
     imageType=FullColor
     mirroring=NoMirror
     resourceName=None
     resourceObject=/BattleRoyaleGameUI/HUD/PlayerActionAlert/Assets/MI_UI_Elimination_Bolt_Right
     tiling=NoTile
     imageSize=<192 20>
     margin=<ltrb 0 0 0 0>
     outlineSettings {
      bUseBrushTransparency=false
      roundingType=HalfHeightRadius
      width=0
      color {colorUseRule=UseColor_Specified, specifiedColor=<rgba 0 0 0 0>}
      cornerRadii=<0 0 0 0>
     }
     tintColor {colorUseRule=UseColor_Specified, specifiedColor=<rgba 1 1 1 1>}
     uVRegion {bIsValid=false, max=<0 0>, min=<0 0>}
    }
    renderTransform {angle=0, scale=<1 1>, shear=<-16 0>, translation=<0 0>}
    renderTransformPivot=<0 0.5>
   }
   slot {class=HorizontalBoxSlot, properties {verticalAlignment=VAlign_Bottom, padding=<ltrb 8 0 0 6>}}
  }
  HBox_SkullContainer {
   widgetClassPath=&HorizontalBox
   bIsVariable=true
   parent=Overlay_EliminationRoot
   widget=HBox_SkullContainer
   properties {bOverride_Cursor=true, renderTransform {angle=0, scale=<1 1>, shear=<0 0>, translation=<0 -18>}}
   slot {class=OverlaySlot, properties {horizontalAlignment=HAlign_Center, padding=<ltrb 0 12 0 0>}}
  }
 ]
}
