WidgetBlueprint /BattleRoyaleGameUI/HUD/PlayerActionAlert/WBP_BRGameUI_SkydiveLeaderAlert
graphs [
 EventGraph {links [], nodes []}
 HandleAlertData {nodes [
   n1 = HandleAlertData!() -> (
    then
    AlertData_AlertType:EAthenaPlayerActionAlert Enum
    AlertData_DetailText:Text
    AlertData_RankedProgress:Ranked Progress Replicated Data Structure
    AlertData_OtherPlayerState:Fort Player State Athena Object Reference
    AlertData_NumKills:int
   ) {is=title K2Node_FunctionEntry}
   > n2 = Branch(Condition="Equal (Enum)"(A:EAthenaPlayerActionAlert Enum=@n1, B:EAthenaPlayerActionAlert Enum=NewZebulonDroneYou))
   n3 = Get Text_JumpLeader() -> (Text_JumpLeader:Common Text Object Reference)
   > "SetText (Text)"(self=@n3, InText="LOCTEXT(\"You\")")
   n4 = Get Text_JumpMessage() -> (Text_JumpMessage:Common Text Object Reference)
   > "SetText (Text)"(self=@n4, InText="LOCTEXT(\"are the Skydive Leader\")")
   "SetText (Text)"(execute=@n2.else, self=@n3, InText=@n1.AlertData_DetailText)
   > "SetText (Text)"(self=@n4, InText="LOCTEXT(\"is the Skydive Leader\")")
  ]}
]
variables []
widgets {
 info {
  inheritedWidgetCount=0
  namedSlotCount=1
  parentClass=&/WBP_BRGameUI_ActionAlertBase.WBP_BRGameUI_ActionAlertBase_C
  rootWidgetClass=null
  widgetCount=3
 }
 widgets [
  JumpLeaderFeedback {widgetClassPath=&HorizontalBox, widget=JumpLeaderFeedback}
  Text_JumpLeader {
   widgetClassPath=&CommonTextBlock
   bIsVariable=true
   parent=JumpLeaderFeedback
   widget=Text_JumpLeader
   properties {
    text="{Player}"
    colorAndOpacity {colorUseRule=UseColor_Specified, specifiedColor=<rgba 1 0.5833 0 1>}
    font {
     bForceMonospaced=false
     bMaterialIsStencil=false
     fontMaterial=null
     fontObject=&Burbank
     letterSpacing=0
     monospacedWidth=1
     size=24
     skewAmount=0.25
     typefaceFontName=Burbank-Black
     outlineSettings {
      bApplyOutlineToDropShadows=false
      bMiteredCorners=false
      bSeparateFillAlpha=false
      outlineMaterial=null
      outlineSize=0
      outlineColor=<rgba 0 0 0 0>
     }
    }
    shadowColorAndOpacity=<rgba 0 0 0 0.666>
    shadowOffset=<3 4>
   }
   slot {class=HorizontalBoxSlot, properties {horizontalAlignment=HAlign_Center, verticalAlignment=VAlign_Center, padding=<ltrb 0 0 4 0>}}
  }
  Text_JumpMessage {
   widgetClassPath=&CommonTextBlock
   bIsVariable=true
   parent=JumpLeaderFeedback
   widget=Text_JumpMessage
   properties {
    text=is the Skydive Leader
    textTransformPolicy=ToUpper
    font {
     bForceMonospaced=false
     bMaterialIsStencil=false
     fontMaterial=null
     fontObject=&Burbank
     letterSpacing=0
     monospacedWidth=1
     size=18
     skewAmount=0.25
     typefaceFontName=BurbankRegular-Black
     outlineSettings {
      bApplyOutlineToDropShadows=false
      bMiteredCorners=false
      bSeparateFillAlpha=false
      outlineMaterial=null
      outlineSize=0
      outlineColor=<rgba 0 0 0 0>
     }
    }
    shadowColorAndOpacity=<rgba 0 0 0 0.5>
    shadowOffset=<2 3>
   }
   slot {class=HorizontalBoxSlot, properties {verticalAlignment=VAlign_Center, padding=<ltrb 4 0 0 0>}}
  }
 ]
}
