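# Sound cue for the team-eliminated UI sound in the Battle Royale game's team lives feature.
# A single wave player feeds the BRGameUI_TeamEliminated wave straight to the output, with boosted volume and no randomization or modulation.
# It routes through the critical in-game UI sound class, overrides concurrency, and has the debug flag left on.
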SoundCue /BattleRoyaleGameUI/Sound/TeamLives/BRGameUI_TeamEliminated_Cue
root=WavePlayer_1
nodes [{id=WavePlayer_1, properties {soundWaveAssetPtr=&/BRGameUI_TeamEliminated}}]
properties {
 bDebug=true
 bOverrideConcurrency=true
 maxDistance=2097152
 soundClassObject=&Fort_UI_InGame_Critical
 volumeMultiplier=1.4
}
