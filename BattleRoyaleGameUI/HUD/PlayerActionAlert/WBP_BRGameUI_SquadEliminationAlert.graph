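# HUD alert widget that announces an eliminated enemy squad, or enemy player when max team size is one or less.
# On Construct, the team-size check sets the message text and the icon material's IsSolo parameter.
# ShowAlert skips ForbiddenFruit-tagged playlists; otherwise it plays AnimShowAlert and, if visible, the team-eliminated sound.
# The animation drives root visibility, transforms, opacity and material parameters for slide-in, icon slashes, glows, particles and fade-out.
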WidgetBlueprint /BattleRoyaleGameUI/HUD/PlayerActionAlert/WBP_BRGameUI_SquadEliminationAlert
animations [AnimShowAlert {
  binding_count=15
  length_seconds=2.5
  object=&self:AnimShowAlert
  sequence {bindings [
    Particles {
     id=#1
     possessed_class=Image
     tracks [{
       type=MovieScene2DTransformTrack
       sections [{type=MovieScene2DTransformSection, channels [Translation.X {type=MovieSceneScriptingFloatChannel, keys [[8,96],[13,0]]}]}]
      }]
    }
    SquadIcon_Glow {
     id=#2
     possessed_class=Image
     tracks [
      {
       type=MovieScene2DTransformTrack
       sections [{
         type=MovieScene2DTransformSection
         channels [
          Scale.X {type=MovieSceneScriptingFloatChannel, keys [[4.3333,10],[7,1]]}
          Scale.Y {type=MovieSceneScriptingFloatChannel, const [1,2]}
         ]
        }]
      }
      {
       type=MovieSceneFloatTrack
       display_name=Render Opacity
       sections [{type=MovieSceneFloatSection, channels [None {type=MovieSceneScriptingFloatChannel, keys [[1.6667,0],[2.6667,1],[7.6667,1],[10,0]]}]}]
      }
     ]
    }
    Image_Icon {
     id=#3
     possessed_class=Image
     tracks [
      {
       type=MovieSceneFloatTrack
       display_name=Render Opacity
       sections [{type=MovieSceneFloatSection, channels [None {type=MovieSceneScriptingFloatChannel, keys [[5,0],[7,1]]}]}]
      }
      {
       type=MovieScene2DTransformTrack
       sections [{
         type=MovieScene2DTransformSection
         channels [
          Scale.X {
           type=MovieSceneScriptingFloatChannel
           keys [[5,4],[8.6667,0.75],[10,1.5],[14,1],[19.3333,1],[21.3333,0.8],[23.3333,1.1],[25.3333,0.75],[27.6667,1]]
          }
          Scale.Y {
           type=MovieSceneScriptingFloatChannel
           keys [[6,2.5],[8.6667,0.75],[10,1.5],[14,1],[19.3333,1],[21.3333,0.8],[23.3333,1.1],[25.3333,0.75],[27.6667,1]]
          }
         ]
        }]
      }
      {
       type=MovieSceneWidgetMaterialTrack
       sections [{
         type=MovieSceneParameterSection
         channels [
          Anim_ScaleSlash {type=MovieSceneScriptingFloatChannel, keys [[15.6667,0],[20.6667,1.005],[24.6667,1]]}
          Anim_ScaleSlash_Right {type=MovieSceneScriptingFloatChannel, keys [[18.6667,0],[23.6667,1.005],[27.6667,1]]}
          Anim_Color {type=MovieSceneScriptingFloatChannel, keys [[9.6667,0],[15.6667,1],[28,1]]}
         ]
        }]
      }
     ]
    }
    DigiNoise {
     id=#4
     possessed_class=Image
     tracks [{
       type=MovieSceneWidgetMaterialTrack
       sections [{
         type=MovieSceneParameterSection
         end_frame=40
         start_frame=0
         channels [AdditiveMix {type=MovieSceneScriptingFloatChannel, keys [[16,8],[25,0]]}]
        }]
      }]
    }
    Overlay_SquadEliminationRoot {
     id=#5
     possessed_class=Overlay
     tracks [
      {
       type=MovieSceneFloatTrack
       display_name=Render Opacity
       sections [{type=MovieSceneFloatSection, channels [None {type=MovieSceneScriptingFloatChannel, keys [[2.3333,0],[5.3333,1],[42,1],[50,0]]}]}]
      }
      {
       type=MovieScene2DTransformTrack
       sections [{
         type=MovieScene2DTransformSection
         end_frame=50
         start_frame=0
         channels [
          Translation.X {type=MovieSceneScriptingFloatChannel, keys [[2.3333,-448],[8.6667,0]]}
          Translation.Y {type=MovieSceneScriptingFloatChannel, default=0, keys [[42,0],[50,24]]}
          Angle {type=MovieSceneScriptingFloatChannel, default=0}
          Scale.X {type=MovieSceneScriptingFloatChannel, default=1, keys [[42,1],[50,0.75]]}
          Scale.Y {type=MovieSceneScriptingFloatChannel, default=1, keys [[42,1],[50,0.9]]}
          Shear.X {type=MovieSceneScriptingFloatChannel, default=-4}
          Shear.Y {type=MovieSceneScriptingFloatChannel, default=0}
         ]
        }]
      }
      {
       type=MovieSceneByteTrack
       display_name=Visibility
       sections [{type=MovieSceneByteSection, end_frame=50, start_frame=0, channels [None {type=MovieSceneScriptingByteChannel, keys [[0,3],[50,2]]}]}]
      }
     ]
    }
    StaticParticles {
     id=#6
     possessed_class=Image
     tracks [
      {
       type=MovieSceneFloatTrack
       display_name=Render Opacity
       sections [{type=MovieSceneFloatSection, channels [None {type=MovieSceneScriptingFloatChannel, keys [[8,0],[9.3333,1],[18,1],[31,0]]}]}]
      }
      {
       type=MovieScene2DTransformTrack
       sections [{
         type=MovieScene2DTransformSection
         channels [
          Scale.X {type=MovieSceneScriptingFloatChannel, keys [[8,1],[31,3]]}
          Scale.Y {type=MovieSceneScriptingFloatChannel, keys [[8,1],[31,3]]}
         ]
        }]
      }
     ]
    }
    TopBolt {
     id=#7
     possessed_class=Image
     tracks [{
       type=MovieSceneWidgetMaterialTrack
       sections [{type=MovieSceneParameterSection, channels [Animate_BackingColor {type=MovieSceneScriptingFloatChannel, keys [[7.3403,1],[11.6667,0]]}]}]
      }]
    }
    Text_Glow {
     id=#8
     possessed_class=Image
     tracks [{
       type=MovieSceneFloatTrack
       display_name=Render Opacity
       sections [{type=MovieSceneFloatSection, start_frame=2, channels [None {type=MovieSceneScriptingFloatChannel, keys [[10,0],[12,1],[14,1],[20,0]]}]}]
      }]
    }
    DigiNoise_AnimController {
     id=#9
     possessed_class=Border
     tracks [{
       type=MovieSceneFloatTrack
       display_name=Render Opacity
       sections [{type=MovieSceneFloatSection, channels [None {type=MovieSceneScriptingFloatChannel, keys [[7,0],[9.3333,1],[40,0]]}]}]
      }]
    }
    Impact {
     id=#10
     possessed_class=Image
     tracks [
      {
       type=MovieSceneWidgetMaterialTrack
       sections [{
         type=MovieSceneParameterSection
         channels [
          Anim_Alpha {type=MovieSceneScriptingFloatChannel, keys [[7,1],[7.6667,0],[10,0],[11.6667,1]]}
          Anim_DF_Scale {type=MovieSceneScriptingFloatChannel, keys [[9.3333,0],[14,1]]}
          Anim_Overall_Scale {type=MovieSceneScriptingFloatChannel, keys [[7,1],[7.6667,0],[8.6667,0],[14,1]]}
         ]
        }]
      }
      {type=MovieScene2DTransformTrack, sections []}
     ]
    }
    RedCircle {
     id=#11
     possessed_class=Image
     tracks [
      {
       type=MovieSceneFloatTrack
       display_name=Render Opacity
       sections [{type=MovieSceneFloatSection, channels [None {type=MovieSceneScriptingFloatChannel, keys [[9,0],[11.6667,1],[26,0]]}]}]
      }
      {
       type=MovieScene2DTransformTrack
       sections [{
         type=MovieScene2DTransformSection
         channels [
          Scale.X {type=MovieSceneScriptingFloatChannel, keys [[9,1],[26,4]]}
          Scale.Y {type=MovieSceneScriptingFloatChannel, keys [[9,1],[26,4]]}
         ]
        }]
      }
     ]
    }
    Particles_AnimController {
     id=#12
     possessed_class=Border
     tracks [{
       type=MovieSceneFloatTrack
       display_name=Render Opacity
       sections [{type=MovieSceneFloatSection, channels [None {type=MovieSceneScriptingFloatChannel, keys [[8,0],[9,1],[16,1],[40,0]]}]}]
      }]
    }
    OverlayText {
     id=#13
     possessed_class=Overlay
     tracks [
      {
       type=MovieSceneFloatTrack
       display_name=Render Opacity
       sections [{type=MovieSceneFloatSection, channels [None {type=MovieSceneScriptingFloatChannel, keys [[11.351,0],[13.3333,1]]}]}]
      }
      {
       type=MovieScene2DTransformTrack
       sections [{
         type=MovieScene2DTransformSection
         channels [
          Translation.X {type=MovieSceneScriptingFloatChannel, const [0,2]}
          Translation.Y {type=MovieSceneScriptingFloatChannel}
          Scale.X {type=MovieSceneScriptingFloatChannel, keys [[12.3333,0.75],[14.3333,1.25],[17.3333,1]]}
          Scale.Y {type=MovieSceneScriptingFloatChannel, keys [[12.3333,0.75],[14.3333,1.25],[17.3333,1]]}
         ]
        }]
      }
     ]
    }
    BoltsContainer {
     id=#14
     possessed_class=VerticalBox
     tracks [
      {
       type=MovieScene2DTransformTrack
       sections [{
         type=MovieScene2DTransformSection
         channels [
          Scale.X {type=MovieSceneScriptingFloatChannel, keys [[7.3333,0.25],[9.3333,1.333],[13.3333,1]]}
          Scale.Y {type=MovieSceneScriptingFloatChannel, keys [[7.3333,0.25],[9.3333,1]]}
         ]
        }]
      }
      {
       type=MovieSceneFloatTrack
       display_name=Render Opacity
       sections [{type=MovieSceneFloatSection, channels [None {type=MovieSceneScriptingFloatChannel, keys [[7.3333,0],[8.3333,1],[16.3333,1],[32.6667,0]]}]}]
      }
     ]
    }
    BottomBolt {
     id=#15
     possessed_class=Image
     tracks [{
       type=MovieSceneWidgetMaterialTrack
       sections [{type=MovieSceneParameterSection, channels [Animate_BackingColor {type=MovieSceneScriptingFloatChannel, keys [[7.3333,1],[11.6667,0]]}]}]
      }]
    }
   ]}
 }]
graphs [
 EventGraph {nodes [
   Event Construct()
   > n1 = GetDynamicMaterial(self=Get Image_Icon() -> (Image_Icon:Image Object Reference))
   n2 = "integer <= integer"(
    A=GetMaxTeamSize!(
     InWorldContext:Object Reference=GetGameInstance!(self:Widget Object Reference) -> (ReturnValue:Game Instance Object Reference)
     InPlaylist:Fort Playlist Athena Object Reference
    ) -> (ReturnValue:int=0)
    B=1
   )
   > SetScalarParameterValue(self=@n1, ParameterName=IsSolo, Value="To Float (Boolean)"(InBool=@n2))
   > "SetText (Text)"(
    self=Get Text_AlertMessage() -> (Text_AlertMessage:Common Text Object Reference)
    InText=SelectText(A="LOCTEXT(\"Enemy Player Wiped!\")", B="LOCTEXT(\"Enemy Team Wiped!\")", bPickA=@n2)
   )
  ]}
 ShowAlert {nodes [
   ShowAlert() {is=title K2Node_FunctionEntry}
   > n1 = Branch(Condition=PlaylistHasTag!(
     self:Fort Game State Athena Object Reference=GetGameStateAthena!() -> (ReturnValue:Fort Game State Athena Object Reference)
     InTag:"Gameplay Tag Structure (by ref)"=MakeLiteralGameplayTag!(Value:Gameplay Tag Structure=(TagName="Athena.Playlist.ForbiddenFruit")) -> (ReturnValue:Gameplay Tag Structure)
    ) -> (ReturnValue:bool))
   PlayAnimation(execute=@n1.else, InAnimation=Get AnimShowAlert() -> (AnimShowAlert:Widget Animation Object Reference))
   > Branch(Condition=IsVisible())
   > PlaySound2D(Sound=/BattleRoyaleGameUI/Sound/TeamLives/BRGameUI_TeamEliminated_Cue)
  ]}
]
variables []
widgets {
 info {inheritedWidgetCount=0, parentClass=&UserWidget, rootWidgetClass=&Overlay, widgetCount=26}
 widgets [
  Overlay_SquadEliminationRoot {
   widgetClassPath=&Overlay
   widget=Overlay_SquadEliminationRoot
   properties {visibility=Hidden, renderTransform {angle=0, scale=<1 1>, shear=<-4 0>, translation=<0 0>}}
  }
  DigiNoise_AnimController {
   widgetClassPath=&Border
   parent=Overlay_SquadEliminationRoot
   widget=DigiNoise_AnimController
   properties {
    horizontalAlignment=HAlign_Left
    visibility=SelfHitTestInvisible
    background {
     bIsDynamicallyLoaded=false
     drawAs=NoDrawType
     imageType=NoImage
     mirroring=NoMirror
     resourceName=None
     resourceObject=null
     tiling=NoTile
     imageSize=<32 32>
     margin=<ltrb 0 0 0 0>
     outlineSettings {
      bUseBrushTransparency=false
      roundingType=HalfHeightRadius
      width=0
      color {colorUseRule=UseColor_Specified, specifiedColor=<rgba 0 0 0 0>}
      cornerRadii=<0 0 0 0>
     }
     tintColor {colorUseRule=UseColor_Specified, specifiedColor=<rgba 1 1 1 1>}
     uVRegion {bIsValid=false, max=<0 0>, min=<0 0>}
    }
    padding=<ltrb 0 0 0 0>
    renderTransformPivot=<0 0.5>
   }
   slot {class=OverlaySlot, properties {horizontalAlignment=HAlign_Fill, verticalAlignment=VAlign_Fill, padding=<ltrb -64 -64 -64 -64>}}
  }
  NoiseSize {
   widgetClassPath=&SizeBox
   parent=DigiNoise_AnimController
   widget=NoiseSize
   properties {bOverride_WidthOverride=true, widthOverride=352}
   slot {class=BorderSlot, properties {horizontalAlignment=HAlign_Left, padding=<ltrb 0 0 0 0>}}
  }
  DigiNoise {
   widgetClassPath=&Image
   parent=NoiseSize
   widget=DigiNoise
   properties {
    visibility=SelfHitTestInvisible
    brush {
     bIsDynamicallyLoaded=false
     drawAs=Image
     imageType=FullColor
     mirroring=NoMirror
     resourceName=None
     resourceObject=/BattleRoyaleGameUI/HUD/PlayerActionAlert/Assets/M_UI_RedRadialGlow_DigiTiles
     tiling=NoTile
     imageSize=<32 32>
     margin=<ltrb 0 0 0 0>
     outlineSettings {
      bUseBrushTransparency=false
      roundingType=HalfHeightRadius
      width=0
      color {colorUseRule=UseColor_Specified, specifiedColor=<rgba 0 0 0 0>}
      cornerRadii=<0 0 0 0>
     }
     tintColor {colorUseRule=UseColor_Specified, specifiedColor=<rgba 1 1 1 1>}
     uVRegion {bIsValid=false, max=<0 0>, min=<0 0>}
    }
    renderTransform {angle=0, scale=<-1 1>, shear=<0 0>, translation=<0 0>}
   }
   slot {class=SizeBoxSlot}
  }
  Particles_AnimController {
   widgetClassPath=&Border
   parent=Overlay_SquadEliminationRoot
   widget=Particles_AnimController
   properties {
    horizontalAlignment=HAlign_Left
    verticalAlignment=VAlign_Center
    visibility=SelfHitTestInvisible
    background {
     bIsDynamicallyLoaded=false
     drawAs=NoDrawType
     imageType=NoImage
     mirroring=NoMirror
     resourceName=None
     resourceObject=null
     tiling=NoTile
     imageSize=<32 32>
     margin=<ltrb 0 0 0 0>
     outlineSettings {
      bUseBrushTransparency=false
      roundingType=HalfHeightRadius
      width=0
      color {colorUseRule=UseColor_Specified, specifiedColor=<rgba 0 0 0 0>}
      cornerRadii=<0 0 0 0>
     }
     tintColor {colorUseRule=UseColor_Specified, specifiedColor=<rgba 1 1 1 1>}
     uVRegion {bIsValid=false, max=<0 0>, min=<0 0>}
    }
    padding=<ltrb 0 0 0 0>
    renderTransformPivot=<0 0.5>
   }
   slot {class=OverlaySlot, properties {horizontalAlignment=HAlign_Fill, verticalAlignment=VAlign_Fill, padding=<ltrb -40 -64 -64 -64>}}
  }
  FortVisualAttachment_0 {
   widgetClassPath=&FortVisualAttachment
   parent=Particles_AnimController
   widget=FortVisualAttachment_0
   properties {bOverride_HeightOverride=true, bOverride_WidthOverride=true, heightOverride=48, widthOverride=420, contentAnchor=<0.9 0.5>}
   slot {class=BorderSlot, properties {horizontalAlignment=HAlign_Left, verticalAlignment=VAlign_Center, padding=<ltrb 0 0 0 0>}}
  }
  ParticlesClipper {
   widgetClassPath=&Overlay
   parent=FortVisualAttachment_0
   widget=ParticlesClipper
   properties {clipping=ClipToBounds}
   slot {class=SizeBoxSlot}
  }
  Particles {
   widgetClassPath=&Image
   parent=ParticlesClipper
   widget=Particles
   properties {
    visibility=SelfHitTestInvisible
    brush {
     bIsDynamicallyLoaded=false
     drawAs=Image
     imageType=FullColor
     mirroring=NoMirror
     resourceName=None
     resourceObject=/BattleRoyaleGameUI/HUD/PlayerActionAlert/Assets/MI_UI_RedDigiTilesFX_Moving
     tiling=NoTile
     imageSize=<32 32>
     margin=<ltrb 0 0 0 0>
     outlineSettings {
      bUseBrushTransparency=false
      roundingType=HalfHeightRadius
      width=0
      color {colorUseRule=UseColor_Specified, specifiedColor=<rgba 0 0 0 0>}
      cornerRadii=<0 0 0 0>
     }
     tintColor {colorUseRule=UseColor_Specified, specifiedColor=<rgba 1 1 1 1>}
     uVRegion {bIsValid=false, max=<0 0>, min=<0 0>}
    }
    renderTransform {angle=0, scale=<-1 1>, shear=<0 0>, translation=<0 0>}
   }
   slot {class=OverlaySlot, properties {horizontalAlignment=HAlign_Fill, verticalAlignment=VAlign_Fill}}
  }
  Bolts {
   widgetClassPath=&VerticalBox
   parent=Overlay_SquadEliminationRoot
   widget=Bolts
   properties {renderTransform {angle=0, scale=<1 1>, shear=<-16 0>, translation=<0 0>}}
   slot {class=OverlaySlot, properties {horizontalAlignment=HAlign_Center, verticalAlignment=VAlign_Fill}}
  }
  FortVisualAttachment_73 {
   widgetClassPath=&FortVisualAttachment
   parent=Bolts
   widget=FortVisualAttachment_73
   properties {bOverride_HeightOverride=true, bOverride_WidthOverride=true, heightOverride=16, widthOverride=320, contentAnchor=<0.5 -0.1366>}
   slot {class=VerticalBoxSlot, properties {verticalAlignment=VAlign_Center}}
  }
  BoltsContainer {
   widgetClassPath=&VerticalBox
   parent=FortVisualAttachment_73
   widget=BoltsContainer
   properties {renderTransformPivot=<0 1>}
   slot {class=SizeBoxSlot}
  }
  TopBolt {
   widgetClassPath=&Image
   parent=BoltsContainer
   widget=TopBolt
   properties {
    visibility=SelfHitTestInvisible
    brush {
     bIsDynamicallyLoaded=false
     drawAs=Image
     imageType=FullColor
     mirroring=NoMirror
     resourceName=None
     resourceObject=/FortUILibrary/Foundation/Art_ToolKit/AngledBoxes/M_UI_LevelUpAngleLine_Inst_Inst_Inst
     tiling=NoTile
     imageSize=<300 20>
     margin=<ltrb 0 0 0 0>
     outlineSettings {
      bUseBrushTransparency=false
      roundingType=HalfHeightRadius
      width=0
      color {colorUseRule=UseColor_Specified, specifiedColor=<rgba 0 0 0 0>}
      cornerRadii=<0 0 0 0>
     }
     tintColor {colorUseRule=UseColor_Specified, specifiedColor=<rgba 1 1 1 1>}
     uVRegion {bIsValid=false, max=<0 0>, min=<0 0>}
    }
    renderTransformPivot=<0 0.5>
   }
   slot {class=VerticalBoxSlot}
  }
  BottomBolt {
   widgetClassPath=&Image
   parent=BoltsContainer
   widget=BottomBolt
   properties {
    visibility=SelfHitTestInvisible
    brush {
     bIsDynamicallyLoaded=false
     drawAs=Image
     imageType=FullColor
     mirroring=NoMirror
     resourceName=None
     resourceObject=/FortUILibrary/Foundation/Art_ToolKit/AngledBoxes/M_UI_LevelUpAngleLine_Inst_Inst
     tiling=NoTile
     imageSize=<32 24>
     margin=<ltrb 0 0 0 0>
     outlineSettings {
      bUseBrushTransparency=false
      roundingType=HalfHeightRadius
      width=0
      color {colorUseRule=UseColor_Specified, specifiedColor=<rgba 0 0 0 0>}
      cornerRadii=<0 0 0 0>
     }
     tintColor {colorUseRule=UseColor_Specified, specifiedColor=<rgba 1 1 1 1>}
     uVRegion {bIsValid=false, max=<0 0>, min=<0 0>}
    }
    renderTransform {angle=0, scale=<1 -1>, shear=<0 0>, translation=<0 0>}
    renderTransformPivot=<0 0.5>
   }
   slot {class=VerticalBoxSlot}
  }
  Box {widgetClassPath=&HorizontalBox, parent=Overlay_SquadEliminationRoot, widget=Box, slot {class=OverlaySlot}}
  OverlaySquadIcon {
   widgetClassPath=&Overlay
   parent=Box
   widget=OverlaySquadIcon
   properties {renderTransform {angle=0, scale=<1.5 1.5>, shear=<-4 0>, translation=<0 0>}}
   slot {class=HorizontalBoxSlot, properties {verticalAlignment=VAlign_Center}}
  }
  Impact {
   widgetClassPath=&Image
   parent=OverlaySquadIcon
   widget=Impact
   properties {
    visibility=SelfHitTestInvisible
    brush {
     bIsDynamicallyLoaded=false
     drawAs=Image
     imageType=FullColor
     mirroring=NoMirror
     resourceName=None
     resourceObject=/BattleRoyaleGameUI/HUD/PlayerActionAlert/Assets/M_UI_ImpactFrame_FX
     tiling=NoTile
     imageSize=<32 32>
     margin=<ltrb 0 0 0 0>
     outlineSettings {
      bUseBrushTransparency=false
      roundingType=HalfHeightRadius
      width=0
      color {colorUseRule=UseColor_Specified, specifiedColor=<rgba 0 0 0 0>}
      cornerRadii=<0 0 0 0>
     }
     tintColor {colorUseRule=UseColor_Specified, specifiedColor=<rgba 1 1 1 1>}
     uVRegion {bIsValid=false, max=<0 0>, min=<0 0>}
    }
    renderTransform {angle=-2, scale=<2.5 2.5>, shear=<-2 -4>, translation=<0 0>}
   }
   slot {class=OverlaySlot, properties {horizontalAlignment=HAlign_Fill, verticalAlignment=VAlign_Fill}}
  }
  RedCircle {
   widgetClassPath=&Image
   parent=OverlaySquadIcon
   widget=RedCircle
   properties {
    renderOpacity=0
    visibility=SelfHitTestInvisible
    brush {
     bIsDynamicallyLoaded=false
     drawAs=Image
     imageType=FullColor
     mirroring=NoMirror
     resourceName=None
     resourceObject=/BattleRoyaleGameUI/HUD/PlayerActionAlert/Assets/M_UI_RedCircle
     tiling=NoTile
     imageSize=<32 32>
     margin=<ltrb 0 0 0 0>
     outlineSettings {
      bUseBrushTransparency=false
      roundingType=HalfHeightRadius
      width=0
      color {colorUseRule=UseColor_Specified, specifiedColor=<rgba 0 0 0 0>}
      cornerRadii=<0 0 0 0>
     }
     tintColor {colorUseRule=UseColor_Specified, specifiedColor=<rgba 1 1 1 1>}
     uVRegion {bIsValid=false, max=<0 0>, min=<0 0>}
    }
    colorAndOpacity=<rgba 0.05 0.05 0.05 1>
    renderTransform {angle=0, scale=<3 3>, shear=<0 0>, translation=<0 0>}
   }
   slot {class=OverlaySlot, properties {horizontalAlignment=HAlign_Fill, verticalAlignment=VAlign_Fill}}
  }
  StaticParticles {
   widgetClassPath=&Image
   parent=OverlaySquadIcon
   widget=StaticParticles
   properties {
    visibility=SelfHitTestInvisible
    brush {
     bIsDynamicallyLoaded=false
     drawAs=Image
     imageType=FullColor
     mirroring=NoMirror
     resourceName=None
     resourceObject=/BattleRoyaleGameUI/HUD/PlayerActionAlert/Assets/M_UI_RedDigiTilesFX
     tiling=NoTile
     imageSize=<32 32>
     margin=<ltrb 0 0 0 0>
     outlineSettings {
      bUseBrushTransparency=false
      roundingType=HalfHeightRadius
      width=0
      color {colorUseRule=UseColor_Specified, specifiedColor=<rgba 0 0 0 0>}
      cornerRadii=<0 0 0 0>
     }
     tintColor {colorUseRule=UseColor_Specified, specifiedColor=<rgba 1 1 1 1>}
     uVRegion {bIsValid=false, max=<0 0>, min=<0 0>}
    }
    renderTransform {angle=0, scale=<2 2>, shear=<0 0>, translation=<0 0>}
   }
   slot {class=OverlaySlot, properties {horizontalAlignment=HAlign_Fill, verticalAlignment=VAlign_Fill}}
  }
  SquadIcon_SizeControl {
   widgetClassPath=&Border
   parent=OverlaySquadIcon
   widget=SquadIcon_SizeControl
   properties {
    visibility=SelfHitTestInvisible
    background {
     bIsDynamicallyLoaded=false
     drawAs=NoDrawType
     imageType=NoImage
     mirroring=NoMirror
     resourceName=None
     resourceObject=null
     tiling=NoTile
     imageSize=<32 32>
     margin=<ltrb 0 0 0 0>
     outlineSettings {
      bUseBrushTransparency=false
      roundingType=HalfHeightRadius
      width=0
      color {colorUseRule=UseColor_Specified, specifiedColor=<rgba 0 0 0 0>}
      cornerRadii=<0 0 0 0>
     }
     tintColor {colorUseRule=UseColor_Specified, specifiedColor=<rgba 1 1 1 1>}
     uVRegion {bIsValid=false, max=<0 0>, min=<0 0>}
    }
    padding=<ltrb 0 0 0 0>
    renderTransform {angle=0, scale=<1.5 1.5>, shear=<0 0>, translation=<0 0>}
   }
   slot {class=OverlaySlot, properties {horizontalAlignment=HAlign_Center, verticalAlignment=VAlign_Center}}
  }
  SquadIcon_Glow {
   widgetClassPath=&Image
   parent=SquadIcon_SizeControl
   widget=SquadIcon_Glow
   properties {
    visibility=SelfHitTestInvisible
    brush {
     bIsDynamicallyLoaded=false
     drawAs=Image
     imageType=FullColor
     mirroring=NoMirror
     resourceName=None
     resourceObject=/BattleRoyaleGameUI/HUD/PlayerActionAlert/Assets/M_UI_RedRadialGlow
     tiling=NoTile
     imageSize=<48 48>
     margin=<ltrb 0 0 0 0>
     outlineSettings {
      bUseBrushTransparency=false
      roundingType=HalfHeightRadius
      width=0
      color {colorUseRule=UseColor_Specified, specifiedColor=<rgba 0 0 0 0>}
      cornerRadii=<0 0 0 0>
     }
     tintColor {colorUseRule=UseColor_Specified, specifiedColor=<rgba 1 1 1 1>}
     uVRegion {bIsValid=false, max=<0 0>, min=<0 0>}
    }
    renderTransform {angle=0, scale=<1.5 1.5>, shear=<0 0>, translation=<0 0>}
   }
   slot {class=BorderSlot, properties {padding=<ltrb 0 0 0 0>}}
  }
  Border_0 {
   widgetClassPath=&Border
   parent=OverlaySquadIcon
   widget=Border_0
   properties {
    visibility=Collapsed
    background {
     bIsDynamicallyLoaded=false
     drawAs=NoDrawType
     imageType=NoImage
     mirroring=NoMirror
     resourceName=None
     resourceObject=null
     tiling=NoTile
     imageSize=<32 32>
     margin=<ltrb 0 0 0 0>
     outlineSettings {
      bUseBrushTransparency=false
      roundingType=HalfHeightRadius
      width=0
      color {colorUseRule=UseColor_Specified, specifiedColor=<rgba 0 0 0 0>}
      cornerRadii=<0 0 0 0>
     }
     tintColor {colorUseRule=UseColor_Specified, specifiedColor=<rgba 1 1 1 1>}
     uVRegion {bIsValid=false, max=<0 0>, min=<0 0>}
    }
    padding=<ltrb 0 0 0 0>
    renderTransform {angle=0, scale=<1 1>, shear=<1 0>, translation=<0 0>}
   }
   slot {class=OverlaySlot, properties {horizontalAlignment=HAlign_Center, verticalAlignment=VAlign_Center}}
  }
  Slashes_Backing {
   widgetClassPath=&Image
   parent=Border_0
   widget=Slashes_Backing
   properties {
    visibility=SelfHitTestInvisible
    brush {
     bIsDynamicallyLoaded=false
     drawAs=Image
     imageType=FullColor
     mirroring=NoMirror
     resourceName=None
     resourceObject=/BattleRoyaleGameUI/HUD/PlayerActionAlert/Assets/MI_UI_SquadKill_Slashes
     tiling=NoTile
     imageSize=<48 48>
     margin=<ltrb 0 0 0 0>
     outlineSettings {
      bUseBrushTransparency=false
      roundingType=HalfHeightRadius
      width=0
      color {colorUseRule=UseColor_Specified, specifiedColor=<rgba 0 0 0 0>}
      cornerRadii=<0 0 0 0>
     }
     tintColor {colorUseRule=UseColor_Specified, specifiedColor=<rgba 1 1 1 1>}
     uVRegion {bIsValid=false, max=<0 0>, min=<0 0>}
    }
   }
   slot {class=BorderSlot, properties {padding=<ltrb 0 0 0 0>}}
  }
  Image_Icon {
   widgetClassPath=&Image
   bIsVariable=true
   parent=OverlaySquadIcon
   widget=Image_Icon
   properties {
    visibility=SelfHitTestInvisible
    brush {
     bIsDynamicallyLoaded=false
     drawAs=Image
     imageType=FullColor
     mirroring=NoMirror
     resourceName=None
     resourceObject=/BattleRoyaleGameUI/HUD/PlayerActionAlert/Assets/M_UI_SquadKill_Icon
     tiling=NoTile
     imageSize=<48 48>
     margin=<ltrb 0 0 0 0>
     outlineSettings {
      bUseBrushTransparency=false
      roundingType=HalfHeightRadius
      width=0
      color {colorUseRule=UseColor_Specified, specifiedColor=<rgba 0 0 0 0>}
      cornerRadii=<0 0 0 0>
     }
     tintColor {colorUseRule=UseColor_Specified, specifiedColor=<rgba 1 1 1 1>}
     uVRegion {bIsValid=false, max=<0 0>, min=<0 0>}
    }
   }
   slot {class=OverlaySlot, properties {horizontalAlignment=HAlign_Center, verticalAlignment=VAlign_Center}}
  }
  OverlayText {
   widgetClassPath=&Overlay
   parent=Box
   widget=OverlayText
   properties {renderTransformPivot=<0 0.5>}
   slot {class=HorizontalBoxSlot, properties {verticalAlignment=VAlign_Center, padding=<ltrb -8 0 0 0>}}
  }
  Text_Glow {
   widgetClassPath=&Image
   parent=OverlayText
   widget=Text_Glow
   properties {
    visibility=SelfHitTestInvisible
    brush {
     bIsDynamicallyLoaded=false
     drawAs=Box
     imageType=FullColor
     mirroring=NoMirror
     resourceName=None
     resourceObject=&M_UI_RedRadialGlow
     tiling=NoTile
     imageSize=<48 48>
     margin=<ltrb 0.5 0.5 0.5 0.5>
     outlineSettings {
      bUseBrushTransparency=false
      roundingType=HalfHeightRadius
      width=0
      color {colorUseRule=UseColor_Specified, specifiedColor=<rgba 0 0 0 0>}
      cornerRadii=<0 0 0 0>
     }
     tintColor {colorUseRule=UseColor_Specified, specifiedColor=<rgba 1 1 1 1>}
     uVRegion {bIsValid=false, max=<0 0>, min=<0 0>}
    }
    renderTransform {angle=0, scale=<1 1>, shear=<-10 0>, translation=<0 0>}
   }
   slot {class=OverlaySlot, properties {horizontalAlignment=HAlign_Fill, verticalAlignment=VAlign_Fill}}
  }
  Text_AlertMessage {
   widgetClassPath=&CommonTextBlock
   bIsVariable=true
   parent=OverlayText
   widget=Text_AlertMessage
   properties {
    lineHeightPercentage=0.95
    pixelSnapping=Disabled
    text="Enemy Team Wiped!"
    textTransformPolicy=ToUpper
    font {
     bForceMonospaced=false
     bMaterialIsStencil=false
     fontMaterial=null
     fontObject=&Burbank
     letterSpacing=0
     monospacedWidth=1
     size=19.5
     skewAmount=0.25
     typefaceFontName=BurbankRegular-Black
     outlineSettings {
      bApplyOutlineToDropShadows=false
      bMiteredCorners=false
      bSeparateFillAlpha=true
      outlineMaterial=&M-FontShadow
      outlineSize=2
      outlineColor=<rgba 0.005 0.005 0.005 1>
     }
    }
    shadowColorAndOpacity=<rgba 0.0203 0 0.0033 0.8>
    shadowOffset=<4 3>
   }
   slot {class=OverlaySlot, properties {horizontalAlignment=HAlign_Fill, verticalAlignment=VAlign_Center, padding=<ltrb 16 0 16 0>}}
  }
 ]
}
