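# Storm alert widget for the Battle Royale HUD, warning the player that they are in the storm and should run.
# It derives from WBP_BRGameUI_ActionAlertBase and supplies a horizontal row with the storm message and a larger orange "Run!" text.
# Its HandleShowAlert override calls the parent, then plays AnimShowAlert, a short animation that scale-pulses only the "Run!" text.
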WidgetBlueprint /BattleRoyaleGameUI/HUD/PlayerActionAlert/WBP_BRGameUI_StormAlert
animations [AnimShowAlert {
  binding_count=1
  length_seconds=0.7
  object=&self:AnimShowAlert
  sequence {bindings [Text_Run {
     id=#1
     possessed_class=CommonTextBlock
     tracks [{
       type=MovieScene2DTransformTrack
       sections [{
         type=MovieScene2DTransformSection
         end_frame=14
         start_frame=0
         channels [
          Translation.X {type=MovieSceneScriptingFloatChannel}
          Translation.Y {type=MovieSceneScriptingFloatChannel}
          Angle {type=MovieSceneScriptingFloatChannel}
          Scale.X {type=MovieSceneScriptingFloatChannel, keys [[0,1],[6,1],[10,2],[13.9997,1]]}
          Scale.Y {type=MovieSceneScriptingFloatChannel, keys [[0,1],[6,1],[10,2],[13.9997,1]]}
          Shear.X {type=MovieSceneScriptingFloatChannel}
          Shear.Y {type=MovieSceneScriptingFloatChannel}
         ]
        }]
      }]
    }]}
 }]
graphs [
 EventGraph {links [], nodes []}
 HandleShowAlert {nodes [
   HandleShowAlert() {is=title K2Node_FunctionEntry}
   > Parent: HandleShowAlert() {is=title K2Node_CallParentFunction}
   > PlayAnimation(InAnimation=Get AnimShowAlert() -> (AnimShowAlert:Widget Animation Object Reference))
  ]}
]
variables []
widgets {
 info {
  inheritedWidgetCount=0
  namedSlotCount=1
  parentClass=&/WBP_BRGameUI_ActionAlertBase.WBP_BRGameUI_ActionAlertBase_C
  rootWidgetClass=null
  widgetCount=3
 }
 widgets [
  StormFeeedback {widgetClassPath=&HorizontalBox, widget=StormFeeedback}
  Text_StormMessage {
   widgetClassPath=&CommonTextBlock
   parent=StormFeeedback
   widget=Text_StormMessage
   properties {
    text=You are in the storm
    textTransformPolicy=ToUpper
    font {
     bForceMonospaced=false
     bMaterialIsStencil=false
     fontMaterial=null
     fontObject=&Burbank
     letterSpacing=0
     monospacedWidth=1
     size=18
     skewAmount=0.25
     typefaceFontName=BurbankRegular-Black
     outlineSettings {
      bApplyOutlineToDropShadows=false
      bMiteredCorners=false
      bSeparateFillAlpha=false
      outlineMaterial=null
      outlineSize=0
      outlineColor=<rgba 0 0 0 0>
     }
    }
    shadowColorAndOpacity=<rgba 0 0 0 0.5>
    shadowOffset=<2 3>
   }
   slot {class=HorizontalBoxSlot, properties {verticalAlignment=VAlign_Center}}
  }
  Text_Run {
   widgetClassPath=&CommonTextBlock
   parent=StormFeeedback
   widget=Text_Run
   properties {
    text="Run!"
    textTransformPolicy=ToUpper
    colorAndOpacity {colorUseRule=UseColor_Specified, specifiedColor=<rgba 1 0.25 0 1>}
    font {
     bForceMonospaced=false
     bMaterialIsStencil=false
     fontMaterial=null
     fontObject=&Burbank
     letterSpacing=0
     monospacedWidth=1
     size=24
     skewAmount=0.25
     typefaceFontName=Burbank-Black
     outlineSettings {
      bApplyOutlineToDropShadows=false
      bMiteredCorners=false
      bSeparateFillAlpha=false
      outlineMaterial=null
      outlineSize=0
      outlineColor=<rgba 0 0 0 0>
     }
    }
    shadowColorAndOpacity=<rgba 0 0 0 0.666>
    shadowOffset=<3 4>
   }
   slot {class=HorizontalBoxSlot, properties {horizontalAlignment=HAlign_Center, verticalAlignment=VAlign_Center, padding=<ltrb 4 0 0 0>}}
  }
 ]
}
